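# Character material function that makes surfaces look wet using a per-character water depth texture.
# The texture, indexed by the third UV channel's U, blends roughness and specular toward wet values and darkens non-metallic base color, with an Alpha input separating skin from non-skin.
# The effect requires render-to-texture features and High or default quality; ES2, ES3.1, Low and Medium pass attributes through unchanged.
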MaterialFunction /Game/Packages/Fortress_SharedMaterials/Base_Material_Functions/BaseCharacterShaderFunctions/WaterInteraction
graph {nodes [
  n1 = FunctionInput() {inputName=Material, inputType=FunctionInput_MaterialAttributes}
  n2 = GetMaterialAttributes(None=@n1) {attributeGetTypes [
    #1
    #2
    #3
    #4
    #5
    #6
   ]}
  n3 = ScalarParameter() {defaultValue=0.0032, group=Water Interaction, parameterName=Wet - Roughness Level}
  n4 = TextureCoordinate() {coordinateIndex=2}
  n5 = TextureSampleParameter2D(UVs=AppendVector(
    A=Constant() {r=0.5}
    B=MaterialFunctionCall(Float2=@n4) {desc=Water Level UVs, materialFunction=&BreakOutFloat2Components}.R
   )) {
   desc="Currently, each character gets its own texture"
   parameterName=WaterDepthTexture
   samplerSource=SSM_Wrap_WorldGroupSettings
   texture=&T_BlacksRGB
  }
  n6 = LinearInterpolate(
   B=LinearInterpolate(
    A=ScalarParameter() {defaultValue=0.05, group=Water Interaction, parameterName=Wet - Diffuse Darkening Non Skin}
    B=ScalarParameter() {defaultValue=0.5, group=Water Interaction, parameterName=Wet - Diffuse Darkening Skin}
    Alpha=FunctionInput() {bUsePreviewValueAsDefault=true, inputName=Alpha, inputType=FunctionInput_Scalar, sortPriority=2}
   )
   Alpha=Multiply(A=OneMinus(None=@n2.Metallic), B=@n5.R)
  ) {constA=1, constB=0.5, desc="b=wet"}
  n7 = ScalarParameter() {defaultValue=0.5, group=Water Interaction, parameterName=Wet - Spec Level}
  n8 = MaterialFunctionCall!(
   With RTT Features
   WithOut RTT Features
   With RTT Features V2
   WithOut RTT Features V2
   With RTT Features Att=SetMaterialAttributes!(
    MaterialAttributes=@n1
    BaseColor=Multiply(A=@n2.BaseColor, B=@n6)
    Roughness=LinearInterpolate(A=@n2.Roughness, B=@n3, Alpha=@n5.R) {constA=1, constB=0, desc="b=wet"}
    Specular=LinearInterpolate(A=@n2.Specular, B=@n7, Alpha=@n5.R) {desc="b=wet"}
   ) {attributeSetTypes [#1, #2, #5]}
   WithOut RTT Features Att=@n1
  ) {materialFunction=/Game/WaterAndWind/EnableOrDisableWind/Fortnite_RenderToTextureEnabled}
  n9 = FeatureLevelSwitch(Default=@n8.Att, ES2=@n1, ES3_1=@n1, SM4_REMOVED=@n8.Att)
  FunctionOutput(None=QualitySwitch(Default=@n9, Low=@n1, High=@n9, Medium=@n1))
 ]}
