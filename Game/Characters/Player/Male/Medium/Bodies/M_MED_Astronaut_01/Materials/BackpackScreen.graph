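# Material function that layers a glossy, emissive computer screen for the astronaut's backpack onto input material attributes, blended by an Alpha input.
# A gradient image tints the background, with elemental color applied through the screen mask's red channel.
# The green channel confines a horizontally panning soundwave texture that forms a tinted heartline.
# Bump offset gives the mask and heartline parallax depth against the background, which shares only their base scale and offset.
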MaterialFunction /Game/Characters/Player/Male/Medium/Bodies/M_MED_Astronaut_01/Materials/BackpackScreen
graph {nodes [
  n1 = ScalarParameter() {group=Backpack Computer Screen, parameterName=BlankScreen}
  n2 = ScalarParameter() {defaultValue=0.66, group=Backpack Computer Screen, parameterName=Screen_Spec}
  n3 = ScalarParameter() {defaultValue=0.05, group=Backpack Computer Screen, parameterName=Screen_Roughness}
  n4 = VectorParameter() {group=Backpack Computer Screen, parameterName=BG_Fill, defaultValue=<rgba 0 0 0 1>}
  n5 = VectorParameter() {group=Backpack Computer Screen, parameterName=BG_SphereTint, defaultValue=<rgba 0.0008 0.0046 0.01 1>}
  n6 = Add(
   A=Multiply(A=TextureCoordinate(), B=ScalarParameter() {defaultValue=1, group=Backpack Computer Screen, parameterName=ComputerScreenScale})
   B=AppendVector(
    A=ScalarParameter() {group=Backpack Computer Screen, parameterName=Screen_Offset_X}
    B=ScalarParameter() {group=Backpack Computer Screen, parameterName=Screen_Offset_Y}
   )
  )
  n7 = ScalarParameter() {defaultValue=1, group=Backpack Computer Screen, parameterName=BackGround Image Scale}
  n8 = ScalarParameter() {group=Backpack Computer Screen, parameterName=BGScreen_Offset_X}
  n9 = ScalarParameter() {group=Backpack Computer Screen, parameterName=BGScreen_Offset_Y}
  n10 = TextureSampleParameter2D(UVs=Add(A=Multiply(A=@n6, B=@n7), B=AppendVector(A=@n8, B=@n9))) {
   parameterName=Background Image
   samplerType=SAMPLERTYPE_LinearColor
   texture=/Game/Global/Textures/Default/Masks_and_Gradients/LinearCircleGrad
  }
  n11 = VectorParameter() {group=Backpack Computer Screen, parameterName=SolidBackpack_ElementalColor, defaultValue=<rgba 0.3103 0.7124 1 1>}
  n12 = BumpOffset(Coordinate=@n6, Height=ScalarParameter() {defaultValue=-10, group=Backpack Computer Screen, parameterName=BumpOffset})
  n13 = TextureSample(UVs=@n12) {
   samplerSource=SSM_Clamp_WorldGroupSettings
   samplerType=SAMPLERTYPE_LinearColor
   texture=/Game/Characters/Player/Male/Medium/Bodies/M_MED_Astronaut_01/Textures/BackpackScreen
  }
  n14 = VectorParameter() {group=Backpack Computer Screen, parameterName=Heartline_EmissiveTint, defaultValue=<rgba 0 0.1413 0.5365 1>}
  n15 = AppendVector(
   A=ScalarParameter() {defaultValue=0.333, group=Backpack Computer Screen, parameterName=Tile_Heartline_A}
   B=ScalarParameter() {defaultValue=3.5, group=Backpack Computer Screen, parameterName=Tile_Heartline_B}
  )
  n16 = ScalarParameter() {group=Backpack Computer Screen, parameterName=SW_Offset_X}
  n17 = ScalarParameter() {defaultValue=0.8242, group=Backpack Computer Screen, parameterName=SW_Offset_Y}
  n18 = Panner(
   Coordinate=Add(A=Multiply(A=@n12, B=@n15), B=AppendVector(A=@n16, B=@n17))
   Speed=AppendVector(A=ScalarParameter() {defaultValue=0.3, group=Backpack Computer Screen, parameterName=Scroll}, B=Constant())
  )
  n19 = TextureObjectParameter() {parameterName=HeartlineTexture, texture=/Game/Weapons/WeaponSkins/Textures/Soundwave}
  n20 = LinearInterpolate(
   A=LinearInterpolate(A=LinearInterpolate(A=@n4.RGB, B=@n5.RGB, Alpha=@n10.R), B=@n11.RGB, Alpha=@n13.R)
   B=@n14.RGB
   Alpha=Multiply(
    A=@n13.G
    B=TextureSample(UVs=@n18, Tex=@n19) {samplerSource=SSM_Wrap_WorldGroupSettings, texture=&Soundwave}.R
   )
  )
  n21 = Constant()
  n22 = Constant() {r=1}
  n23 = Constant()
  FunctionOutput(None=BlendMaterialAttributes(
    A=FunctionInput() {inputName=Material, inputType=FunctionInput_MaterialAttributes}
    B=MakeMaterialAttributes!(
     BaseColor=@n1
     Metallic
     Specular=@n2
     Roughness=@n3
     Anisotropy
     EmissiveColor=@n20
     Opacity=@n21
     OpacityMask=@n22
     Normal
     Tangent
     WorldPositionOffset
     SubsurfaceColor=@n23
     ClearCoat=@n23
     ClearCoatRoughness=@n23
     AmbientOcclusion
     Refraction
     CustomizedUVs_0
     CustomizedUVs_1
     CustomizedUVs_2
     CustomizedUVs_3
     CustomizedUVs_4
     CustomizedUVs_5
     CustomizedUVs_6
     CustomizedUVs_7
     PixelDepthOffset
     ShadingModel
     Displacement
    )
    Alpha=FunctionInput() {bUsePreviewValueAsDefault=true, inputName=Alpha, inputType=FunctionInput_Scalar}
   ))
 ]}
