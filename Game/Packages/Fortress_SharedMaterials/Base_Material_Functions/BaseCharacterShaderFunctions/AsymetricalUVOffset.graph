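# Material function for Fortress character shaders that overlays an offset, asymmetrical texture onto the base color of incoming material attributes.
# A scalar parameter shifts the U coordinate to sample a texture parameter with a shared clamp sampler, defaulting to black.
# The sampled RGB is both blend color and per-channel lerp alpha, so black texels leave base color untouched.
# Only BaseColor is rewritten before the attributes are output.
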MaterialFunction /Game/Packages/Fortress_SharedMaterials/Base_Material_Functions/BaseCharacterShaderFunctions/AsymetricalUVOffset
graph {nodes [
  n1 = GetMaterialAttributes(None=FunctionInput() {inputType=FunctionInput_MaterialAttributes}) {attributeGetTypes [#1]}
  n2 = ScalarParameter() {group=Asymmetrical Texture, parameterName=Asymmetrical UV Offset}
  n3 = Constant()
  n4 = TextureSampleParameter2D(UVs=Add(A=TextureCoordinate(), B=AppendVector(A=@n2, B=@n3))) {
   group=Asymmetrical Texture
   parameterName=Asymmetrical Texture
   samplerSource=SSM_Clamp_WorldGroupSettings
   texture=/Engine/EngineResources/Black_Low.Black
  }
  FunctionOutput(None=SetMaterialAttributes!(MaterialAttributes=@n1.MaterialAttributes, BaseColor=LinearInterpolate(A=@n1.BaseColor, B=@n4.RGB, Alpha=@n4.RGB)) {attributeSetTypes [#1]})
 ]}
