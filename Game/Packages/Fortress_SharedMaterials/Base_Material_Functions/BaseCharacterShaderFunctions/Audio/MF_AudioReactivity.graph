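# Material function that outputs music Amplitude and Spectrum for audio-reactive materials, under a character shader function path.
# It reads the audio-reactive parameter collection or, via UseSpectrumHistory, interpolated values from SampleAudioMPCHistory with adjustable strength and optional smoothstep.
# Low quality always uses raw collection values unless the UseFakeAudioData switch substitutes manual parameters for both outputs.
# FortniteMaterialParameters EventAudio nodes are present but unconnected.
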MaterialFunction /Game/Packages/Fortress_SharedMaterials/Base_Material_Functions/BaseCharacterShaderFunctions/Audio/MF_AudioReactivity
graph {
 properties {
  bExposeToLibrary=true
  description="outputs the music Amplitude(S) and Spectrum(V4) with parameters for filtering the data across multible frames. also has switches for fake data to easliy conrtol the audio spectrum"
  libraryCategoriesText [FN]
 }
 nodes [
  CollectionParameter() {collection=&FortniteMaterialParameters, parameterName=EventAudio_Spectrum}
  CollectionParameter() {collection=&FortniteMaterialParameters, parameterName=EventAudio_Amplitude}
  n1 = MaterialFunctionCall!(
   useSmoothStep=StaticBoolParameter() {defaultValue=true, group=Audio Reactivity, parameterName=UseSmoothStep}
   uvx=ScalarParameter() {defaultValue=0.1, group=Audio Reactivity, parameterName=Spectrum_InterpolationStrenght, sliderMax=1}
  ) {materialFunction=/Game/Accessories/FORT_Backpacks/Backpack_F_MED_Cactus_Rocker/Materials/SampleAudioMPCHistory}
  n2 = CollectionParameter() {collection=/Game/Athena/Events/AudioAnalysis/MPC_Audio_Reactive_Parameters, parameterName=Amplitude_Music}
  n3 = StaticSwitchParameter(True=@n1.Interpolated Amplitude Music, False=@n2) {defaultValue=true, group=Audio Reactivity, parameterName=UseSpectrumHistory}
  FunctionOutput(None=StaticSwitchParameter(
    True=ScalarParameter() {defaultValue=1, group=Audio Reactivity, parameterName=FakeAudioData_Amplitude, sliderMax=20}
    False=QualitySwitch(Default=@n3, Low=@n2)
   ) {group=Audio Reactivity, parameterName=UseFakeAudioData}) {outputName=Amplitude}
  n4 = VectorParameter() {group=Audio Reactivity, parameterName=FakeAudioData_Spectrum, defaultValue=<rgba 1 1 1 1>}
  n5 = CollectionParameter() {collection=&MPC_Audio_Reactive_Parameters, parameterName=Spectrum_Music}
  n6 = StaticSwitchParameter(True=@n1.Interpolated Spectrum, False=@n5) {defaultValue=true, group=Audio Reactivity, parameterName=UseSpectrumHistory}
  FunctionOutput(None=StaticSwitchParameter(True=AppendVector(A=@n4.RGB, B=@n4.RGB), False=QualitySwitch(Default=@n6, Low=@n5)) {group=Audio Reactivity, parameterName=UseFakeAudioData}) {outputName=Spectrum, sortPriority=1}
  Saturate() {desc=??}
  ShadingPathSwitch() {desc=??}
 ]
}
