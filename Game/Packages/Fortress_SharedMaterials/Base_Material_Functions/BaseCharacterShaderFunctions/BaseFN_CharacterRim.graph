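# Material function for character rim shading, acting as a thin wrapper around the shared Character_Rim function.
# It exposes BalanceDistance, AO, Diffuse and Fresnel inputs, each using its preview value as the default.
# These feed straight into Character_Rim, with Fresnel supplying its N dot C input, and that function's Result becomes the only output.
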MaterialFunction /Game/Packages/Fortress_SharedMaterials/Base_Material_Functions/BaseCharacterShaderFunctions/BaseFN_CharacterRim
graph {nodes [FunctionOutput(None=MaterialFunctionCall!(
    BalanceDistance=FunctionInput() {bUsePreviewValueAsDefault=true, inputName=BalanceDistance, inputType=FunctionInput_Scalar}
    AO Texture=FunctionInput() {bUsePreviewValueAsDefault=true, inputName=AO, inputType=FunctionInput_Scalar}
    Diffuse Texture=FunctionInput() {bUsePreviewValueAsDefault=true, inputName=Diffuse}
    Fresnel N dot C=FunctionInput() {bUsePreviewValueAsDefault=true, inputName=Fresnel, inputType=FunctionInput_Scalar}
   ) {materialFunction=/Game/Packages/Fortress_SharedMaterials/Base_Material_Functions/Character_Rim}.Result)]}
