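# Material function for Fortress character shaders that adds emissive to incoming material attributes and exposes a separate Rim output.
# A shared Fresnel input drives both EmissiveMask, whose result is added to the existing EmissiveColor, and BaseFN_CharacterRim, which also takes AO, Diffuse and BalanceDistance.
# The Normals input is unconnected, and an MF_TeamOutline call receives the modified attributes but feeds no output.
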MaterialFunction /Game/Packages/Fortress_SharedMaterials/Base_Material_Functions/BaseCharacterShaderFunctions/BaseFN_CharacterEmissive
graph {nodes [
  FunctionInput() {inputName=Normals, sortPriority=2}
  n1 = FunctionInput() {inputName=Fresnel, inputType=FunctionInput_Scalar, sortPriority=4}
  FunctionOutput(None=MaterialFunctionCall!(
    AO=FunctionInput() {inputName=AO, inputType=FunctionInput_Scalar, sortPriority=3}
    Diffuse=FunctionInput() {inputName=Diffuse}
    BalanceDistance=FunctionInput() {bUsePreviewValueAsDefault=true, inputName=BalanceDistance, inputType=FunctionInput_Scalar}
    Fresnel=@n1
   ) {materialFunction=&/BaseFN_CharacterRim}.Result) {outputName=Rim}
  n2 = GetMaterialAttributes(None=FunctionInput() {inputName=Material, inputType=FunctionInput_MaterialAttributes, sortPriority=-1}) {attributeGetTypes [#1]}
  n3 = SetMaterialAttributes!(
   MaterialAttributes=@n2.MaterialAttributes
   EmissiveColor=Add(
    A=@n2.EmissiveColor
    B=MaterialFunctionCall!(Fresnel=@n1) {materialFunction=/Game/Packages/Fortress_SharedMaterials/Base_Material_Functions/EmissiveMask}.Result
   )
  ) {attributeSetTypes [#1]}
  FunctionOutput(None=@n3)
  MaterialFunctionCall!(Material=@n3) {materialFunction=/Game/Packages/Fortress_SharedMaterials/Base_Material_Functions/MF_TeamOutline}
 ]}
