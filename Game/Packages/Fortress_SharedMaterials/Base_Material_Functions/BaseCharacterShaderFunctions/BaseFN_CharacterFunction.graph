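# Base character material function that samples Diffuse, Normals, SpecularMasks and M textures, outputting material attributes plus individual map outputs.
# The specular mask splits into specular, metallic and roughness through a green/blue swizzle, with optional metallic power adjustment.
# Roughness remaps between min/max parameters, with an optional mobile range at Low quality.
# A static IsToonMaterial switch replaces diffuse and masks with constants, with PBR_Shading blending the M mask.
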MaterialFunction /Game/Packages/Fortress_SharedMaterials/Base_Material_Functions/BaseCharacterShaderFunctions/BaseFN_CharacterFunction
graph {nodes [
  n1 = Constant()
  n2 = Constant()
  FunctionOutput(None=MakeMaterialAttributes!(
    BaseColor=NamedRerouteUsage()
    Metallic=NamedRerouteUsage()
    Specular=NamedRerouteUsage()
    Roughness=NamedRerouteUsage()
    Anisotropy
    EmissiveColor=@n1
    Opacity=@n1
    OpacityMask
    Normal=NamedRerouteUsage()
    Tangent
    WorldPositionOffset
    SubsurfaceColor=@n2
    ClearCoat=@n2
    ClearCoatRoughness=@n2
    AmbientOcclusion
    Refraction
    CustomizedUVs_0
    CustomizedUVs_1
    CustomizedUVs_2
    CustomizedUVs_3
    CustomizedUVs_4
    CustomizedUVs_5
    CustomizedUVs_6
    CustomizedUVs_7
    PixelDepthOffset
    ShadingModel=ShadingModel()
    Displacement
   ))
  n3 = FunctionInput(Preview=TextureCoordinate()) {bUsePreviewValueAsDefault=true, inputName=TexCoords, inputType=FunctionInput_Vector2}
  n4 = TextureSampleParameter2D(UVs=@n3) {group=Normal, parameterName=Normals, samplerType=SAMPLERTYPE_Normal, texture=&FlatNormalBlue}
  FunctionOutput(None=NamedRerouteDeclaration(None=@n4.RGB) {name=Normals, nodeColor=<rgba 0.51 0.528 1 1>}) {outputName=Normal Map, sortPriority=2}
  n5 = Constant3Vector() {constant=<rgba 0.5 0 0.6 1>}
  n6 = TextureSampleParameter2D(UVs=@n3) {
   desc="R = SpecMask G = Metallic B = Roughness"
   group=Specular
   parameterName=SpecularMasks
   samplerType=SAMPLERTYPE_Masks
   texture=&T_Fortnite_Default_S
  }
  n7 = StaticBoolParameter() {group=Toon_Shading, parameterName=IsToonMaterial, sortPriority=-1}
  n8 = MaterialFunctionCall(Float3=FunctionOutput(None=StaticSwitch(True=@n5, False=@n6.RGB, Value=@n7)) {outputName=SpecularMask, sortPriority=3}) {materialFunction=&BreakOutFloat3Components}
  n9 = StaticBoolParameter() {group=Specular, parameterName=SwizzleRoughnessToGreen}
  n10 = MaterialFunctionCall(Float3=MaterialFunctionCall(X=@n8.R, Y=StaticSwitch(True=@n8.B, False=@n8.B, Value=@n9), Z=StaticSwitch(True=@n8.G, False=@n8.G, Value=@n9)) {materialFunction=&MakeFloat3}.Result) {materialFunction=&BreakOutFloat3Components}
  n11 = NamedRerouteDeclaration(None=@n10.B) {name=PreFX_RoughnessRawPaint, nodeColor=<rgba 1 0.5156 0 1>}
  n12 = LinearInterpolate(
   A=ScalarParameter() {group=Roughness, parameterName=RawRoughnessMin}
   B=ScalarParameter() {defaultValue=1, group=Roughness, parameterName=RawRoughnessMax}
   Alpha=@n11
  )
  n13 = StaticSwitchParameter(
   True=LinearInterpolate(
    A=ScalarParameter() {group=Roughness, parameterName=MobileRoughnessMin}
    B=ScalarParameter() {defaultValue=1, group=Roughness, parameterName=MobileRoughnessMax}
    Alpha=@n11
   )
   False=@n12
  ) {group=Roughness, parameterName=OverrideRoughnessOnMobile, sortPriority=0}
  FunctionOutput(None=NamedRerouteDeclaration(None=QualitySwitch(Default=@n12, Low=@n13)) {name=Roughness, nodeColor=<rgba 0 0.0917 1 1>}) {outputName=Roughness, sortPriority=8}
  n14 = Constant3Vector() {constant=<rgba 1 0 0 0>}
  n15 = Constant3Vector() {constant=<rgba 1 0.5 0 1>}
  n16 = Saturate(None=ScalarParameter() {group=Toon_Shading_PBR_Override, parameterName=PBR_Shading, sliderMax=1, sortPriority=10})
  n17 = TextureSampleParameter2D(UVs=@n3) {
   desc="Masks R=AO G= Fresnel B = Tmission/Skin"
   group=Masks
   parameterName=M
   samplerType=SAMPLERTYPE_Masks
   texture=&T_BlackMask
  }
  FunctionOutput(None=StaticSwitch(True=LinearInterpolate(A=@n14, B=@n15, Alpha=@n16), False=@n17.RGB, Value=@n7)) {outputName=M, sortPriority=4}
  n18 = Constant3Vector()
  n19 = TextureSampleParameter2D(UVs=@n3) {group=Diffuse, parameterName=Diffuse, texture=/Game/Global/Textures/Default/Blanks/T_MidgreysRGB}
  FunctionOutput(None=NamedRerouteDeclaration(None=StaticSwitch(True=@n18, False=@n19.RGB, Value=@n7)) {name=Diffuse, nodeColor=<rgba 1 0.4251 0.1928 1>}) {outputName=DIffuse, sortPriority=1}
  n20 = Power(Base=@n10.G, Exp=ScalarParameter() {defaultValue=0.4, group=Masks, parameterName=MetalMaskPower})
  NamedRerouteDeclaration(None=StaticSwitchParameter(True=Saturate(None=@n20), False=@n10.G) {group=Masks, parameterName=useMetalMaskPowerAdjustment}) {name=Metallic, nodeColor=<rgba 0.0398 1 0 1>}
  NamedRerouteDeclaration(None=@n10.R) {name=Specular, nodeColor=<rgba 1 0 0.1107 1>}
 ]}
