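# MF_HitGlow is a character material function that adds a hit-effect glow to emissive color, black by default.
# A Fresnel term from the material's transformed normal, with adjustable exponent and brightness, is clamped to blend HitGlowOuter into HitGlow.
# HitGlowUsesPulse blends between constant intensity and a half-second sine pulse that scales that color.
# The declared Fresnel input is unconnected, as the function computes its own.
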MaterialFunction /Game/Packages/Fortress_SharedMaterials/Base_Material_Functions/BaseCharacterShaderFunctions/MF_HitGlow
graph {nodes [
  FunctionInput() {bUsePreviewValueAsDefault=true, inputName=Fresnel, inputType=FunctionInput_Scalar, sortPriority=10}
  n1 = GetMaterialAttributes(None=FunctionInput() {inputName=Material, inputType=FunctionInput_MaterialAttributes}) {attributeGetTypes [#1, #2]}
  n2 = Multiply(
   A=ScalarParameter() {defaultValue=1.4, group=Hit Effects, parameterName=HitGlowFresnelBrightness}
   B=Fresnel(ExponentIn=ScalarParameter() {defaultValue=1, group=Hit Effects, parameterName=HitGlowFresnelExponent}, Normal=Transform(None=@n1.Normal)) {baseReflectFraction=0}
  )
  n3 = Sine(None=Time() {bOverride_Period=true, period=0.5}) {period=0.5}
  n4 = Multiply(
   A=LinearInterpolate(
    A=VectorParameter() {group=Hit Effects, parameterName=HitGlowOuter, defaultValue=<rgba 0 0 0 1>}.RGB
    B=VectorParameter() {group=Hit Effects, parameterName=HitGlow, defaultValue=<rgba 0 0 0 1>}.RGB
    Alpha=Saturate(None=@n2) {desc=to prevent the lerp from ussing inverted or out of range colors}
   )
   B=LinearInterpolate(
    B=ConstantBiasScale(None=@n3) {bias=1.5, scale=0.4}
    Alpha=ScalarParameter() {defaultValue=1, group=Hit Effects, parameterName=HitGlowUsesPulse}
   ) {constA=1}
  )
  FunctionOutput(None=SetMaterialAttributes!(MaterialAttributes=@n1.MaterialAttributes, EmissiveColor=Add(A=@n1.EmissiveColor, B=@n4)) {attributeSetTypes [#1]})
 ]}
