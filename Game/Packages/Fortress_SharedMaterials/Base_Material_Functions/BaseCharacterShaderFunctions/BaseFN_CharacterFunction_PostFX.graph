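# Material function that applies the post-effect stack to Fortress character material attributes, exposed to the material library.
# It chains MF_RimV3 rim lighting, hit glow, tech art overrides, character opacity and a de-rez transition, in that order.
# AO, Gmap, distance blend and pre-skinned local position feed the rim stage, while Fresnel feeds hit glow and de-rez.
# The EnableRimlight input is declared but left unconnected.
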MaterialFunction /Game/Packages/Fortress_SharedMaterials/Base_Material_Functions/BaseCharacterShaderFunctions/BaseFN_CharacterFunction_PostFX
graph {
 properties {bExposeToLibrary=true}
 nodes [
  FunctionInput() {bUsePreviewValueAsDefault=true, inputName=EnableRimlight, inputType=FunctionInput_Scalar, previewValue=<1 0 0 1>}
  n1 = FunctionInput(Preview=NamedRerouteDeclaration() {name=PreFX_Result, nodeColor=<rgba 1 0.5156 0 1>}) {bUsePreviewValueAsDefault=true, inputName=Material, inputType=FunctionInput_MaterialAttributes, sortPriority=6}
  n2 = FunctionInput(Preview=NamedRerouteDeclaration() {name=Pre Skinned Local Postition, nodeColor=<rgba 1 0.5156 0 1>}) {bUsePreviewValueAsDefault=true, inputName=Pre Skinned Local Position, sortPriority=5}
  n3 = FunctionInput(Preview=NamedRerouteDeclaration() {name=PreFX_AO, nodeColor=<rgba 1 0.5156 0 1>}) {bUsePreviewValueAsDefault=true, inputName=AO, inputType=FunctionInput_Scalar, sortPriority=1}
  n4 = FunctionInput(Preview=NamedRerouteDeclaration() {name=PreFX_Gmap, nodeColor=<rgba 1 0.5156 0 1>}) {bUsePreviewValueAsDefault=true, inputName=Gmap, inputType=FunctionInput_Scalar, sortPriority=4}
  n5 = FunctionInput(Preview=NamedRerouteDeclaration() {name=PreFX_Distance Blend, nodeColor=<rgba 1 0.5156 0 1>}) {bUsePreviewValueAsDefault=true, inputName=Distance Blend, inputType=FunctionInput_Scalar, sortPriority=2}
  n6 = FunctionInput(Preview=NamedRerouteDeclaration() {name=PreFX_Fresnel, nodeColor=<rgba 1 0.5156 0 1>}) {bUsePreviewValueAsDefault=true, inputName=Fresnel, inputType=FunctionInput_Scalar, sortPriority=3}
  n7 = MaterialFunctionCall(In=MaterialFunctionCall!(
    Material=MaterialFunctionCall!(Material=@n1, Pre Skinned Local Position=@n2, AO=@n3, Gmap=@n4, Distance Blend=@n5) {materialFunction=/Game/Packages/Fortress_SharedMaterials/Base_Material_Functions/MF_RimV3}.Result
    Fresnel=@n6
   ) {materialFunction=&/MF_HitGlow}.Result) {materialFunction=&MF_TechArtOverrides}
  FunctionOutput(None=MaterialFunctionCall!(
    Fresnel=@n6
    Material=MaterialFunctionCall!(Material=@n7.Result) {materialFunction=&/BaseFN_CharacterOpacity}.Result
   ) {materialFunction=/Game/Packages/Fortress_SharedMaterials/Base_Material_Functions/CharacterDeRez_Transition}.Result)
 ]
}
