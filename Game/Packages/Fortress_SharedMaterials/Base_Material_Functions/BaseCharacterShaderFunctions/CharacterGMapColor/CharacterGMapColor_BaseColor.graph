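# Material function for the base color layer of the Fortress character gradient-map (GMap) color shaders.
# It remaps the scalar GMap through three color parameters with Lerp_3Color, or uses the Diffuse input instead via a static switch.
# That color lerps toward the incoming ColorizedGMap by Masks, then saturates to the output.
# A second switch for blending Diffuse with base color is left unconnected to the output.
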MaterialFunction /Game/Packages/Fortress_SharedMaterials/Base_Material_Functions/BaseCharacterShaderFunctions/CharacterGMapColor/CharacterGMapColor_BaseColor
graph {nodes [
  n1 = VectorParameter() {group=Diffuse, parameterName=Base Color: Color B, sortPriority=1, defaultValue=<rgba 0.5 0.5 0.5 1>}
  n2 = MaterialFunctionCall(
   A=VectorParameter() {group=Diffuse, parameterName=Base Color: Color A, sortPriority=1, defaultValue=<rgba 0 0 0 1>}.RGB
   B=@n1.RGB
   C=VectorParameter() {group=Diffuse, parameterName=Base Color: Color C, sortPriority=1, defaultValue=<rgba 1 1 1 1>}.RGB
   Alpha=FunctionInput() {inputName=GMap, inputType=FunctionInput_Scalar, sortPriority=1}
  ) {materialFunction=&Lerp_3Color}
  n3 = FunctionInput() {inputName=Diffuse, sortPriority=2}
  n4 = StaticSwitchParameter(True=@n3, False=@n2.Result) {group=Diffuse, parameterName=Use Diffuse as Base Color}
  n5 = FunctionInput() {inputName=ColorizedGMap, sortPriority=3}
  n6 = FunctionInput() {inputName=Masks, inputType=FunctionInput_Scalar}
  FunctionOutput(None=Saturate(None=LinearInterpolate(A=@n4, B=@n5, Alpha=@n6))) {outputName=ColorizedGMap, sortPriority=1}
  StaticSwitchParameter(
   True=LinearInterpolate(A=@n3, B=@n2.Result, Alpha=ScalarParameter() {group=Diffuse, parameterName=Diffuse to Base Color Blend, sliderMax=1})
   False=@n3
  ) {group=Diffuse, parameterName=Allow Blending Between Diffuse and Base Color}
 ]}
