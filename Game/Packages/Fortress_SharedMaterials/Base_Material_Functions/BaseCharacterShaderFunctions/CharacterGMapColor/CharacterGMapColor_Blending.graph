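# Material function for character shaders that blends up to three colorized GMap inputs into one result and outputs their combined color mask.
# Saturated masks 2 and 3 drive layered lerps over the first GMap, while mask 1 only feeds the clamped mask sum.
# Static switches "Uses 2+ Color Masks" and "Uses 3 Color Masks" select the layer count for both outputs.
# A partly wired duplicate chain with "_1" parameters reaches no output.
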MaterialFunction /Game/Packages/Fortress_SharedMaterials/Base_Material_Functions/BaseCharacterShaderFunctions/CharacterGMapColor/CharacterGMapColor_Blending
graph {nodes [
  n1 = FunctionInput() {inputName=ColorizedGMap 1, sortPriority=1}
  n2 = Saturate(None=FunctionInput() {inputName=Color Mask 2, inputType=FunctionInput_Scalar, sortPriority=2})
  n3 = LinearInterpolate(A=@n1, B=FunctionInput() {inputName=ColorizedGMap 2, sortPriority=3}, Alpha=@n2)
  n4 = Saturate(None=FunctionInput() {inputName=Color Mask 3, inputType=FunctionInput_Scalar, sortPriority=4})
  n5 = LinearInterpolate(A=@n3, B=FunctionInput() {inputName=ColorizedGMap 3, sortPriority=5}, Alpha=@n4)
  n6 = StaticBoolParameter() {group=Diffuse, parameterName=Uses 3 Color Masks}
  n7 = StaticBoolParameter() {defaultValue=true, group=Diffuse, parameterName=Uses 2+ Color Masks}
  FunctionOutput(None=StaticSwitch(True=StaticSwitch(True=@n5, False=@n3, Value=@n6), False=@n1, Value=@n7)) {outputName=ColorizedGMap, sortPriority=1}
  n8 = Saturate(None=FunctionInput() {inputName=Color Mask 1, inputType=FunctionInput_Scalar})
  n9 = Add(A=@n8, B=@n2)
  n10 = StaticSwitch(True=Add(A=@n9, B=@n4), False=@n9, Value=@n6)
  FunctionOutput(None=Saturate(None=StaticSwitch(True=@n10, False=@n8, Value=@n7))) {outputName=Masks}
  n11 = Saturate()
  n12 = Saturate()
  n13 = Add(A=@n11, B=@n12)
  n14 = Saturate()
  n15 = Add(A=@n13, B=@n14)
  n16 = StaticBoolParameter() {group=Diffuse, parameterName=Uses 3 Color Masks_1}
  n17 = StaticBoolParameter() {defaultValue=true, group=Diffuse, parameterName=Uses 2+ Color Masks_1}
  Saturate(None=StaticSwitch(True=StaticSwitch(True=@n15, False=@n13, Value=@n16), False=@n11, Value=@n17))
  n18 = LinearInterpolate(A=LinearInterpolate(A=Reroute(), Alpha=@n11) {constB=0}, B=LinearInterpolate(Alpha=@n12) {constB=0}, Alpha=@n11)
  n19 = LinearInterpolate(Alpha=Saturate(None=Add(A=@n12, B=@n11))) {constB=0}
  StaticSwitch(True=StaticSwitch(True=LinearInterpolate(A=@n18, B=@n19, Alpha=@n14), False=@n18, Value=@n16), False=Reroute(), Value=@n17)
 ]}
