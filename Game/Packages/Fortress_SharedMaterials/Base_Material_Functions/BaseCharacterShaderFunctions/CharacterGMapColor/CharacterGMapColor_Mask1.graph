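# Material function for character shaders that colorizes a grayscale GMap input into three tinted regions defined by the Color Mask 1 texture parameter.
# The GMap drives three Lerp_3Color gradients, each with its own A, B and C color parameters.
# The R gradient is the base, with G and B gradients blended over it by their mask channels.
# It outputs that color plus the summed RGB mask; three mask-by-GMap multiplies are unused.
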MaterialFunction /Game/Packages/Fortress_SharedMaterials/Base_Material_Functions/BaseCharacterShaderFunctions/CharacterGMapColor/CharacterGMapColor_Mask1
graph {nodes [
  n1 = VectorParameter() {group=Diffuse, parameterName=Color Mask 1-R: Color B, sortPriority=3, defaultValue=<rgba 0.5 0.5 0.5 1>}
  n2 = FunctionInput() {inputName=GMap, inputType=FunctionInput_Scalar}
  n3 = MaterialFunctionCall(
   A=VectorParameter() {group=Diffuse, parameterName=Color Mask 1-R: Color A, sortPriority=3, defaultValue=<rgba 0 0 0 1>}.RGB
   B=@n1.RGB
   C=VectorParameter() {group=Diffuse, parameterName=Color Mask 1-R: Color C, sortPriority=3, defaultValue=<rgba 1 1 1 1>}.RGB
   Alpha=@n2
  ) {materialFunction=&Lerp_3Color}
  n4 = VectorParameter() {group=Diffuse, parameterName=Color Mask 1-G: Color B, sortPriority=4, defaultValue=<rgba 0.5 0.5 0.5 1>}
  n5 = MaterialFunctionCall(
   A=VectorParameter() {group=Diffuse, parameterName=Color Mask 1-G: Color A, sortPriority=4, defaultValue=<rgba 0 0 0 1>}.RGB
   B=@n4.RGB
   C=VectorParameter() {group=Diffuse, parameterName=Color Mask 1-G: Color C, sortPriority=4, defaultValue=<rgba 1 1 1 1>}.RGB
   Alpha=@n2
  ) {materialFunction=&Lerp_3Color}
  n6 = TextureSampleParameter2D() {
   group=Diffuse
   parameterName=Color Mask 1
   samplerSource=SSM_Wrap_WorldGroupSettings
   samplerType=SAMPLERTYPE_Masks
   sortPriority=2
   texture=&T_BlackMask
  }
  n7 = VectorParameter() {group=Diffuse, parameterName=Color Mask 1-B: Color A, sortPriority=5, defaultValue=<rgba 0 0 0 1>}
  n8 = VectorParameter() {group=Diffuse, parameterName=Color Mask 1-B: Color B, sortPriority=5, defaultValue=<rgba 0.5 0.5 0.5 1>}
  n9 = VectorParameter() {group=Diffuse, parameterName=Color Mask 1-B: Color C, sortPriority=5, defaultValue=<rgba 1 1 1 1>}
  FunctionOutput(None=LinearInterpolate(
    A=LinearInterpolate(A=@n3.Result, B=@n5.Result, Alpha=@n6.G)
    B=MaterialFunctionCall(A=@n7.RGB, B=@n8.RGB, C=@n9.RGB, Alpha=@n2) {materialFunction=&Lerp_3Color}.Result
    Alpha=@n6.B
   )) {outputName=ColorizedGMap 1, sortPriority=1}
  FunctionOutput(None=Add(A=Add(A=@n6.R, B=@n6.G), B=@n6.B)) {outputName=Color Mask 1}
  Multiply(A=@n6.R, B=@n2)
  Multiply(A=@n6.G, B=@n2)
  Multiply(A=@n6.B, B=@n2)
 ]}
