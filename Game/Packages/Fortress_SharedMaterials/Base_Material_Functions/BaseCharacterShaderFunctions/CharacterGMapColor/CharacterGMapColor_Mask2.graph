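# Material function for character shaders that colorizes a grayscale GMap input into diffuse color for the second color mask set.
# The scalar GMap drives three Lerp_3Color gradients, each built from editable A, B and C color parameters for the R, G and B regions.
# The Color Mask 2 texture's G and B channels blend those gradients over the R gradient base.
# It outputs the colorized result plus the summed RGB mask, while three mask-times-GMap multiplies are left unconnected.
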MaterialFunction /Game/Packages/Fortress_SharedMaterials/Base_Material_Functions/BaseCharacterShaderFunctions/CharacterGMapColor/CharacterGMapColor_Mask2
graph {nodes [
  n1 = VectorParameter() {group=Diffuse, parameterName=Color Mask 2-R: Color B, sortPriority=7, defaultValue=<rgba 0.5 0.5 0.5 1>}
  n2 = FunctionInput() {inputName=GMap, inputType=FunctionInput_Scalar}
  n3 = MaterialFunctionCall(
   A=VectorParameter() {group=Diffuse, parameterName=Color Mask 2-R: Color A, sortPriority=7, defaultValue=<rgba 0 0 0 1>}.RGB
   B=@n1.RGB
   C=VectorParameter() {group=Diffuse, parameterName=Color Mask 2-R: Color C, sortPriority=7, defaultValue=<rgba 1 1 1 1>}.RGB
   Alpha=@n2
  ) {materialFunction=&Lerp_3Color}
  n4 = VectorParameter() {group=Diffuse, parameterName=Color Mask 2-G: Color B, sortPriority=8, defaultValue=<rgba 0.5 0.5 0.5 1>}
  n5 = MaterialFunctionCall(
   A=VectorParameter() {group=Diffuse, parameterName=Color Mask 2-G: Color A, sortPriority=8, defaultValue=<rgba 0 0 0 1>}.RGB
   B=@n4.RGB
   C=VectorParameter() {group=Diffuse, parameterName=Color Mask 2-G: Color C, sortPriority=8, defaultValue=<rgba 1 1 1 1>}.RGB
   Alpha=@n2
  ) {materialFunction=&Lerp_3Color}
  n6 = TextureSampleParameter2D() {
   group=Diffuse
   parameterName=Color Mask 2
   samplerSource=SSM_Wrap_WorldGroupSettings
   samplerType=SAMPLERTYPE_Masks
   sortPriority=6
   texture=&T_BlackMask
  }
  n7 = VectorParameter() {group=Diffuse, parameterName=Color Mask 2-B: Color A, sortPriority=9, defaultValue=<rgba 0 0 0 1>}
  n8 = VectorParameter() {group=Diffuse, parameterName=Color Mask 2-B: Color B, sortPriority=9, defaultValue=<rgba 0.5 0.5 0.5 1>}
  n9 = VectorParameter() {group=Diffuse, parameterName=Color Mask 2-B: Color C, sortPriority=9, defaultValue=<rgba 1 1 1 1>}
  FunctionOutput(None=LinearInterpolate(
    A=LinearInterpolate(A=@n3.Result, B=@n5.Result, Alpha=@n6.G)
    B=MaterialFunctionCall(A=@n7.RGB, B=@n8.RGB, C=@n9.RGB, Alpha=@n2) {materialFunction=&Lerp_3Color}.Result
    Alpha=@n6.B
   )) {outputName=ColorizedGMap 2, sortPriority=1}
  FunctionOutput(None=Add(A=Add(A=@n6.R, B=@n6.G), B=@n6.B)) {outputName=Color Mask 2}
  Multiply(A=@n6.G, B=@n2)
  Multiply(A=@n6.B, B=@n2)
  Multiply(A=@n6.R, B=@n2)
 ]}
