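# Material function for character shaders that colorizes a scalar gradient map (GMap) for the third color mask.
# The GMap drives three Lerp_3Color ramps built from Diffuse color parameters, one each for R, G and B.
# The R ramp is the base, with G then B ramps blended over it by the Color Mask 3 texture channels.
# It outputs that color plus the summed RGB mask; three mask-times-GMap multiplies are unused.
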MaterialFunction /Game/Packages/Fortress_SharedMaterials/Base_Material_Functions/BaseCharacterShaderFunctions/CharacterGMapColor/CharacterGMapColor_Mask3
graph {nodes [
  n1 = VectorParameter() {group=Diffuse, parameterName=Color Mask 3-R: Color A, sortPriority=11, defaultValue=<rgba 0 0 0 1>}
  n2 = VectorParameter() {group=Diffuse, parameterName=Color Mask 3-R: Color B, sortPriority=11, defaultValue=<rgba 0.5 0.5 0.5 1>}
  n3 = VectorParameter() {group=Diffuse, parameterName=Color Mask 3-R: Color C, sortPriority=11, defaultValue=<rgba 1 1 1 1>}
  n4 = FunctionInput() {inputName=GMap, inputType=FunctionInput_Scalar}
  n5 = MaterialFunctionCall(A=@n1.RGB, B=@n2.RGB, C=@n3.RGB, Alpha=@n4) {materialFunction=&Lerp_3Color}
  n6 = VectorParameter() {group=Diffuse, parameterName=Color Mask 3-G: Color A, sortPriority=12, defaultValue=<rgba 0 0 0 1>}
  n7 = VectorParameter() {group=Diffuse, parameterName=Color Mask 3-G: Color B, sortPriority=12, defaultValue=<rgba 0.5 0.5 0.5 1>}
  n8 = VectorParameter() {group=Diffuse, parameterName=Color Mask 3-G: Color C, sortPriority=12, defaultValue=<rgba 1 1 1 1>}
  n9 = MaterialFunctionCall(A=@n6.RGB, B=@n7.RGB, C=@n8.RGB, Alpha=@n4) {materialFunction=&Lerp_3Color}
  n10 = TextureSampleParameter2D() {
   group=Diffuse
   parameterName=Color Mask 3
   samplerSource=SSM_Wrap_WorldGroupSettings
   samplerType=SAMPLERTYPE_Masks
   sortPriority=10
   texture=&T_BlackMask
  }
  n11 = VectorParameter() {group=Diffuse, parameterName=Color Mask 3-B: Color A, sortPriority=13, defaultValue=<rgba 0 0 0 1>}
  n12 = VectorParameter() {group=Diffuse, parameterName=Color Mask 3-B: Color B, sortPriority=13, defaultValue=<rgba 0.5 0.5 0.5 1>}
  n13 = VectorParameter() {group=Diffuse, parameterName=Color Mask 3-B: Color C, sortPriority=13, defaultValue=<rgba 1 1 1 1>}
  FunctionOutput(None=LinearInterpolate(
    A=LinearInterpolate(A=@n5.Result, B=@n9.Result, Alpha=@n10.G)
    B=MaterialFunctionCall(A=@n11.RGB, B=@n12.RGB, C=@n13.RGB, Alpha=@n4) {materialFunction=&Lerp_3Color}.Result
    Alpha=@n10.B
   )) {outputName=ColorizedGMap 3, sortPriority=1}
  FunctionOutput(None=Add(A=Add(A=@n10.R, B=@n10.G), B=@n10.B)) {outputName=Color Mask 3}
  Multiply(A=@n10.R, B=@n4)
  Multiply(A=@n10.G, B=@n4)
  Multiply(A=@n10.B, B=@n4)
 ]}
