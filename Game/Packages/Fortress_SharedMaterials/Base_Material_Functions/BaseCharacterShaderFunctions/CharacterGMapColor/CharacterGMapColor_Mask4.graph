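# Material function for character shaders that layers color variety, scratch and dirt tints onto a colorized gradient map.
# The scalar GMap input drives a three-color variety ramp and two-color scratch and dirt ramps.
# Variety blends by an opacity parameter, not mask R; mask G and B then blend in scratch and dirt.
# It also outputs the saturated R+G+B mask sum; three channel-times-GMap multiplies are unconnected.
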MaterialFunction /Game/Packages/Fortress_SharedMaterials/Base_Material_Functions/BaseCharacterShaderFunctions/CharacterGMapColor/CharacterGMapColor_Mask4
graph {nodes [
  n1 = VectorParameter() {group=Diffuse, parameterName=Color Variety Mask: Color A, sortPriority=15, defaultValue=<rgba 0 0 0 1>}
  n2 = VectorParameter() {group=Diffuse, parameterName=Color Variety Mask: Color B, sortPriority=15, defaultValue=<rgba 0.5 0.5 0.5 1>}
  n3 = VectorParameter() {group=Diffuse, parameterName=Color Variety Mask: Color C, sortPriority=15, defaultValue=<rgba 1 1 1 1>}
  n4 = FunctionInput() {inputName=GMap, inputType=FunctionInput_Scalar}
  n5 = LinearInterpolate(
   A=FunctionInput() {inputName=ColorizedGMap, sortPriority=1}
   B=MaterialFunctionCall(A=@n1.RGB, B=@n2.RGB, C=@n3.RGB, Alpha=@n4) {materialFunction=&Lerp_3Color}.Result
   Alpha=ScalarParameter() {defaultValue=1, group=Diffuse, parameterName=Color Variety Mask: Opacity, sliderMax=1, sortPriority=15}
  )
  n6 = LinearInterpolate(
   A=VectorParameter() {group=Diffuse, parameterName=Scratch Color A, sortPriority=16, defaultValue=<rgba 0 0 0 1>}.RGB
   B=VectorParameter() {group=Diffuse, parameterName=Scratch Color B, sortPriority=16, defaultValue=<rgba 1 1 1 1>}.RGB
   Alpha=@n4
  )
  n7 = TextureSampleParameter2D() {
   desc="R = Color Variety\r\nG = Scratch\r\nB = Dirt"
   group=Diffuse
   parameterName=ColorVariety/Scratch/Dirt Mask
   samplerType=SAMPLERTYPE_Masks
   sortPriority=14
   texture=&T_BlackMask
  }
  n8 = VectorParameter() {group=Diffuse, parameterName=Dirt Color A, sortPriority=17, defaultValue=<rgba 0 0 0 1>}
  n9 = VectorParameter() {group=Diffuse, parameterName=Dirt Color B, sortPriority=17, defaultValue=<rgba 1 1 1 1>}
  FunctionOutput(None=LinearInterpolate(A=LinearInterpolate(A=@n5, B=@n6, Alpha=@n7.G), B=LinearInterpolate(A=@n8.RGB, B=@n9.RGB, Alpha=@n4), Alpha=@n7.B)) {outputName=ColorizedGMap with CV/S/D, sortPriority=1}
  n10 = Add(A=@n7.R, B=@n7.G)
  FunctionOutput(None=Saturate(None=Add(A=@n10, B=@n7.B))) {outputName=CV/S/D Mask}
  Multiply(A=@n7.R, B=@n4)
  Multiply(A=@n7.G, B=@n4)
  Multiply(A=@n7.B, B=@n4)
 ]}
