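# Character material function that colorizes base color in-engine from a scalar GMap input and a Diffuse input.
# Three Mask sub-functions feed a Blending function, whose result drives a BaseColor function to build the colorized GMap.
# A static switch optionally lerps in a Mask4 color variety, scratch and dirt layer.
# The "Use Engine Colorized GMap" switch selects this result or passes through the original color for both material attributes and Diffuse outputs.
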MaterialFunction /Game/Packages/Fortress_SharedMaterials/Base_Material_Functions/BaseCharacterShaderFunctions/BaseFN_CharacterGMapColor
graph {nodes [
  n1 = GetMaterialAttributes(None=FunctionInput() {inputName=Material, inputType=FunctionInput_MaterialAttributes}) {attributeGetTypes [#1]}
  n2 = StaticBoolParameter() {group=Diffuse, parameterName=Use Engine Colorized GMap}
  n3 = FunctionInput() {bUsePreviewValueAsDefault=true, inputName=GMap, inputType=FunctionInput_Scalar, sortPriority=1}
  n4 = MaterialFunctionCall!(GMap=@n3) {materialFunction=/Game/Packages/Fortress_SharedMaterials/Base_Material_Functions/BaseCharacterShaderFunctions/CharacterGMapColor/CharacterGMapColor_Mask1}
  n5 = MaterialFunctionCall!(GMap=@n3) {materialFunction=/Game/Packages/Fortress_SharedMaterials/Base_Material_Functions/BaseCharacterShaderFunctions/CharacterGMapColor/CharacterGMapColor_Mask2}
  n6 = MaterialFunctionCall!(GMap=@n3) {materialFunction=/Game/Packages/Fortress_SharedMaterials/Base_Material_Functions/BaseCharacterShaderFunctions/CharacterGMapColor/CharacterGMapColor_Mask3}
  n7 = MaterialFunctionCall!(
   Color Mask 1=@n4.Color Mask 1
   ColorizedGMap 1=@n4.Color Mask 1
   Color Mask 2=@n5.Color Mask 2
   ColorizedGMap 2=@n5.Color Mask 2
   Color Mask 3=@n6.Color Mask 3
   ColorizedGMap 3=@n6.Color Mask 3
  ) {materialFunction=/Game/Packages/Fortress_SharedMaterials/Base_Material_Functions/BaseCharacterShaderFunctions/CharacterGMapColor/CharacterGMapColor_Blending}
  n8 = FunctionInput() {inputName=Diffuse}
  n9 = MaterialFunctionCall!(Masks=@n7.Masks, GMap=@n3, Diffuse=@n8, ColorizedGMap=@n7.Masks) {materialFunction=/Game/Packages/Fortress_SharedMaterials/Base_Material_Functions/BaseCharacterShaderFunctions/CharacterGMapColor/CharacterGMapColor_BaseColor}
  n10 = MaterialFunctionCall!(GMap=@n3, ColorizedGMap=@n9.ColorizedGMap) {materialFunction=/Game/Packages/Fortress_SharedMaterials/Base_Material_Functions/BaseCharacterShaderFunctions/CharacterGMapColor/CharacterGMapColor_Mask4}
  n11 = StaticSwitchParameter(True=LinearInterpolate(A=@n9.ColorizedGMap, B=@n10.ColorizedGMap with CV/S/D, Alpha=@n10.ColorizedGMap with CV/S/D), False=@n9.ColorizedGMap) {group=Diffuse, parameterName=Uses ColorVariety/Scratch/Dirt Mask}
  FunctionOutput(None=SetMaterialAttributes!(MaterialAttributes=@n1.MaterialAttributes, BaseColor=StaticSwitch(True=@n11, False=@n1.BaseColor, Value=@n2)) {attributeSetTypes [#1]})
  FunctionOutput(None=StaticSwitch(True=@n11, False=@n8, Value=@n2)) {outputName=DIffuse, sortPriority=1}
 ]}
