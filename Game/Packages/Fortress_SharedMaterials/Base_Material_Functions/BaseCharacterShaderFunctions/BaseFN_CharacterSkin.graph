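# Material function for character skin that tints incoming material attributes and switches their shading model to Subsurface Profile.
# A Fresnel-weighted boost color brightens base color, blended by HumanSkin, while a scalar parameter drives subsurface strength through Opacity.
# The skin result blends over the input by the Skin Mask, and a static switch can bypass it entirely.
# Tiled pore-normal blending, including an AO-masked high-quality path, is present but not connected to the output.
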MaterialFunction /Game/Packages/Fortress_SharedMaterials/Base_Material_Functions/BaseCharacterShaderFunctions/BaseFN_CharacterSkin
graph {nodes [
  Constant() {r=1}
  n1 = GetMaterialAttributes(None=FunctionInput() {inputName=Material, inputType=FunctionInput_MaterialAttributes}) {attributeGetTypes [
    #1
    #2
    #3
    #4
   ]}
  n2 = VectorParameter() {group=SkinColor, parameterName=Skin Boost Color And Exponent, defaultValue=<rgba 0.87 0.6688 0.6948 2>}
  n3 = OneMinus(None=Fresnel(ExponentIn=@n2.A) {exponent=2})
  n4 = Multiply(A=@n1.BaseColor, B=@n2.RGB)
  n5 = Multiply(A=ScalarParameter() {defaultValue=0.75, parameterName=Global Skin Boost EX}, B=Add(A=Multiply(A=@n3, B=@n4), B=@n4))
  n6 = ScalarParameter() {defaultValue=1, group=SkinColor, parameterName=HumanSkin}
  n7 = SetMaterialAttributes!(
   MaterialAttributes=@n1.MaterialAttributes
   BaseColor=Saturate(None=LinearInterpolate(A=@n1.BaseColor, B=@n5, Alpha=@n6))
   ShadingModel=ShadingModel() {shadingModel=MSM_SubsurfaceProfile}
   Normal=@n1.MaterialAttributes
   Opacity=ScalarParameter() {defaultValue=1, group=SkinColor, parameterName=SSS Brightness}
   Roughness=@n1.MaterialAttributes
  ) {attributeSetTypes [
    #1
    #5
    #2
    #3
    #4
   ]}
  n8 = FunctionInput() {inputName=Skin Mask, inputType=FunctionInput_Scalar, sortPriority=3}
  FunctionOutput(None=StaticSwitchParameter(True=@n1.MaterialAttributes, False=BlendMaterialAttributes(A=@n1.MaterialAttributes, B=@n7, Alpha=@n8)) {group=SkinColor, parameterName=Disable Skin Material})
  n9 = TextureCoordinate()
  n10 = ScalarParameter() {defaultValue=50, parameterName=PoreTiling}
  n11 = MaterialFunctionCall(
   Normal=TextureSample(UVs=Multiply(A=@n9, B=@n10)) {samplerType=SAMPLERTYPE_Normal, texture=/Game/Global/Textures/Default/Tileable_Patterns/T_SkinBump_N}.RGB
   Flatness=ScalarParameter() {group=SkinColor, parameterName=Skin Pore Flatness}
  ) {materialFunction=&FlattenNormal}
  n12 = MaterialFunctionCall(BaseNormal=@n1.Normal, AdditionalNormal=@n11.Result) {materialFunction=&BlendAngleCorrectedNormals}
  n13 = FunctionInput() {inputName=AO, inputType=FunctionInput_Scalar, sortPriority=2}
  Normalize(VectorInput=QualitySwitch(Default=@n1.Normal, High=LinearInterpolate(A=@n1.Normal, B=@n12.Result, Alpha=@n13)))
  n14 = Constant3Vector() {constant=<rgba 1 1 0 0>}
  Normalize(VectorInput=Add(A=Multiply(A=@n14, B=@n11.Result), B=@n1.Normal))
  RayTracingQualitySwitch()
  ShadingPathSwitch()
 ]}
