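# Character material function that layers an animated, glowing fringe over incoming material attributes within Fringe_Mask.
# Two panning pre-skinned-space noises, flipped in direction about a center offset, erode the red-channel mask, with optional fresnel.
# The result drives emissive with a time-shifting two-color glow, plus a forced edge from the green channel.
# It also optionally darkens base color past the noise boundary, clips opacity mask, and wiggles vertices through noise-driven WPO.
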MaterialFunction /Game/Packages/Fortress_SharedMaterials/Base_Material_Functions/BaseCharacterShaderFunctions/MF_Fringe_Effects
graph {nodes [
  n1 = NamedRerouteUsage()
  Add(A=@n1, B=ScalarParameter() {group=Fringe, parameterName=NudgeOpacityMaskPoint, sortPriority=35})
  n2 = ComponentMask(None=NamedRerouteDeclaration(None=TextureSampleParameter2D() {group=Fringe, parameterName=Fringe_Mask, texture=&T_BlacksRGB}.RGB) {name=Fringe_Mask_Tex, nodeColor=<rgba 0.5469 0 1 1>}) {r=true}
  n3 = Power(Base=@n2, Exp=ScalarParameter() {defaultValue=1, group=Fringe, parameterName=FringeMask_Pow})
  n4 = ScalarParameter() {defaultValue=1, group=Fringe, parameterName=Fringe_Mult}
  n5 = NamedRerouteUsage()
  n6 = Sign(None=Subtract(A=VectorParameter() {group=Fringe, parameterName=CenterOffset}.B, B=ComponentMask(None=@n5) {b=true}))
  n7 = Time()
  n8 = Multiply(A=VectorParameter() {group=Fringe, parameterName=PanningSpeed_LargeNoise, defaultValue=<rgba -8 0 0 0>}.RGB, B=@n7)
  n9 = Noise!(World Position=Add(A=@n5, B=Multiply(A=@n6, B=@n8)), FilterWidth) {bTurbulence=false, levelScale=3, levels=2, outputMax=0.5, outputMin=0, scale=0.05}
  n10 = Saturate(None=Subtract(A=ComponentMask(None=@n9) {r=true}, B=ScalarParameter() {defaultValue=0.3, group=Fringe, parameterName=LargeNoise DarkShift}) {constB=0.3})
  n11 = Saturate(None=Subtract(A=Multiply(A=@n3, B=@n4), B=@n10))
  n12 = Multiply(A=VectorParameter() {group=Fringe, parameterName=PanningSpeed, defaultValue=<rgba -8 0 0 0>}.RGB, B=@n7)
  n13 = Noise(World Position=Add(A=@n5, B=Multiply(A=@n6, B=@n12))) {bTurbulence=false, levels=3, outputMin=0, scale=0.13}
  n14 = Saturate(None=Multiply(A=ComponentMask(None=@n13) {r=true}, B=ScalarParameter() {defaultValue=1, group=Fringe, parameterName=SoftenSmallNoise}))
  n15 = Saturate(None=Add(A=Multiply(A=@n11, B=@n14), B=ScalarParameter() {group=Fringe, parameterName=AdjustSmallNoiseContribution}))
  n16 = NamedRerouteDeclaration(None=Saturate(None=Add(A=@n11, B=@n15))) {name=Fringe Mask, nodeColor=<rgba 1 0.2109 0 1>}
  n17 = Add(A=@n16, B=ScalarParameter() {group=Fringe, parameterName=NoiseBoundryShift})
  n18 = Saturate(None=Floor(None=@n17))
  n19 = LinearInterpolate(
   A=NamedRerouteUsage()
   B=Constant3Vector()
   Alpha=Multiply(
    A=NamedRerouteDeclaration(None=@n18) {name=ColorLerp, nodeColor=<rgba 0 0.4609 1 1>}
    B=ScalarParameter() {group=Fringe, parameterName=UseMaskAsBoundry}
   )
  )
  n20 = Fresnel()
  n21 = Saturate(None=Multiply(A=Multiply(A=@n11, B=@n20), B=ScalarParameter() {group=Fringe, parameterName=AdjustFresnelContribution}))
  n22 = Multiply(A=@n19, B=Saturate(None=Add(A=@n21, B=@n16)))
  n23 = NamedRerouteUsage()
  n24 = ScalarParameter() {group=Fringe, parameterName=UsePrimaryFringeAsHotEdgeNoiseMAsk, sortPriority=40}
  n25 = Multiply(
   A=VectorParameter() {group=Fringe, parameterName=ForcedFringeColor}.RGB
   B=Multiply(A=ScalarParameter() {defaultValue=1, group=Fringe, parameterName=Forced_Edge_Brightness}, B=LinearInterpolate(B=@n23, Alpha=@n24) {constA=1})
  )
  n26 = NamedRerouteUsage()
  Add(A=@n22, B=Multiply(A=@n25, B=@n26))
  n27 = GetMaterialAttributes(None=FunctionInput() {inputName=MA, inputType=FunctionInput_MaterialAttributes}) {attributeGetTypes [
    #1
    #2
    #3
    #4
   ]}
  n28 = NamedRerouteUsage()
  n29 = ScalarParameter() {group=Fringe, parameterName=UseMaskAsBoundry}
  n30 = Ceil(None=@n1)
  n31 = SetMaterialAttributes!(
   MaterialAttributes=@n27.MaterialAttributes
   WorldPositionOffset=Add(A=@n27.WorldPositionOffset, B=NamedRerouteUsage())
   EmissiveColor=Add(A=@n27.EmissiveColor, B=NamedRerouteUsage())
   BaseColor=LinearInterpolate(
    A=@n27.BaseColor
    B=VectorParameter() {group=Fringe, parameterName=ColorBeyondBoundry, defaultValue=<rgba 0.01 0.01 0.01 0>}.RGB
    Alpha=Multiply(A=@n28, B=@n29)
   )
   OpacityMask=StaticSwitchParameter(True=Min(A=@n27.OpacityMask, B=@n30), False=@n27.OpacityMask) {group=Fringe, parameterName=UseFringeMaskForOpacityMask, sortPriority=34}
  ) {attributeSetTypes [
    #2
    #3
    #1
    #4
   ]}
  FunctionOutput(None=@n31)
  Multiply(A=@n6) {constB=-1}
  n32 = FunctionInput(Preview=VertexInterpolator(VS=PreSkinnedPosition()).PS) {bUsePreviewValueAsDefault=true, inputName=PreskinnedLocalPoistion, sortPriority=1}
  n33 = NamedRerouteUsage()
  n34 = Multiply(A=Time(), B=VectorParameter() {group=Fringe, parameterName=PanWPONoise}.RGB)
  n35 = Multiply(
   A=Normalize(VectorInput=NamedRerouteDeclaration(None=@n32) {name=PreskinnedLocalPosition, nodeColor=<rgba 0 1 0.4609 1>})
   B=Noise(World Position=Add(A=@n33, B=@n34)) {outputMax=0.1, outputMin=-0.5, scale=0.01}
  )
  NamedRerouteDeclaration(None=Multiply(A=Multiply(A=@n35, B=@n2), B=ScalarParameter() {group=Fringe, parameterName=Scale WPO Effect})) {name=WPO_Wiggle, nodeColor=<rgba 0.1719 0 1 1>}
  n36 = Multiply(
   A=Max(A=@n10, B=ScalarParameter() {defaultValue=0.2, group=Fringe, parameterName=OtherEmissiveMin})
   B=TextureSampleParameter2D() {parameterName=Emissive, texture=&T_BlacksRGB}.RGB
  )
  n37 = ScalarParameter() {group=Fringe, parameterName=EmissiveBoost on Other}
  NamedRerouteDeclaration(None=Add(A=LinearInterpolate(A=@n22, B=@n25, Alpha=@n26), B=Multiply(A=@n36, B=@n37))) {name=Fringe_Emissive, nodeColor=<rgba 1 0.7969 0 1>}
  n38 = VectorParameter() {group=Fringe, parameterName=Fringe Glow}
  n39 = VectorParameter() {group=Fringe, parameterName=Fringe Glow Second Color}
  n40 = Multiply(A=Time(), B=ScalarParameter() {defaultValue=0.001, group=Fringe, parameterName=SecondColorShiftRate})
  n41 = Multiply(A=Abs(None=Sine(None=@n40)), B=ScalarParameter() {group=Fringe, parameterName=SecondColorShiftOn})
  NamedRerouteDeclaration(None=Multiply(
    A=LinearInterpolate(A=@n38.RGB, B=@n39.RGB, Alpha=@n41)
    B=ScalarParameter() {defaultValue=3, group=Fringe, parameterName=Emissive Color Intensity}
   ) {constB=3}) {name=Color, nodeColor=<rgba 1 0 0.375 1>}
  n42 = ComponentMask(None=NamedRerouteUsage()) {g=true}
  n43 = ScalarParameter() {group=Fringe, parameterName=ForcedFringeEdge Mult}
  n44 = Power(Base=Multiply(A=@n42, B=@n43), Exp=ScalarParameter() {defaultValue=1, group=Fringe, parameterName=ForcedFringeEdge Pow})
  NamedRerouteDeclaration(None=Multiply(A=Saturate(None=@n44), B=ScalarParameter() {group=Fringe, parameterName=ForcedHardGlowEdgeContributtion})) {name=ForcedFringeEdgeMask, nodeColor=<rgba 0 0.9766 1 1>}
  NamedRerouteDeclaration(None=@n19) {name=ToOpacityMask, nodeColor=<rgba 0.4062 1 0 1>}
  NamedRerouteDeclaration(None=ComponentMask(None=Desaturation(None=@n22)) {r=true}) {name=NoiseMaskFromPRimaryFringe, nodeColor=<rgba 0 0.2734 1 1>}
  NamedRerouteUsage()
  Round(None=@n17)
  Subtract(A=@n42, B=@n43)
 ]}
