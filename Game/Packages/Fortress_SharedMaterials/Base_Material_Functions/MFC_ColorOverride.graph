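# Material function that recolors incoming material attributes by overriding Base Color, Emissive and Metallic, behind the "Use Color Override" static switch.
# One color parameter supplies flat base color and tints desaturated emissive, while metallic blends toward a scalar.
# Each channel has an override strength and a use-mask blend.
# The mask is a selected channel of either the Skin FX Mask input or a custom texture.
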MaterialFunction /Game/Packages/Fortress_SharedMaterials/Base_Material_Functions/MFC_ColorOverride
graph {nodes [
  n1 = FunctionInput() {inputType=FunctionInput_MaterialAttributes}
  n2 = GetMaterialAttributes(None=@n1) {attributeGetTypes [#1, #2, #3]}
  n3 = ScalarParameter() {defaultValue=1, group=Color Override, parameterName=CO_OverrideBaseColor}
  n4 = NamedRerouteUsage()
  n5 = LinearInterpolate(
   A=NamedRerouteDeclaration(None=@n2.BaseColor) {name=Base Color In, nodeColor=<rgba 0 0.1562 1 1>}
   B=NamedRerouteUsage()
   Alpha=LinearInterpolate(
    A=@n3
    B=Multiply(A=@n3, B=@n4)
    Alpha=ScalarParameter() {defaultValue=1, group=Color Override, parameterName=CO_UseMaskToOverrideBaseColor, sortPriority=3}
   )
  )
  n6 = ScalarParameter() {defaultValue=1, group=Color Override, parameterName=CO_OverrideEmissive}
  n7 = NamedRerouteUsage()
  n8 = LinearInterpolate(
   A=NamedRerouteDeclaration(None=@n2.EmissiveColor) {name=Emissive Color In, nodeColor=<rgba 0.5 0 1 1>}
   B=NamedRerouteUsage()
   Alpha=LinearInterpolate(A=@n6, B=Multiply(A=@n6, B=@n7), Alpha=ScalarParameter() {group=Color Override, parameterName=CO_UseMaskToOverrideEmissive, sortPriority=5})
  )
  n9 = ScalarParameter() {defaultValue=1, group=Color Override, parameterName=CO_OverrideMetallic}
  n10 = NamedRerouteUsage()
  n11 = LinearInterpolate(
   A=@n2.Metallic
   B=ScalarParameter() {defaultValue=0.5, group=Color Override, parameterName=CO_Metallic}
   Alpha=LinearInterpolate(
    A=@n9
    B=Multiply(A=@n9, B=@n10)
    Alpha=ScalarParameter() {defaultValue=1, group=Color Override, parameterName=CO_UseMaskToOverrideMetallic, sortPriority=12}
   )
  )
  FunctionOutput(None=StaticSwitchParameter(
    True=SetMaterialAttributes!(MaterialAttributes=@n2.MaterialAttributes, BaseColor=@n5, EmissiveColor=@n8, Metallic=@n11) {attributeSetTypes [#1, #2, #3]}
    False=@n1
   ) {group=Skin FX, parameterName=Use Color Override})
  n12 = VectorParameter() {group=Color Override, parameterName=CO_ColorOverride, defaultValue=<rgba 0.5776 0.0931 0.0168 1>}
  Multiply(A=@n12.RGB, B=Desaturation(None=NamedRerouteUsage(), Fraction=Constant() {r=1}))
  NamedRerouteDeclaration(None=@n12.RGB) {name=Base Color Out, nodeColor=<rgba 0.6875 1 0 1>}
  n13 = NamedRerouteUsage()
  n14 = Constant() {r=1}
  NamedRerouteDeclaration(None=Multiply(A=Desaturation(None=@n13, Fraction=@n14), B=@n12.RGB)) {name=Emissive Color Out, nodeColor=<rgba 1 0 0.3516 1>}
  n15 = TextureSampleParameter2D() {group=Color Override, parameterName=CO_CustomMask, samplerType=SAMPLERTYPE_Masks, texture=&T_BlackMask}
  n16 = FunctionInput() {bUsePreviewValueAsDefault=true, inputName=Skin FX Mask, sortPriority=100, previewValue=<1 0 0 1>}
  NamedRerouteDeclaration(None=StaticSwitchParameter(
    True=ChannelMaskParameter(None=@n15.RGBA) {group=Color Override, parameterName=CO_MaskChannel}.R
    False=ChannelMaskParameter(None=@n16) {group=Color Override, parameterName=CO_MaskChannel}.R
   ) {group=Color Override, parameterName=CO Use Custom Mask}) {name=Mask, nodeColor=<rgba 0.3125 1 0 1>}
 ]}
