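# Material function for a Skin FX "Defect Metal" effect that turns masked areas into one of three metals, behind a static switch.
# A Metal Index parameter blends from original attributes through three tints over desaturated base color, with per-metal roughness ranges and raised metallic.
# The third metal hue-shifts using a pre-skinned normal dot product.
# A channel-selected Skin FX or custom texture mask blends results into base color, roughness and metallic.
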MaterialFunction /Game/Packages/Fortress_SharedMaterials/Base_Material_Functions/MFC_DefectMetal
graph {nodes [
  n1 = FunctionInput() {inputType=FunctionInput_MaterialAttributes}
  n2 = GetMaterialAttributes(None=@n1) {attributeGetTypes [#1, #2, #3]}
  n3 = NamedRerouteUsage()
  n4 = LinearInterpolate(A=NamedRerouteDeclaration(None=@n2.BaseColor) {name=Base Color In, nodeColor=<rgba 0 0.1562 1 1>}, B=NamedRerouteUsage(), Alpha=@n3)
  n5 = LinearInterpolate(A=NamedRerouteDeclaration(None=@n2.Roughness) {name=Roughness In, nodeColor=<rgba 0 0.0156 1 1>}, B=NamedRerouteUsage(), Alpha=@n3)
  n6 = LinearInterpolate(A=NamedRerouteDeclaration(None=@n2.Metallic) {name=Metallic In, nodeColor=<rgba 1 0 0.1406 1>}, B=NamedRerouteUsage(), Alpha=@n3)
  FunctionOutput(None=StaticSwitchParameter(
    True=SetMaterialAttributes!(MaterialAttributes=@n2.MaterialAttributes, BaseColor=@n4, Roughness=@n5, Metallic=@n6) {attributeSetTypes [#1, #2, #3]}
    False=@n1
   ) {group=Skin FX, parameterName=Use Defect Metal})
  n7 = NamedRerouteUsage()
  n8 = NamedRerouteUsage()
  n9 = VertexInterpolator(VS=PreSkinnedNormal())
  n10 = Normalize(VectorInput=VectorParameter() {group=Defect Metal, parameterName=DcM HueShift Normal Direction, defaultValue=<rgba 0.2923 1 0 1>}.RGB)
  n11 = Multiply(
   A=Abs(None=DotProduct(A=@n9.PS, B=@n10))
   B=ScalarParameter() {defaultValue=0.5, group=Defect Metal, parameterName=DcM Normal HueShift Multiplier}
  )
  n12 = Multiply(
   A=Add(A=ScalarParameter() {defaultValue=0.5, group=Defect Metal, parameterName=DcM Heat HueShift Bias}, B=Add(A=@n8, B=@n11))
   B=ScalarParameter() {defaultValue=-0.9, group=Defect Metal, parameterName=DcM Heat HueShift Multiplier}
  )
  n13 = Multiply(
   A=NamedRerouteUsage()
   B=VectorParameter() {group=Defect Metal, parameterName=Metal 3 Base Color, defaultValue=<rgba 0.9851 0.4145 0.9948 0>}.RGB
  )
  n14 = ScalarParameter() {group=Defect Metal, parameterName=DcM Metal Index, sliderMax=3}
  n15 = MaterialFunctionCall(
   1=@n7
   2=VectorParameter() {group=Defect Metal, parameterName=Metal 1 Base Color, defaultValue=<rgba 0.9271 0.626 0.3042 1>}.RGB
   3=VectorParameter() {group=Defect Metal, parameterName=Metal 2 Base Color, defaultValue=<rgba 2 0.963 0.78 1>}.RGB
   4=Multiply(
    A=MaterialFunctionCall(Hue Shift Percentage=@n12, Texture=@n13) {materialFunction=&HueShift}.Result
    B=ScalarParameter() {defaultValue=2.1469, group=Defect Metal, parameterName=Metal 3 Color Lift}
   )
   A=Divide(A=@n14) {constB=3}
  ) {materialFunction=&Lerp_Multiple_Float3}
  n16 = Desaturation(None=NamedRerouteUsage(), Fraction=Constant() {r=1})
  n17 = Multiply(
   A=NamedRerouteDeclaration(None=ComponentMask(None=@n16) {r=true}) {name=Desaturated Base Color, nodeColor=<rgba 1 0.4922 0 1>}
   B=ScalarParameter() {defaultValue=1, group=Defect Metal, parameterName=DcM Base Color Desaturate Lift}
  )
  NamedRerouteDeclaration(None=LinearInterpolate(A=@n7, B=Multiply(A=@n15.Lerp 4 Inputs, B=@n17), Alpha=Saturate(None=@n14))) {name=Base Color Out, nodeColor=<rgba 0.6875 1 0 1>}
  n18 = TextureSampleParameter2D() {parameterName=DcM_CustomMask, samplerType=SAMPLERTYPE_Masks, texture=&T_BlackMask}
  n19 = FunctionInput() {bUsePreviewValueAsDefault=true, inputName=Skin FX Mask, sortPriority=100, previewValue=<1 0 0 1>}
  NamedRerouteDeclaration(None=StaticSwitchParameter(
    True=ChannelMaskParameter(None=@n18.RGBA) {group=Defect Metal, parameterName=DcM MaskChannel}.R
    False=ChannelMaskParameter(None=@n19) {group=Defect Metal, parameterName=DcM MaskChannel}.R
   ) {group=Defect Metal, parameterName=DcM Use Custom Mask}) {name=Mask, nodeColor=<rgba 0.3125 1 0 1>}
  n20 = ScalarParameter() {defaultValue=0.3, group=Defect Metal, parameterName=DcM Roughness Min 1, sliderMax=1}
  n21 = ScalarParameter() {defaultValue=0.6, group=Defect Metal, parameterName=DcM Roughness Max 1, sliderMax=1}
  n22 = NamedRerouteUsage()
  n23 = ScalarParameter() {defaultValue=0.25, group=Defect Metal, parameterName=DcM Roughness Min 2, sliderMax=1}
  n24 = ScalarParameter() {defaultValue=0.4, group=Defect Metal, parameterName=DcM Roughness Max 2, sliderMax=1}
  n25 = NamedRerouteUsage()
  n26 = ScalarParameter() {defaultValue=0.2, group=Defect Metal, parameterName=DcM Roughness Min 3, sliderMax=1}
  n27 = ScalarParameter() {defaultValue=0.8, group=Defect Metal, parameterName=DcM Roughness Max 3, sliderMax=1}
  n28 = NamedRerouteUsage()
  n29 = Divide(A=ScalarParameter() {group=Defect Metal, parameterName=DcM Metal Index, sliderMax=3}) {constB=3}
  NamedRerouteDeclaration(None=MaterialFunctionCall!(
    1=NamedRerouteUsage()
    2=LinearInterpolate(A=@n20, B=@n21, Alpha=@n22)
    3=LinearInterpolate(A=@n23, B=@n24, Alpha=@n25)
    4=LinearInterpolate(A=@n26, B=@n27, Alpha=@n28)
    A=NamedRerouteDeclaration(None=@n29) {name=Metal Index, nodeColor=<rgba 0.1719 1 0 1>}
   ) {materialFunction=/Engine/Functions/Engine_MaterialFunctions02/Lerp_Multiple_Float}.Lerp 4 Inputs) {name=Roughness Out, nodeColor=<rgba 1 0 0.6797 1>}
  n30 = ScalarParameter() {group=Defect Metal, parameterName=DcM Metal Index, sliderMax=3}
  NamedRerouteDeclaration(None=LinearInterpolate(A=NamedRerouteUsage(), Alpha=Saturate(None=@n30))) {name=Metallic Out, nodeColor=<rgba 1 0 0.1406 1>}
 ]}
