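# Material function that layers a configurable eye shader with iris, limbus and pupil onto incoming Material Attributes, or passes them through when disabled.
# Iris UVs get IOR-based depth refraction and pupil scaling, then drive iris color, emissive, masks and normal, with the normal feeding a Tangent output.
# Emissive effects add a fake highlight, caustics with a cine variant, and skylight reflection.
# A static switch selects the Eye shading model or per-channel blending.
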MaterialFunction /Game/Packages/Fortress_SharedMaterials/Base_Material_Functions/MFC_EyeShader
graph {nodes [
  n1 = NamedRerouteUsage()
  Distance(A=@n1, B=Constant())
  n2 = MaterialFunctionCall(Color=NamedRerouteUsage()) {materialFunction=&SplitComponents}
  n3 = TextureSample() {samplerType=SAMPLERTYPE_Normal, texture=/Game/Global/Textures/Default/Eyes/T_Eye_Wet_Normal}
  n4 = Constant() {r=0.95}
  n5 = FunctionInput(Preview=SetMaterialAttributes!(
    MaterialAttributes
    BaseColor=TextureSample() {texture=/Game/Global/Textures/Default/Eyes/EyeScleraBaseColor_Example_2}.RGB
    Normal=MaterialFunctionCall(Normal=@n3.RGB, Flatness=@n4) {materialFunction=&FlattenNormal}.Result
    ShadingModel=ShadingModel() {shadingModel=MSM_Eye}
   ) {attributeSetTypes [#1, #2, #3]}) {inputName=MA, inputType=FunctionInput_MaterialAttributes}
  n6 = GetMaterialAttributes(None=@n5) {attributeGetTypes [
    #1
    #4
    #5
    #6
    #2
    #7
    #3
   ]}
  n7 = SetMaterialAttributes!(
   MaterialAttributes=@n6.MaterialAttributes
   BaseColor=NamedRerouteUsage()
   Metallic=@n2.G
   Specular=@n2.G
   Roughness=@n2.G
   EmissiveColor=NamedRerouteUsage()
   ShadingModel=ShadingModel() {shadingModel=MSM_Eye}
  ) {attributeSetTypes [
    #1
    #4
    #5
    #6
    #7
    #3
   ]}
  n8 = BlendMaterialAttributes(A=@n5, B=@n7, Alpha=NamedRerouteUsage())
  n9 = NamedRerouteUsage()
  n10 = MaterialFunctionCall(Color=LinearInterpolate(
    A=MaterialFunctionCall(X=@n6.Specular, Y=@n6.Specular, Z=@n6.Specular) {materialFunction=&MakeFloat3}.Result
    B=NamedRerouteUsage()
    Alpha=@n9
   )) {materialFunction=&SplitComponents}
  n11 = SetMaterialAttributes!(
   MaterialAttributes=@n6.MaterialAttributes
   BaseColor=LinearInterpolate(A=@n6.BaseColor, B=NamedRerouteUsage(), Alpha=@n9)
   Metallic=@n10.G
   Specular=@n10.G
   Roughness=@n10.G
   EmissiveColor=LinearInterpolate(A=@n6.EmissiveColor, B=NamedRerouteUsage(), Alpha=@n9)
  ) {attributeSetTypes [
    #1
    #4
    #5
    #6
    #7
   ]}
  FunctionOutput(None=StaticSwitchParameter(True=StaticSwitchParameter(True=@n8, False=@n11) {parameterName=Use Eye Shading Model}, False=@n5) {group=Skin FX, parameterName=Use Eye Shader})
  LinearInterpolate(B=Constant() {r=1}, Alpha=SmoothStep(Value=NamedRerouteUsage()))
  NamedRerouteDeclaration(None=CameraVectorWS()) {name=Camera Vector, nodeColor=<rgba 1 0 1 1>}
  n12 = Subtract(A=@n1) {constB=0.5}
  n13 = Constant()
  n14 = NamedRerouteUsage()
  n15 = NamedRerouteUsage()
  n16 = Divide(A=Subtract(A=Distance(A=@n12, B=@n13) {desc="length(UV)"}, B=Subtract(A=@n14, B=@n15)), B=@n15)
  n17 = SmoothStep(Value=Saturate(None=Subtract(B=@n16))) {desc=m}
  NamedRerouteDeclaration(None=ComponentMask(None=@n17) {r=true}) {name=Iris Mask, nodeColor=<rgba 0 1 0.9531 1>}
  n18 = NamedRerouteUsage()
  n19 = ComponentMask(None=@n18) {r=true}
  n20 = ComponentMask(None=VertexInterpolator(VS=PreSkinnedPosition()).PS) {r=true}
  n21 = LinearInterpolate(A=@n18, B=AppendVector(A=OneMinus(None=@n19), B=ComponentMask(None=@n18) {g=true}), Alpha=Sign(None=Saturate(None=@n20)))
  n22 = Transform(None=NamedRerouteUsage()) {transformSourceType=TRANSFORMSOURCE_World, transformType=TRANSFORM_Local}
  n23 = Multiply(A=Normalize(VectorInput=ComponentMask(None=@n22) {b=true, r=true}), B=Constant2Vector() {g=-1, r=1})
  n24 = NamedRerouteDeclaration(None=DotProduct(A=@n22, B=Constant3Vector() {constant=<rgba 0 1 0 0>})) {name=LocalDir, nodeColor=<rgba 0.4766 1 0 1>}
  n25 = Multiply(A=OneMinus(None=Saturate(None=@n24))) {constB=0.1}
  n26 = ScalarParameter() {defaultValue=0.1, parameterName=Spec Position X}
  n27 = ScalarParameter() {defaultValue=0.02, parameterName=Spec Position Y}
  n28 = Subtract(B=StaticSwitchParameter(True=Multiply(A=@n23, B=@n25) {constB=0.1}, False=AppendVector(A=@n26, B=@n27)) {group=Eye Shader, parameterName=Use Specular Follow Sun}) {constA=0.5}
  n29 = ScalarParameter() {defaultValue=0.05, parameterName=Highlight Size}
  n30 = ScalarParameter() {parameterName=Highlight Hardness, sliderMax=1}
  n31 = StaticSwitchParameter(
   True=StaticSwitchParameter(True=NamedRerouteUsage(), False=SphereMask(A=@n21, B=@n28, Radius=@n29, Hardness=@n30)) {parameterName=Use Cine Specular Highlight}
   False=Constant()
  ) {defaultValue=true, parameterName=Use Specular Highlight}
  n32 = StaticSwitchParameter(
   True=StaticSwitchParameter(True=NamedRerouteUsage(), False=NamedRerouteUsage()) {group=Eye Shader Cine, parameterName=Cine Caustic}
   False=Constant()
  ) {defaultValue=true, parameterName=Use Caustic}
  n33 = Add(A=ReflectionVectorWS(), B=NamedRerouteUsage())
  n34 = Fresnel()
  n35 = ScalarParameter() {defaultValue=1, parameterName=Extra Reflection Use Fresnel}
  n36 = Multiply(
   A=SkyLightEnvMapSample(Direction=Normalize(VectorInput=@n33))
   B=Multiply(A=LinearInterpolate(B=@n34, Alpha=@n35) {constA=1}, B=ScalarParameter() {defaultValue=1, parameterName=Extra Reflection Multiplier})
  )
  n37 = Constant()
  NamedRerouteDeclaration(None=Add(A=Add(A=@n31, B=@n32), B=StaticSwitchParameter(True=@n36, False=@n37) {defaultValue=true, parameterName=Use Extra Reflection})) {name=Emissive Effects, nodeColor=<rgba 1 0 0.1641 1>}
  NamedRerouteDeclaration(None=@n17) {name=EyeIrisUVMask, nodeColor=<rgba 0 1 0.6953 1>}
  n38 = Distance(A=NamedRerouteUsage(), B=Constant2Vector() {g=0.5, r=0.5})
  n39 = Multiply(A=Power(Base=Multiply(A=@n38) {constB=2.75})) {constB=-1}
  n40 = Power(Base=Max(A=Add(A=@n39)) {constB=0}) {constExponent=0.7}
  n41 = Multiply(A=Power(Base=Multiply(A=@n38) {constB=8})) {constB=-1}
  n42 = Power(Base=Max(A=Add(A=@n41)) {constB=0}) {constExponent=0.9}
  n43 = NamedRerouteDeclaration(None=Add(A=Multiply(A=@n40) {constB=1.8}, B=Multiply(A=@n42) {constB=0.1})) {name=MidPlaneDisplacement, nodeColor=<rgba 0.3828 0 1 1>}
  n44 = NamedRerouteDeclaration(None=ScalarParameter() {defaultValue=1.37, group=Eye Shader, parameterName=SE_DepthPlaneOffset}) {name=DepthPlaneOffset, nodeColor=<rgba 1 0.0938 0 1>}
  n45 = Multiply(
   A=Max(A=Subtract(A=@n43, B=@n44)) {constB=0}
   B=NamedRerouteDeclaration(None=ScalarParameter() {defaultValue=1.2, group=Eye Shader, parameterName=ES_DepthScale}) {name=DepthScale, nodeColor=<rgba 0.8516 0 1 1>}
  )
  n46 = CameraVectorWS()
  n47 = NamedRerouteUsage()
  n48 = LinearInterpolate(Alpha=Power(Base=DotProduct(A=@n46, B=@n47))) {constA=0.325}
  NamedRerouteDeclaration(None=Multiply(A=NamedRerouteUsage(), B=Divide(A=@n45, B=@n48))) {name=Refraction Vector, nodeColor=<rgba 0.0781 1 0 1>}
  n49 = ScalarParameter() {defaultValue=0.5, group=Eye Shader, parameterName=EyeUVPositionX}
  n50 = ScalarParameter() {defaultValue=0.5, group=Eye Shader, parameterName=EyeUVPositionY}
  n51 = NamedRerouteDeclaration(None=Subtract(A=AppendVector(A=@n49, B=@n50)) {constB=0.5}) {name=UVOffset, nodeColor=<rgba 0 0.0391 1 1>}
  n52 = TextureCoordinate()
  n53 = AppendVector(
   A=ScalarParameter() {defaultValue=1, group=Eye Shader, parameterName=EyeUVSizeX}
   B=ScalarParameter() {defaultValue=1, group=Eye Shader, parameterName=EyeUVSizeY}
  )
  NamedRerouteDeclaration(None=MaterialFunctionCall(UVs=Add(A=@n51, B=@n52), Texture Scale=NamedRerouteDeclaration(None=@n53) {name=IrisSize, nodeColor=<rgba 1 0.1172 0 1>}) {materialFunction=&ScaleUVsByCenter}.UVs) {name=Iris UVs, nodeColor=<rgba 0 0.2266 1 1>}
  NamedRerouteDeclaration(None=PixelNormalWS()) {name=Normal, nodeColor=<rgba 0 0 1 1>}
  NamedRerouteDeclaration(None=Divide(A=ScalarParameter() {defaultValue=1.336, group=Eye Shader, parameterName=Iris IOR}) {constA=1.0003, desc=n}) {name=n, nodeColor=<rgba 0.6172 0 1 1>}
  n54 = NamedRerouteUsage()
  n55 = NamedRerouteUsage()
  NamedRerouteDeclaration(None=Multiply(A=NamedRerouteUsage(), B=DotProduct(A=@n54, B=@n55) {desc=facing}) {desc=w}) {name=w, nodeColor=<rgba 1 0.8297 0 1>}
  n56 = NamedRerouteUsage()
  n57 = NamedRerouteUsage()
  n58 = Multiply(A=Subtract(A=@n56, B=@n57), B=Add(A=@n56, B=@n57))
  NamedRerouteDeclaration(None=SquareRoot(None=Add(A=@n58)) {desc=k}) {name=k, nodeColor=<rgba 0.518 1 0 1>}
  n59 = NamedRerouteUsage()
  n60 = NamedRerouteUsage()
  n61 = Subtract(A=Multiply(A=Subtract(A=@n59, B=@n60), B=NamedRerouteUsage()), B=Multiply(A=NamedRerouteUsage(), B=NamedRerouteUsage())) {desc=t}
  NamedRerouteDeclaration(None=Multiply(A=Normalize(VectorInput=@n61)) {constB=-1}) {name=Eye Refraction Direction, nodeColor=<rgba 0 0.2882 1 1>}
  n62 = NamedRerouteUsage()
  n63 = NamedRerouteUsage()
  n64 = NamedRerouteUsage()
  n65 = ScalarParameter() {defaultValue=1, group=Eye Shader, parameterName=Refraction On/Off, sliderMax=1}
  n66 = Multiply(A=Divide(B=Multiply(A=@n62) {constB=2}), B=Subtract(A=LinearInterpolate(A=@n63, B=@n64, Alpha=@n65)) {constB=0.5})
  n67 = Multiply(A=Normalize(VectorInput=@n66), B=ScalarParameter() {defaultValue=0.5, group=Eye Shader, parameterName=Iris Texture Edge}) {desc=UV Max}
  n68 = Multiply(A=Distance(A=Constant(), B=@n66) {desc=UV Length}) {constB=2}
  n69 = Saturate(None=Multiply(A=OneMinus(None=@n68), B=ScalarParameter() {defaultValue=1, group=Eye Shader, parameterName=PupilScale}))
  NamedRerouteDeclaration(None=Add(A=LinearInterpolate(A=@n67, Alpha=@n69) {constB=0}) {constB=0.5}) {name=Final UV, nodeColor=<rgba 0 1 0.5232 1>}
  NamedRerouteDeclaration(None=@n6.Normal) {name=Original Normal, nodeColor=<rgba 1 0 0.1172 1>}
  NamedRerouteDeclaration(None=ComponentMask(None=@n17) {b=true}) {name=Limbus Ring, nodeColor=<rgba 0 1 0.9531 1>}
  n70 = ComponentMask(None=Multiply(A=NamedRerouteUsage()) {constB=-1}) {b=true, g=true, r=true}
  n71 = Constant3Vector()
  n72 = MaterialFunctionCall() {materialFunction=&CameraDirectionVector}
  n73 = Constant3Vector() {constant=<rgba 0 0 -1 0>}
  n74 = Subtract(
   A=Add(A=Multiply(A=@n72.Result) {constB=2}, B=@n73)
   B=Multiply(A=MaterialFunctionCall!(Vector1=@n72.Result, Vector2=@n73) {materialFunction=/Engine/Functions/Engine_MaterialFunctions02/Math/CreateThirdOrthogonalVector}.Vector3) {constB=2}
  )
  NamedRerouteDeclaration(None=If(A=Distance(A=@n70, B=@n71), B=Constant(), "A > B"=@n70, "A == B"=Normalize(VectorInput=@n74), "A < B"=@n70)) {name=LightingAngle, nodeColor=<rgba 1 0.5684 0 1>}
  n75 = Multiply(A=PixelNormalWS(), B=Transform(None=Constant3Vector() {constant=<rgba 1 -1 1 0>}) {transformSourceType=TRANSFORMSOURCE_Local})
  n76 = Multiply(A=NamedRerouteUsage()) {constB=-1}
  n77 = Power(Base=Saturate(None=DotProduct(A=@n75, B=@n76))) {constExponent=4}
  n78 = NamedRerouteUsage()
  n79 = ScalarParameter() {defaultValue=4, parameterName=Caustic Refracted Light Intensity}
  n80 = ScalarParameter() {defaultValue=0.4, parameterName=Caustic Direct Light Intensity}
  n81 = Saturate(None=NamedRerouteUsage())
  NamedRerouteDeclaration(None=Multiply(A=Multiply(A=@n77, B=@n78), B=LinearInterpolate(A=@n79, B=@n80, Alpha=@n81) {constA=1, constB=0.1})) {name=Caustic, nodeColor=<rgba 0 1 0.6016 1>}
  n82 = NamedRerouteUsage()
  NamedRerouteDeclaration(None=TextureSampleParameter2D(UVs=@n82) {group=Eye Shader, parameterName=IrisDiffuse, texture=/Game/Global/Textures/Default/Eyes/EyeIrisBaseColor_Example_2}.RGB) {name=Iris Base Color, nodeColor=<rgba 1 0 0.4219 1>}
  n83 = ScalarParameter() {defaultValue=0.55, group=Analytical Eye Highlight Controls, parameterName=SpecR2}
  n84 = Power(Base=Power(Base=Multiply(A=@n83)) {desc=a}) {desc=a2}
  n85 = Multiply(A=Normalize(VectorInput=NamedRerouteUsage())) {constB=-1}
  n86 = CameraVectorWS()
  n87 = Transform(None=VertexNormalWS()) {transformSourceType=TRANSFORMSOURCE_World, transformType=TRANSFORM_Camera}
  n88 = Constant3Vector() {constant=<rgba -1 -1 1 0>}
  n89 = DotProduct(A=Normalize(VectorInput=Add(A=@n85, B=@n86)), B=Transform(None=Multiply(A=@n87, B=@n88)) {transformSourceType=TRANSFORMSOURCE_Camera})
  n90 = Multiply(A=@n84, B=@n89)
  n91 = Add(A=Multiply(A=Subtract(A=@n90, B=@n89), B=@n89)) {desc=d}
  n92 = Divide(
   A=@n84
   B=Multiply(A=MaterialFunctionCall() {materialFunction=&Pi}.Result, B=Power(Base=@n91))
  )
  NamedRerouteDeclaration(None=Multiply(A=NamedRerouteUsage(), B=Power(Base=@n92) {constExponent=3})) {name=Cine Caustic, nodeColor=<rgba 0 1 0.6016 1>}
  n93 = TextureSampleParameter2D(UVs=NamedRerouteUsage()) {group=Eye Shader, parameterName=IrisSpecular, samplerType=SAMPLERTYPE_Masks, texture=&T_WhiteMask}
  n94 = ScalarParameter() {parameterName=IrisSpecular}
  n95 = ScalarParameter() {parameterName=IrisMetallic}
  n96 = ScalarParameter() {parameterName=IrisRoughness}
  NamedRerouteDeclaration(None=Multiply(
    A=@n93.RGB
    B=MaterialFunctionCall(X=@n94, Y=@n95, Z=@n96) {materialFunction=&MakeFloat3}.Result
   )) {name=Iris Specular, nodeColor=<rgba 0 0.7656 1 1>}
  NamedRerouteDeclaration(None=StaticSwitchParameter(
    True=Constant3Vector()
    False=MaterialFunctionCall() {materialFunction=/Game/Global/MaterialFunctions/FNCollectionParameters/MF_SunParameters_FNC}.Direction
   ) {group=Eye Shader Cine, parameterName=Use Cine Sun Direction}) {name=SunDirection, nodeColor=<rgba 0.6641 1 0 1>}
  NamedRerouteDeclaration(None=Constant()) {name=Cine Specular, nodeColor=<rgba 0 1 0.8828 1>}
  NamedRerouteDeclaration(None=Add(
    A=TextureSampleParameter2D(UVs=@n82) {group=Eye Shader, parameterName=IrisEmissive, texture=&T_BlacksRGB}.RGB
    B=NamedRerouteUsage()
   ) {constB=0}) {name=Iris Emissive, nodeColor=<rgba 1 0.375 0 1>}
  n97 = NamedRerouteUsage()
  n98 = MaterialFunctionCall() {materialFunction=&TangentBasis}
  n99 = NamedRerouteUsage()
  n100 = DotProduct(A=@n98.X U, B=@n99)
  n101 = Normalize(VectorInput=Subtract(A=@n98.X U, B=Multiply(A=@n100, B=@n99)))
  n102 = NamedRerouteUsage()
  n103 = CrossProduct(A=@n101, B=@n99)
  n104 = Multiply(
   A=Multiply(A=NamedRerouteUsage(), B=Constant2Vector() {g=1, r=-1})
   B=AppendVector(A=DotProduct(A=@n101, B=@n102), B=DotProduct(A=@n103, B=@n102))
  )
  NamedRerouteDeclaration(None=LinearInterpolate(A=@n97, B=Add(A=@n104, B=@n97) {desc=Refracted UV}, Alpha=NamedRerouteUsage())) {name=Refracted UV Offset, nodeColor=<rgba 0.6406 1 0 1>}
  NamedRerouteDeclaration(None=ComponentMask(None=@n17) {g=true}) {name=Iris UV Blend, nodeColor=<rgba 0 1 0.9531 1>}
  NamedRerouteDeclaration(None=ScalarParameter() {defaultValue=0.25, group=Eye Shader, parameterName=Iris UV Radius}) {name=IrisUVRadius, nodeColor=<rgba 0.6172 0 1 1>}
  n105 = VectorParameter() {group=Eye Shader, parameterName=Limbus, channelNames {a="", b="", g="", r=Iris Fade}, defaultValue=<rgba 0.035 0.045 1 0>}
  NamedRerouteDeclaration(None=@n105) {name=LimbusUVWidth, nodeColor=<rgba 1 0 0.1172 1>}
  ScalarParameter() {defaultValue=0.9, group=Eye Shader, parameterName=Limbal Ring}
  n106 = NamedRerouteUsage()
  n107 = MaterialFunctionCall(
   Normal=NamedRerouteDeclaration(None=TextureSampleParameter2D(UVs=@n106) {parameterName=ES_PupilNormal, samplerType=SAMPLERTYPE_Normal, texture=/Game/Global/Textures/Default/Eyes/Sphere_EYE_NORMALS}.RGB) {name=Iris Normal, nodeColor=<rgba 0 0 1 1>}
   Flatness=ScalarParameter() {group=Eye Shader, parameterName=Flatten Normal}
  ) {materialFunction=&FlattenNormal}
  TangentOutput!(None=NamedRerouteDeclaration(None=Transform(None=@n107.Result)) {name=EyeDirection, nodeColor=<rgba 1 0.3047 0 1>})
  n108 = Constant3Vector() {constant=<rgba 0 0 1 0>}
  Transform(None=Multiply(A=Transform(None=@n108) {transformType=TRANSFORM_Camera}, B=Constant3Vector() {constant=<rgba 1 -1 1 0>})) {transformSourceType=TRANSFORMSOURCE_Camera, transformType=TRANSFORM_Tangent}
  VertexInterpolator(VS=PreSkinnedNormal())
  VertexNormalWS()
 ]}
