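# Material function that gives Skin FX materials an optional flat-shaded look, toggled by the UseFlatShading static switch.
# It builds a flat variant with black base color, full roughness and the original base color fed into emissive.
# A channel-masked, contrast-adjusted SkinFX input drives the blend between original and flat attributes.
# The blended result gets its shading model set explicitly, while the disabled switch path bypasses the blend.
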MaterialFunction /Game/Packages/Fortress_SharedMaterials/Base_Material_Functions/MFC_FlatShading
graph {nodes [
  n1 = BlendMaterialAttributes(A=NamedRerouteUsage(), B=NamedRerouteUsage(), Alpha=NamedRerouteUsage())
  n2 = ShadingModel()
  FunctionOutput(None=StaticSwitchParameter(
    True=SetMaterialAttributes!(MaterialAttributes=@n1, ShadingModel=@n2) {attributeSetTypes [#1]}
    False=NamedRerouteUsage()
   ) {group=Skin FX, parameterName=UseFlatShading})
  NamedRerouteDeclaration(None=FunctionInput() {inputName=MA, inputType=FunctionInput_MaterialAttributes}) {name=MA, nodeColor=<rgba 0 0.6484 1 1>}
  n3 = GetMaterialAttributes(None=NamedRerouteUsage()) {attributeGetTypes [#2, #3]}
  NamedRerouteDeclaration(None=SetMaterialAttributes!(
    MaterialAttributes=@n3.MaterialAttributes
    BaseColor=Constant()
    EmissiveColor=Add(A=@n3.BaseColor, B=@n3.BaseColor)
    Roughness=Constant() {r=1}
   ) {attributeSetTypes [#2, #3, #4]}) {name=MA_Out, nodeColor=<rgba 1 0.0234 0 1>}
  n4 = FunctionInput() {inputName=SkinFX, inputType=FunctionInput_Vector4, sortPriority=1}
  NamedRerouteDeclaration(None=MaterialFunctionCall(
    In=ChannelMaskParameter(None=@n4) {group=FlatShading, parameterName=FlatShading_SkinFX_Mask}.R
    Contrast=ScalarParameter() {defaultValue=1, group=FlatShading, parameterName=SkinFXMask_Contrast, sliderMax=8}
   ) {materialFunction=&CheapContrast}.Result) {name=SkinFX_Mask, nodeColor=<rgba 0 0.5312 1 1>}
 ]}
