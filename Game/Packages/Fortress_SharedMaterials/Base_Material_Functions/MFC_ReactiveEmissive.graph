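# Material attributes function for Skin FX that animates emissive color in response to timestamp parameters, passing attributes through unchanged when its static switch is off.
# Two multipliers each fade from max to min after their trigger time, one smoothstepped and one linear with optional Fresnel shaping, and combine with an intensity multiplier.
# Negative default times prevent playback at game start.
# Optional switches darken base color and opacity using the desaturated, boosted emissive as a mask.
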MaterialFunction /Game/Packages/Fortress_SharedMaterials/Base_Material_Functions/MFC_ReactiveEmissive
graph {nodes [
  n1 = FunctionInput() {inputType=FunctionInput_MaterialAttributes}
  n2 = GetMaterialAttributes(None=@n1) {attributeGetTypes [#1, #2, #3]}
  n3 = ScalarParameter() {
   defaultValue=-10
   desc="Negative Value by default so it\r\ndoesn't play when game starts"
   group=Reactive Emissive
   parameterName=RE_Anim1Time_CCPM
  }
  n4 = Time()
  n5 = Divide(A=Subtract(A=@n4, B=@n3), B=ScalarParameter() {defaultValue=1, group=Reactive Emissive, parameterName=RE_Anim1Duration})
  n6 = LinearInterpolate(
   A=ScalarParameter() {defaultValue=1, group=Reactive Emissive, parameterName=RE_Anim1MultiplierMax}
   B=ScalarParameter() {group=Reactive Emissive, parameterName=RE_Anim1MultiplierMin}
   Alpha=SmoothStep(Value=Saturate(None=@n5))
  )
  n7 = ScalarParameter() {defaultValue=1.5, group=Reactive Emissive, parameterName=RE_Anim2MultiplierMax}
  n8 = ScalarParameter() {defaultValue=1, group=Reactive Emissive, parameterName=RE_Anim2MultiplierMin}
  n9 = LinearInterpolate(
   A=@n7
   B=LinearInterpolate(A=@n8, B=@n7, Alpha=Fresnel() {baseReflectFraction=0, exponent=0.5}) {constA=1, constB=2}
   Alpha=ScalarParameter() {group=Reactive Emissive, parameterName=RE_Anim2UseFresnel}
  ) {constA=1}
  n10 = ScalarParameter() {
   defaultValue=-10
   desc="Negative Value by default so it\r\ndoesn't play when game starts"
   group=Reactive Emissive
   parameterName=RE_Anim2Time_CCPM
  }
  n11 = Saturate(None=Divide(A=Subtract(A=@n4, B=@n10), B=ScalarParameter() {defaultValue=0.1, group=Reactive Emissive, parameterName=RE_Anim2Duration}) {constB=0.1})
  n12 = Multiply(
   A=Multiply(A=Multiply(A=@n2.EmissiveColor, B=@n6), B=LinearInterpolate(A=@n9, B=@n8, Alpha=@n11) {constA=1.5})
   B=ScalarParameter() {defaultValue=1, group=Reactive Emissive, parameterName=RE_EmissiveIntensityMultiplier}
  )
  n13 = Saturate(None=Multiply(A=Desaturation(None=@n12)) {constB=10})
  n14 = StaticSwitchParameter(True=LinearInterpolate(A=@n2.BaseColor, Alpha=@n13) {constB=0}, False=@n2.BaseColor) {group=Reactive Emissive, parameterName=RE_DarkenBaseColor}
  n15 = StaticSwitchParameter(True=LinearInterpolate(A=@n2.Opacity, Alpha=@n13) {constB=0}, False=@n2.Opacity) {group=Reactive Emissive, parameterName="RE_DarkenOpacity(UseWithSubsurface)"}
  FunctionOutput(None=StaticSwitchParameter(
    True=SetMaterialAttributes!(MaterialAttributes=@n2.MaterialAttributes, EmissiveColor=@n12, BaseColor=@n14, Opacity=@n15) {attributeSetTypes [#1, #2, #3]}
    False=@n1
   ) {group=Skin FX, parameterName=Use Reactive Emissive})
 ]}
