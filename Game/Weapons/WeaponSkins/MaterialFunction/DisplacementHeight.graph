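# Weapon-skin material function that displaces a surface along its world-space vertex normal, adding to the incoming World Position Offset.
# The saturated shrinkage parameter scales the offset from zero to MaximumShrinkage, which defaults negative to pull vertices inward.
# Sides scales normal X and Y, while Vert scales an even blend of clamped and absolute Z, biasing vertical movement.
# The Alpha input blends original and displaced material attributes.
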MaterialFunction /Game/Weapons/WeaponSkins/MaterialFunction/DisplacementHeight
graph {nodes [
  n1 = FunctionInput() {inputName=Materials, inputType=FunctionInput_MaterialAttributes}
  n2 = GetMaterialAttributes(None=@n1) {attributeGetTypes [#1]}
  n3 = ScalarParameter() {defaultValue=-4.0661, parameterName=MaximumShrinkage}
  n4 = Saturate(None=ScalarParameter() {group=Displacement, parameterName=shrinkage})
  n5 = ScalarParameter() {defaultValue=0.9571, parameterName=Sides}
  n6 = MaterialFunctionCall(Float3=VertexNormalWS()) {materialFunction=&BreakOutFloat3Components}
  n7 = Multiply(A=@n5, B=@n6.R)
  n8 = Multiply(A=@n5, B=@n6.G)
  n9 = ScalarParameter() {defaultValue=0.3332, parameterName=min}
  n10 = Multiply(A=LinearInterpolate(A=Clamp(None=@n6.B, Min=@n9), B=Abs(None=@n6.B)), B=ScalarParameter() {defaultValue=1.2823, parameterName=Vert})
  n11 = Add(
   A=@n2.WorldPositionOffset
   B=Multiply(
    A=LinearInterpolate(B=@n3, Alpha=@n4)
    B=MaterialFunctionCall(X=@n7, Y=@n8, Z=@n10) {materialFunction=&MakeFloat3}.Result
   )
  )
  FunctionOutput(None=BlendMaterialAttributes(
    A=@n1
    B=SetMaterialAttributes!(MaterialAttributes=@n2.MaterialAttributes, WorldPositionOffset=@n11) {attributeSetTypes [#1]}
    Alpha=FunctionInput() {inputName=Alpha, inputType=FunctionInput_Scalar}
   ))
  PreSkinnedNormal()
 ]}
