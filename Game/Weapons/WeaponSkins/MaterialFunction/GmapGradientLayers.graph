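# Material function for weapon skins that recolors a material with up to five gradient-mapped layers.
# The saturated Gmap scalar samples each layer's gradient texture, and mask RGB channels blend the layers, with alpha adding a fifth via static switch.
# The same mask blends per-layer roughness multipliers and metalness, which either scales or overwrites incoming metallic.
# The resulting base color, roughness and metallic blend over the input attributes by Alpha.
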MaterialFunction /Game/Weapons/WeaponSkins/MaterialFunction/GmapGradientLayers
graph {nodes [
  n1 = FunctionInput() {inputName=Material, inputType=FunctionInput_MaterialAttributes}
  n2 = GetMaterialAttributes(None=@n1) {attributeGetTypes [#1, #2, #3]}
  n3 = Saturate(None=FunctionInput(Preview=TextureSample() {texture=/Game/Global/Textures/Default/Blanks/T_WhiteUserInterface}.R) {bUsePreviewValueAsDefault=true, inputName=Gmap, inputType=FunctionInput_Scalar})
  n4 = TextureSampleParameter2D(UVs=@n3) {
   group=Gmap Layers
   parameterName=Layer1_Gradient
   samplerSource=SSM_Clamp_WorldGroupSettings
   sortPriority=1
   texture=/Game/Characters/Player/Male/Medium/Bodies/M_MED_Dragon_Ninja/Textures/V1/T_DragonNinja_RedArmor
  }
  n5 = TextureSampleParameter2D(UVs=@n3) {
   group=Gmap Layers
   parameterName=Layer2_Gradient
   samplerSource=SSM_Clamp_WorldGroupSettings
   sortPriority=2
   texture=&T_DragonNinja_RedArmor
  }
  n6 = TextureSampleParameter2D(UVs=TextureCoordinate()) {
   group=Gmap Layers
   parameterName=Layer Mask
   samplerSource=SSM_Wrap_WorldGroupSettings
   samplerType=SAMPLERTYPE_Masks
   sortPriority=-2
   texture=&T_BlackMask
  }
  n7 = TextureSampleParameter2D(UVs=@n3) {
   group=Gmap Layers
   parameterName=Layer3_Gradient
   samplerSource=SSM_Clamp_WorldGroupSettings
   sortPriority=3
   texture=&T_DragonNinja_RedArmor
  }
  n8 = TextureSampleParameter2D(UVs=@n3) {
   group=Gmap Layers
   parameterName=Layer4_Gradient
   samplerSource=SSM_Clamp_WorldGroupSettings
   sortPriority=4
   texture=&T_DragonNinja_RedArmor
  }
  n9 = LinearInterpolate(A=LinearInterpolate(A=LinearInterpolate(A=@n4.RGB, B=@n5.RGB, Alpha=@n6.R), B=@n7.RGB, Alpha=@n6.G), B=@n8.RGB, Alpha=@n6.B)
  n10 = TextureSampleParameter2D(UVs=@n3) {
   group=Gmap Layers
   parameterName=Layer5_Gradient
   samplerSource=SSM_Clamp_WorldGroupSettings
   sortPriority=5
   texture=&T_DragonNinja_RedArmor
  }
  n11 = StaticBoolParameter() {group=Gmap Layers, parameterName=use Alpha Channel as mask, sortPriority=-1}
  n12 = StaticSwitch(True=LinearInterpolate(A=@n9, B=@n10.RGB, Alpha=@n6.A), False=@n9, Value=@n11)
  n13 = Multiply(A=@n2.Roughness, B=ScalarParameter() {defaultValue=1, group=Gmap Layers, parameterName=Layer1_Roughness, sortPriority=1})
  n14 = Multiply(A=@n2.Roughness, B=ScalarParameter() {defaultValue=1, group=Gmap Layers, parameterName=Layer2_Roughness, sortPriority=2})
  n15 = ScalarParameter() {defaultValue=1, group=Gmap Layers, parameterName=Layer3_Roughness, sortPriority=3}
  n16 = LinearInterpolate(
   A=LinearInterpolate(A=LinearInterpolate(A=@n13, B=@n14, Alpha=@n6.R), B=Multiply(A=@n2.Roughness, B=@n15), Alpha=@n6.G)
   B=Multiply(A=@n2.Roughness, B=ScalarParameter() {defaultValue=1, group=Gmap Layers, parameterName=Layer4_Roughness, sortPriority=4})
   Alpha=@n6.B
  )
  n17 = ScalarParameter() {defaultValue=1, group=Gmap Layers, parameterName=Layer5_Roughness, sortPriority=5}
  n18 = StaticSwitch(True=LinearInterpolate(A=@n16, B=Multiply(A=@n2.Roughness, B=@n17), Alpha=@n6.A), False=@n16, Value=@n11)
  n19 = Multiply(A=@n2.Metallic, B=ScalarParameter() {group=Gmap Layers, parameterName=Layer1_Metalness, sortPriority=1})
  n20 = Multiply(A=@n2.Metallic, B=ScalarParameter() {group=Gmap Layers, parameterName=Layer2_Metalness, sortPriority=2})
  n21 = ScalarParameter() {group=Gmap Layers, parameterName=Layer3_Metalness, sortPriority=3}
  n22 = LinearInterpolate(
   A=LinearInterpolate(A=LinearInterpolate(A=@n19, B=@n20, Alpha=@n6.R), B=Multiply(A=@n2.Metallic, B=@n21), Alpha=@n6.G)
   B=Multiply(A=@n2.Metallic, B=ScalarParameter() {group=Gmap Layers, parameterName=Layer4_Metalness, sortPriority=4})
   Alpha=@n6.B
  )
  n23 = Multiply(A=@n2.Metallic, B=ScalarParameter() {group=Gmap Layers, parameterName=Layer5_Metalness, sortPriority=5})
  n24 = ScalarParameter() {group=Gmap Layers, parameterName=Layer1_Metalness, sortPriority=1}
  n25 = ScalarParameter() {group=Gmap Layers, parameterName=Layer2_Metalness, sortPriority=2}
  n26 = LinearInterpolate(
   A=LinearInterpolate(
    A=LinearInterpolate(A=@n24, B=@n25, Alpha=@n6.R)
    B=ScalarParameter() {group=Gmap Layers, parameterName=Layer3_Metalness, sortPriority=3}
    Alpha=@n6.G
   )
   B=ScalarParameter() {group=Gmap Layers, parameterName=Layer4_Metalness, sortPriority=4}
   Alpha=@n6.B
  )
  n27 = ScalarParameter() {group=Gmap Layers, parameterName=Layer5_Metalness, sortPriority=5}
  n28 = LinearInterpolate(
   A=StaticSwitch(True=LinearInterpolate(A=@n22, B=@n23, Alpha=@n6.A), False=@n22, Value=@n11)
   B=StaticSwitch(True=LinearInterpolate(A=@n26, B=@n27, Alpha=@n6.A), False=@n26, Value=@n11)
   Alpha=ScalarParameter() {defaultValue=1, group=Gmap Layers, parameterName=Overwrite Metallic from _S Mask}
  )
  FunctionOutput(None=BlendMaterialAttributes(
    A=@n1
    B=SetMaterialAttributes!(MaterialAttributes=@n2.MaterialAttributes, BaseColor=@n12, Roughness=@n18, Metallic=@n28) {attributeSetTypes [#1, #2, #3]}
    Alpha=FunctionInput() {bUsePreviewValueAsDefault=true, inputName=Alpha, inputType=FunctionInput_Scalar}
   ))
 ]}
