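# Material function for weapon skins that layers scratch wear and grime onto an incoming material's base color and roughness.
# The Scratch mask blends toward a desaturated, brightness-scaled base color and a multiplied roughness.
# The Grime mask then blends over that toward a fixed roughness parameter and a grayscale tone, read from the red channel of a gradient texture using the scalar Gmap input as UV.
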MaterialFunction /Game/Weapons/WeaponSkins/MaterialFunction/GrimeScratch
graph {nodes [
  n1 = GetMaterialAttributes(None=FunctionInput() {bUsePreviewValueAsDefault=true, inputName=Material, inputType=FunctionInput_MaterialAttributes}) {attributeGetTypes [#1, #2]}
  n2 = Desaturation(None=@n1.BaseColor, Fraction=ScalarParameter() {defaultValue=0.5, group=Scratch and Grime, parameterName=ScratchSaturation})
  n3 = ScalarParameter() {defaultValue=1.2, group=Scratch and Grime, parameterName=ScratchBrightness}
  n4 = FunctionInput() {bUsePreviewValueAsDefault=true, inputName=Scratch, inputType=FunctionInput_Scalar}
  n5 = LinearInterpolate(A=@n1.BaseColor, B=Saturate(None=Multiply(A=@n2, B=@n3)), Alpha=@n4)
  n6 = TextureSampleParameter2D(UVs=FunctionInput() {bUsePreviewValueAsDefault=true, inputName=Gmap, inputType=FunctionInput_Scalar}) {group=Scratch and Grime, parameterName=GrimeGradient, texture=/Game/Weapons/WeaponSkins/Textures/T_Grime_Grad}
  n7 = FunctionInput() {bUsePreviewValueAsDefault=true, inputName=Grime, inputType=FunctionInput_Scalar}
  n8 = LinearInterpolate(
   A=@n1.Roughness
   B=Multiply(A=@n1.Roughness, B=ScalarParameter() {defaultValue=1.2, group=Scratch and Grime, parameterName=ScratchRoughness})
   Alpha=@n4
  )
  n9 = ScalarParameter() {defaultValue=1, group=Scratch and Grime, parameterName=GrimeRoughness}
  FunctionOutput(None=SetMaterialAttributes!(
    MaterialAttributes=@n1.MaterialAttributes
    BaseColor=LinearInterpolate(A=@n5, B=@n6.R, Alpha=@n7)
    Roughness=LinearInterpolate(A=@n8, B=@n9, Alpha=@n7)
   ) {attributeSetTypes [#1, #2]})
 ]}
