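# Weapon-skin material function that blends a holographic neon scanline look into incoming material attributes through a MASK input.
# Screen-space scanlines, shaped by Fresnel, modulate top/bottom color gradients built from radius-normalized pre-skinned height via a vertex interpolator.
# A fixed fake light picks between main and edge colors, while bright base-color red-channel areas get separate white-mask colors.
# Fresnel and a blend color shape masked emissive; unmasked areas receive an adjustable Fresnel glow.
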MaterialFunction /Game/Weapons/WeaponSkins/MaterialFunction/Holographic_Neon_Lines
graph {nodes [
  n1 = FunctionInput() {inputName=Material, inputType=FunctionInput_MaterialAttributes, sortPriority=-200}
  n2 = GetMaterialAttributes(None=@n1) {attributeGetTypes [
    #1
    #2
    #3
    #4
    #5
    #6
   ]}
  n3 = Saturate(None=@n2.EmissiveColor)
  n4 = ScreenPosition()
  n5 = ScalarParameter() {defaultValue=0.2, group=Holographic_NeonLines, parameterName=Holo_ Scanlines Inner Thickness, sliderMax=1, sliderMin=0.01, sortPriority=-199}
  n6 = Fresnel(ExponentIn=ScalarParameter() {defaultValue=3, group=Holographic_NeonLines, parameterName=Holo_ Fresnel Exponent}) {exponent=3}
  n7 = OneMinus(None=Power(
    Base=MaterialFunctionCall!(ScanLine_Tiling_1, Scanline_PannerSpeed_1, ScanLine_Tiling, Scanline_PannerSpeed, UVs=@n4.ViewportUV) {materialFunction=&/ScanLines}.Result
    Exp=Add(A=@n5, B=@n6)
   ))
  n8 = ScalarParameter() {defaultValue=0.2, group=Holographic_NeonLines, parameterName=Holo_ Scanlines Dark Value, sortPriority=-198}
  n9 = VectorParameter() {group=Holographic_NeonLines, parameterName=Holo_ White Mask Base Color, defaultValue=<rgba 0.6698 0.784 0.9387 1>}
  n10 = VectorParameter() {group=Holographic_NeonLines, parameterName=Holo_ White Mask Fresnel Color, defaultValue=<rgba 0 0.5056 1 1>}
  n11 = Constant3Vector() {desc=Light Direction, constant=<rgba 0 0 -1 1>}
  n12 = GetMaterialAttributes(None=@n1) {attributeGetTypes [#6, #1]}
  n13 = Transform(None=@n12.Normal)
  n14 = Power(Base=Saturate(None=DotProduct(A=@n11, B=@n13)), Exp=Constant() {r=0.6}) {constExponent=0.6}
  n15 = Saturate(None=Add(A=@n14, B=@n6))
  n16 = ComponentMask(None=@n12.BaseColor) {r=true}
  n17 = Constant() {r=2}
  n18 = Saturate(None=MaterialFunctionCall(
    In=MaterialFunctionCall!(Image=@n16, Black intensity level=@n17) {materialFunction=/Engine/Functions/Engine_MaterialFunctions02/Texturing/RaiseBlackLevelsByPercentage}.Result
    Contrast=Constant() {r=1}
   ) {materialFunction=&CheapContrast}.Result)
  n19 = VectorParameter() {
   desc="Accent color color for the back of the gradient. Gradient settings can be found under the \"Gradient\" group."
   group=Holographic_NeonLines
   parameterName=Holo_ Edge Color Bottom
   sortPriority=-449
   defaultValue=<rgba 0 2 1 1>
  }
  n20 = VectorParameter() {
   desc="Accent color for the front of the gradient. Gradient settings can be found under the \"Gradient\" group."
   group=Holographic_NeonLines
   parameterName=Holo_ Edge Color Top
   sortPriority=-499
   defaultValue=<rgba 0.0521 2 0.0521 1>
  }
  n21 = ScalarParameter() {defaultValue=-5, desc=Holo_ Gradient Mask Vertical Contrast, group=Holographic_NeonLines, parameterName=Holo_ Gradient Mask Vertical Balance}
  n22 = ScalarParameter() {defaultValue=1, desc=Holo_ Gradient Mask Vertical Contrast, group=Holographic_NeonLines, parameterName=Holo_ Gradient Mask Vertical Contrast}
  n23 = MaterialFunctionCall(Float3=PreSkinnedPosition()) {desc=chooses gradient direction, materialFunction=&BreakOutFloat3Components}
  n24 = ObjectRadius()
  n25 = LinearInterpolate(A=Subtract(A=@n21, B=@n22), B=@n22, Alpha=ConstantBiasScale(None=Divide(A=@n23.B, B=@n24) {desc=sets gradient scale based on mesh size}))
  n26 = ComponentMask(None=VertexInterpolator(VS=Saturate(None=@n25) {desc=necessary to keep upper portions of gradient from blowing out}).PS) {desc="only outputs the first channel from vertex shader; also fixes red tint bug", r=true}
  n27 = LinearInterpolate(A=@n19.RGB, B=@n20.RGB, Alpha=@n26)
  n28 = VectorParameter() {group=Holographic_NeonLines, parameterName=Holo_ Main Color Bottom, sortPriority=-450, defaultValue=<rgba 0 0.5382 0.6458 1>}
  n29 = VectorParameter() {group=Holographic_NeonLines, parameterName=Holo_ Main Color Top, sortPriority=-500, defaultValue=<rgba 0 0.6562 0.3281 1>}
  n30 = LinearInterpolate(A=@n28.RGB, B=@n29.RGB, Alpha=@n26)
  n31 = Multiply(
   A=Add(
    A=Clamp(None=@n7, Min=@n8)
    B=ScalarParameter() {defaultValue=0.5, group=Holographic_NeonLines, parameterName=Holo_ Scanlines Additional Brightness, sortPriority=-197}
   )
   B=LinearInterpolate(A=LinearInterpolate(A=@n30, B=@n27, Alpha=@n14), B=LinearInterpolate(A=@n9.RGB, B=@n10.RGB, Alpha=@n15), Alpha=@n18)
  )
  n32 = Add(
   A=Multiply(A=Multiply(A=@n31, B=@n6), B=ScalarParameter() {group=Holographic_NeonLines, parameterName=Holo_ Non-FX Masked Areas Emissive Strength})
   B=@n3
  )
  n33 = LinearInterpolate(A=@n2.BaseColor, B=Constant(), Alpha=@n18)
  n34 = Saturate(None=LinearInterpolate(
    A=ScalarParameter() {defaultValue=-0.8, group=Holographic_NeonLines, parameterName=Holo_ Fresnel Core Intensity, sortPriority=-399}
    B=ScalarParameter() {defaultValue=1.5, group=Holographic_NeonLines, parameterName=Holo_ Fresnel Edge Intensity, sortPriority=-398}
    Alpha=@n6
   ))
  n35 = ScalarParameter() {defaultValue=0.14, group=Holographic_NeonLines, parameterName=Holo_ Fresnel Dark Value Clamp, sortPriority=-397}
  n36 = Add(A=@n6) {constB=-0.02}
  n37 = VectorParameter() {
   desc=Additional color used across the entire wrap to unify gradient color scheme. Set to black to disable.
   group=Holographic_NeonLines
   parameterName=Holo_ Blend Color
   sortPriority=-420
   defaultValue=<rgba 0.8333 0 2 0>
  }
  n38 = Add(A=Multiply(A=@n31, B=Add(A=@n34, B=@n35)), B=LinearInterpolate(A=Multiply(A=@n36, B=@n37.RGB), Alpha=@n34) {constB=0})
  n39 = Saturate(None=LinearInterpolate(A=Add(A=@n3, B=@n38), B=@n31, Alpha=@n18))
  FunctionOutput(None=BlendMaterialAttributes(
    A=SetMaterialAttributes!(MaterialAttributes=@n2.MaterialAttributes, EmissiveColor=@n32) {attributeSetTypes [#5]}
    B=MakeMaterialAttributes!(
     BaseColor=@n33
     Metallic=@n2.Metallic
     Specular=@n2.Metallic
     Roughness=@n2.Metallic
     Anisotropy
     EmissiveColor=@n39
     Opacity
     OpacityMask
     Normal=@n2.Metallic
     Tangent
     WorldPositionOffset
     SubsurfaceColor
     ClearCoat
     ClearCoatRoughness
     AmbientOcclusion
     Refraction
     CustomizedUVs_0
     CustomizedUVs_1
     CustomizedUVs_2
     CustomizedUVs_3
     CustomizedUVs_4
     CustomizedUVs_5
     CustomizedUVs_6
     CustomizedUVs_7
     PixelDepthOffset
     ShadingModel
     Displacement
    )
    Alpha=FunctionInput() {inputName=MASK, inputType=FunctionInput_Scalar}
   ))
 ]}
