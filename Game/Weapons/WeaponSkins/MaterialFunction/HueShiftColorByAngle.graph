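# Material function for weapon skins that hue-shifts a material's base color according to camera direction.
# It takes a Material Attributes input, extracts BaseColor and runs it through HueShift.
# The shift amount is CameraDirectionVector multiplied by the CameraDirectionalVectorHueShiftScale parameter in the UseColorShiftByAngle group.
# The shifted color is written back to BaseColor, passing all other attributes through unchanged.
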MaterialFunction /Game/Weapons/WeaponSkins/MaterialFunction/HueShiftColorByAngle
graph {nodes [
  n1 = GetMaterialAttributes(None=FunctionInput() {inputType=FunctionInput_MaterialAttributes}) {attributeGetTypes [#1]}
  n2 = Multiply(
   A=MaterialFunctionCall() {materialFunction=&CameraDirectionVector}.Result
   B=ScalarParameter() {defaultValue=1, group=UseColorShiftByAngle, parameterName=CameraDirectionalVectorHueShiftScale}
  )
  FunctionOutput(None=SetMaterialAttributes!(
    MaterialAttributes=@n1.MaterialAttributes
    BaseColor=MaterialFunctionCall(Hue Shift Percentage=@n2, Texture=@n1.BaseColor) {materialFunction=&HueShift}.Result
   ) {attributeSetTypes [#1]})
 ]}
