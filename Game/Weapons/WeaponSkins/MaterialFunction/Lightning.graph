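# Material function for weapon skins that adds animated, flickering lightning to the emissive of incoming material attributes.
# Randomly offset, scaled time varies its speed in steps and jumps the parallax-shifted, tiled UVs of a lightning mask.
# Overlapping time fractions drive dead/active windows that, with Perlin noise, gate the bolts before color tinting.
# An Alpha input blends the result over the original, keeping its vertex attributes.
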MaterialFunction /Game/Weapons/WeaponSkins/MaterialFunction/Lightning
graph {nodes [
  n1 = FunctionInput() {inputName=Material, inputType=FunctionInput_MaterialAttributes}
  n2 = GetMaterialAttributes(None=@n1) {attributeGetTypes [#1]}
  n3 = TextureCoordinate()
  n4 = ScalarParameter() {defaultValue=1, group=Lightning, parameterName=Lightning Tiling}
  n5 = ScalarParameter() {defaultValue=0.1215, group=Lightning, parameterName=Lightning Offset_X}
  n6 = ScalarParameter() {defaultValue=0.214, group=Lightning, parameterName=Lightning Offset_Y}
  n7 = Multiply(
   A=Add(
    A=MaterialFunctionCall() {materialFunction=/Game/Packages/Fortress_SharedMaterials/Base_Material_Functions/BaseCustomizationFunctions/MF_BaseWeapon_RandomFloat}.randomFloat
    B=Time()
   )
   B=ScalarParameter() {defaultValue=0.8, group=Lightning, parameterName=Lightning TimeMult}
  )
  n8 = Multiply(A=@n7) {constB=0.25}
  n9 = Multiply(
   A=@n7
   B=LinearInterpolate(
    A=ScalarParameter() {defaultValue=1, parameterName=TimeMin}
    B=ScalarParameter() {defaultValue=2, parameterName=TimeMax}
    Alpha=Round(None=@n8)
   )
  )
  n10 = Multiply(A=@n9) {constB=0.5}
  n11 = Add(
   A=BumpOffset(Coordinate=Multiply(A=@n3, B=@n4), Height=ScalarParameter() {defaultValue=-0.05, group=Lightning, parameterName=Lightning Parallax})
   B=Multiply(A=AppendVector(A=@n5, B=@n6), B=Round(None=@n10))
  )
  n12 = TextureSample(UVs=@n11) {samplerType=SAMPLERTYPE_Masks, texture=/Game/Characters/Player/Female/Medium/Bodies/F_MED_Dark_Viking_01/Textures/LightningStrike_M}
  n13 = Round(None=Frac(None=Multiply(A=@n9) {constB=0.5}))
  n14 = TextureSample(UVs=Multiply(A=@n11) {constB=0.25}) {texture=/Game/Global/Textures/Default/Noise/3xPerlin_M}
  n15 = LinearInterpolate(A=@n14.R, B=@n14.R, Alpha=@n13)
  n16 = Multiply(A=@n9) {constB=0.98}
  n17 = LinearInterpolate(
   A=ScalarParameter() {defaultValue=-3.5, group=Lightning, parameterName=DeadTime}
   B=ScalarParameter() {defaultValue=2.5, group=Lightning, parameterName=ActiveTime}
   Alpha=Max(A=Frac(None=@n9), B=Frac(None=@n16))
  )
  n18 = Multiply(A=LinearInterpolate(A=@n12.R, B=@n12.R, Alpha=@n13), B=Add(A=Subtract(A=@n15), B=Saturate(None=@n17)))
  n19 = Add(
   A=@n2.EmissiveColor
   B=Multiply(A=Saturate(None=@n18), B=VectorParameter() {group=Lightning, parameterName=LightningColor, defaultValue=<rgba 15.7487 0 45 1>}.RGB)
  )
  FunctionOutput(None=BlendMaterialAttributes(
    A=@n1
    B=SetMaterialAttributes!(MaterialAttributes=@n2.MaterialAttributes, EmissiveColor=@n19) {attributeSetTypes [#1]}
    Alpha=FunctionInput() {inputName=Alpha, inputType=FunctionInput_Scalar}
   ) {vertexAttributeBlendType=UseA})
 ]}
