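# Material function for a Constellation skin that darkens masked areas of incoming material attributes and replaces emissive with animated stars and glowing outlines.
# Three star layers sample separate texture channels with alpha-mask-flipped UVs and camera-to-actor parallax; two pan.
# A static switch overlays a scrolling, hue-shifted nebula and colored details, then two fresnel/textured outlines are added.
# An actor-centered screen-space UV chain remains unconnected.
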MaterialFunction /Game/Weapons/WeaponSkins/MaterialFunction/MF_Constellation
graph {nodes [
  n1 = MaterialFunctionCall(Float2=TextureCoordinate()) {materialFunction=&BreakOutFloat2Components}
  n2 = TextureSampleParameter2D() {
   group=Constellation
   parameterName=ConstellationMask
   samplerType=SAMPLERTYPE_Masks
   texture=/Game/Characters/Player/Female/Medium/Bodies/F_MED_Rebirth_Soldier/Skins/Constellation/Textures/T_F_MED_Constellation_Body_Masks
  }
  n3 = Multiply(A=@n2.A) {constB=2}
  n4 = Multiply(A=Abs(None=Subtract(A=@n3))) {constB=2}
  n5 = Multiply(A=MaterialFunctionCall!(In=@n2.A) {materialFunction=/Engine/Functions/Engine_MaterialFunctions02/Utility/Round}.Result) {constB=2}
  n6 = AppendVector(A=Multiply(A=@n1.R, B=Subtract(A=@n4)), B=Multiply(A=Subtract(A=@n5), B=@n1.G))
  n7 = Multiply(
   A=@n6
   B=ComponentMask(None=VectorParameter() {group=Constellation, parameterName=TexCoordScale, defaultValue=<rgba 10 10 0 0>}.RGB) {g=true, r=true}
  )
  n8 = Subtract(A=ActorPositionWS(), B=CameraPositionWS())
  n9 = Multiply(A=Sign(None=ComponentMask(None=@n8) {r=true})) {constB=-1}
  n10 = Normalize(VectorInput=ComponentMask(None=@n8) {g=true, r=true})
  n11 = Constant2Vector() {g=1}
  n12 = Add(A=Multiply(A=DotProduct(A=@n10, B=@n11)) {constB=0.5}) {constB=0.5}
  n13 = Multiply(A=DotProduct(A=Normalize(VectorInput=@n8), B=Constant3Vector() {constant=<rgba 0 0 1 0>})) {constB=0.5}
  n14 = VertexInterpolator(VS=AppendVector(A=Multiply(A=@n9, B=@n12), B=Add(A=@n13) {constB=0.5}))
  n15 = Multiply(
   A=@n14.PS
   B=AppendVector(
    A=ScalarParameter() {defaultValue=1, group=Constellation, parameterName=Parallax0_X}
    B=ScalarParameter() {defaultValue=1, group=Constellation, parameterName=Parallax0_Y}
   )
  )
  n16 = VectorParameter() {group=Constellation, parameterName=CS_Layer_01_PanSpeedAndDirection, defaultValue=<rgba 0.035 -0.25 0 1>}
  n17 = Panner(
   Coordinate=Multiply(A=Add(A=@n7, B=@n15) {constB=0}, B=ScalarParameter() {defaultValue=0.25, group=Constellation, parameterName=Layer0_Tiling})
   Speed=ComponentMask(None=@n16.RGB) {g=true, r=true}
  )
  n18 = TextureObjectParameter() {
   group=Constellation
   parameterName=StarsTexture
   texture=/Game/Characters/Player/Female/Medium/Bodies/F_MED_Gift_Wrap/Textures/T_Snowflake_Pattern
  }
  n19 = Power(Base=@n2.R, Exp=ScalarParameter() {defaultValue=1, group=Constellation, parameterName=StarsMaskPow})
  n20 = Multiply(
   A=Multiply(
    A=TextureSample(UVs=@n17, Tex=@n18) {constMipValue=0, texture=&T_Snowflake_Pattern}.G
    B=VectorParameter() {group=Constellation, parameterName=Layer1_color, defaultValue=<rgba 0.0314 3.8406 5 1>}.RGB
   )
   B=@n19
  )
  n21 = Multiply(
   A=@n14.PS
   B=AppendVector(A=ScalarParameter() {group=Constellation, parameterName=Parallax1_X}, B=ScalarParameter() {group=Constellation, parameterName=Parallax1_Y})
  )
  n22 = VectorParameter() {group=Constellation, parameterName=CS_Layer_02_PanSpeedAndDirection, defaultValue=<rgba 0.035 -0.25 0 1>}
  n23 = Panner(
   Coordinate=Multiply(A=Add(A=@n7, B=@n21) {constB=0}, B=ScalarParameter() {defaultValue=0.75, group=Constellation, parameterName=Layer1_Tiling})
   Speed=ComponentMask(None=@n22.RGB) {g=true, r=true}
  )
  n24 = OneMinus(None=ScalarParameter() {parameterName=Layer1StarMaskIntensity})
  n25 = Multiply(
   A=Multiply(
    A=TextureSample(UVs=@n23, Tex=@n18) {texture=&T_Snowflake_Pattern}.R
    B=VectorParameter() {group=Constellation, parameterName=Layer2_color, defaultValue=<rgba 0.0125 1.5363 2 1>}.RGB
   ) {constB=0.5}
   B=Saturate(None=Add(A=@n19, B=@n24))
  )
  n26 = ScalarParameter() {group=Constellation, parameterName=Parallax2_X}
  n27 = ScalarParameter() {group=Constellation, parameterName=Parallax2_Y}
  n28 = Add(A=@n7, B=Multiply(A=@n14.PS, B=AppendVector(A=@n26, B=@n27))) {constB=0}
  n29 = ScalarParameter() {defaultValue=1.25, group=Constellation, parameterName=Layer2_Tiling}
  n30 = Multiply(
   A=TextureSample(UVs=Multiply(A=@n28, B=@n29), Tex=@n18) {texture=&T_Snowflake_Pattern}.B
   B=VectorParameter() {group=Constellation, parameterName=Layer3_color, defaultValue=<rgba 0.0063 1 0.991 1>}.RGB
  ) {constB=0.25}
  Add(A=Add(A=@n20, B=@n25), B=@n30)
  n31 = FunctionInput(Preview=TextureSample() {texture=/Game/Characters/Player/Female/Medium/Bodies/F_MED_Rebirth_Soldier/Skins/Constellation/Textures/T_F_MED_Constellation_Body_FX}.RGB) {bUsePreviewValueAsDefault=true, inputName=Mask}
  ComponentMask(None=@n31) {g=true}
  n32 = GetMaterialAttributes(None=FunctionInput() {inputName=Material, inputType=FunctionInput_MaterialAttributes}) {attributeGetTypes [
    #1
    #2
    #3
    #4
    #5
   ]}
  n33 = ScalarParameter() {defaultValue=1, group=Constellation, parameterName=BaseColorMultiplier}
  n34 = Constant()
  n35 = ComponentMask(None=@n31) {r=true}
  n36 = Subtract(A=@n35, B=ScalarParameter() {group=Constellation, parameterName=KeepExistingMaterial})
  n37 = Max(A=@n20, B=@n25)
  n38 = CameraVectorWS()
  n39 = DotProduct(A=@n38, B=StaticSwitchParameter(True=PixelNormalWS(), False=VertexNormalWS()) {group=Constellation, parameterName=FresnelUsesPixelNormals})
  n40 = Saturate(None=@n39)
  n41 = Multiply(A=Power(Base=Subtract(A=@n40) {constB=0.1})) {constB=2}
  n42 = Multiply(A=Multiply(A=Max(A=@n37, B=@n30), B=Clamp(None=@n41)), B=ScalarParameter() {group=Constellation, parameterName=StarsEmissiveStrength})
  n43 = ScalarParameter() {group=Constellation, parameterName=NebulaScrollingSpeedU}
  n44 = ScalarParameter() {group=Constellation, parameterName=NebulaScrollingSpeedV}
  n45 = Add(
   A=@n6
   B=Multiply(A=Time(), B=MaterialFunctionCall(X=@n43, Y=@n44) {materialFunction=&MakeFloat2}.Result)
  )
  n46 = ScalarParameter() {group=Constellation, parameterName=NebulaRotation}
  n47 = TextureSampleParameter2D(UVs=Multiply(
    A=MaterialFunctionCall(UVs=@n45, "Rotation Angle (0-1)"=@n46) {materialFunction=&CustomRotator}.Rotated Values
    B=ScalarParameter() {defaultValue=1, group=Constellation, parameterName=NebulaTiling}
   )) {
   group=Constellation
   parameterName=NebulaTexture
   texture=/Game/Characters/Player/Male/Medium/Bodies/M_Med_Soldier_04/Skins/Galaxy/Textures/T_FN_Nebula
  }
  n48 = Power(
   Base=Divide(A=MaterialFunctionCall(f3=@n47.RGB) {materialFunction=&AddComponents}.f3) {constB=3}
   Exp=ScalarParameter() {defaultValue=1, group=Constellation, parameterName=NebulaPow}
  )
  n49 = VectorParameter() {group=Constellation, parameterName=ConstellationSurfaceDetailsColorA, defaultValue=<rgba 0.974 0.224 0.0287 0>}
  n50 = MaterialFunctionCall(Hue Shift Percentage=Multiply(A=@n48, B=ScalarParameter() {parameterName=NebulaeHue}), Texture=@n49.RGB) {materialFunction=&HueShift}
  n51 = TextureSampleParameter2D() {
   group=Constellation
   parameterName=ConstellationSurfaceDetailsMask
   samplerType=SAMPLERTYPE_Masks
   texture=&T_BlackMask
  }
  n52 = VectorParameter() {group=Constellation, parameterName=ConstellationSurfaceDetailsColorG, defaultValue=<rgba 0.974 0.224 0.0287 0>}
  n53 = LinearInterpolate(A=LinearInterpolate(A=@n42, B=Multiply(A=@n48, B=@n50.Result), Alpha=@n51.A), B=@n52.RGB, Alpha=@n51.G)
  n54 = VectorParameter() {group=Constellation, parameterName=ConstellationSurfaceDetailsColorR, defaultValue=<rgba 0.974 0.224 0.0287 0>}
  n55 = Power(Base=@n51.R, Exp=ScalarParameter() {defaultValue=1, group=Constellation, parameterName=ConstellationSurfaceDetailsColorR_MaskPow})
  n56 = VectorParameter() {group=Constellation, parameterName=ConstellationSurfaceDetailsColorB, defaultValue=<rgba 0.974 0.224 0.0287 0>}
  n57 = StaticSwitchParameter(True=LinearInterpolate(A=LinearInterpolate(A=@n53, B=@n54.RGB, Alpha=@n55), B=@n56.RGB, Alpha=@n51.B), False=@n42) {group=Constellation, parameterName=UseConstellationSun}
  n58 = ScalarParameter() {defaultValue=2.23, group=Constellation, parameterName=TexturedOutlinePow}
  n59 = OneMinus(None=@n39)
  n60 = Abs(None=DotProduct(A=@n38, B=Constant3Vector() {constant=<rgba 0 0 1 0>}))
  n61 = Subtract(
   A=@n59
   B=LinearInterpolate(
    A=ScalarParameter() {group=Constellation, parameterName=CS_EmissiveFalloff}
    B=ScalarParameter() {group=Constellation, parameterName=CS_EmissiveFalloffVert}
    Alpha=@n60
   )
  ) {constB=0.35}
  n62 = ScalarParameter() {defaultValue=2, group=Constellation, parameterName=CS_EmissvePower}
  n63 = Max(
   A=Multiply(A=Power(Base=@n2.G, Exp=@n58) {constExponent=2.23}, B=ScalarParameter() {group=Constellation, parameterName=TexturedOutlineIntensity})
   B=Clamp(None=Power(Base=@n61, Exp=@n62))
  )
  n64 = VectorParameter() {group=Constellation, parameterName=CS_EmissiveColor, defaultValue=<rgba 0.006 0.9647 0.956 1>}
  n65 = Add(A=@n2.B, B=OneMinus(None=ScalarParameter() {defaultValue=1, group=Constellation, parameterName=OutlineMask1Intensity}))
  n66 = ScalarParameter() {defaultValue=2.23, group=Constellation, parameterName=TexturedOutlinePow2}
  n67 = Subtract(
   A=@n59
   B=LinearInterpolate(
    A=ScalarParameter() {group=Constellation, parameterName=CS_EmissiveFalloff2}
    B=ScalarParameter() {group=Constellation, parameterName=CS_EmissiveFalloff2Vert}
    Alpha=@n60
   )
  ) {constB=0.35}
  n68 = ScalarParameter() {defaultValue=2, group=Constellation, parameterName=CS_EmissvePower2}
  n69 = Max(
   A=Multiply(A=Power(Base=@n2.G, Exp=@n66), B=ScalarParameter() {group=Constellation, parameterName=TexturedOutlineIntensity2})
   B=Clamp(None=Power(Base=@n67, Exp=@n68))
  )
  n70 = VectorParameter() {group=Constellation, parameterName=CS_EmissiveColor2, defaultValue=<rgba 0 0.1667 0.5573 1>}
  n71 = Add(A=@n2.B, B=OneMinus(None=ScalarParameter() {defaultValue=1, group=Constellation, parameterName=OutlineMask2Intensity}))
  n72 = Add(A=Multiply(A=Multiply(A=@n63, B=@n64.RGB), B=Saturate(None=@n65)), B=Multiply(A=Multiply(A=@n69, B=@n70.RGB), B=Saturate(None=@n71)))
  n73 = SetMaterialAttributes!(
   MaterialAttributes=@n32.MaterialAttributes
   BaseColor=LinearInterpolate(A=Multiply(A=@n32.BaseColor, B=@n33), B=@n34, Alpha=@n36)
   Metallic=LinearInterpolate(A=@n32.Metallic, B=@n34, Alpha=@n36)
   Specular=LinearInterpolate(A=Multiply(A=@n32.Specular, B=@n33), B=@n34, Alpha=@n36)
   Roughness=LinearInterpolate(A=@n32.Roughness, B=@n34, Alpha=@n36)
   EmissiveColor=LinearInterpolate(A=@n32.EmissiveColor, B=Add(A=@n57, B=@n72), Alpha=@n35)
  ) {attributeSetTypes [
    #1
    #2
    #3
    #4
    #5
   ]}
  FunctionOutput(None=@n73)
  n74 = ConstantBiasScale(None=ScreenPosition().ViewportUV) {bias=-0.5, scale=2}
  n75 = MaterialFunctionCall(Float2=ViewSize()) {materialFunction=&BreakOutFloat2Components}
  n76 = AppendVector(A=Constant() {r=1}, B=Divide(A=@n75.G, B=@n75.G))
  n77 = Constant2Vector() {g=-0.5, r=0.5}
  n78 = MaterialFunctionCall(World Position=ActorPositionWS()) {materialFunction=&TransformToClipSpace}
  n79 = Subtract(A=ConstantBiasScale(None=Multiply(A=@n74, B=@n76)), B=Multiply(A=@n76, B=Multiply(A=@n77, B=@n78.Clip Space XY)))
  n80 = TransformPosition(None=ActorPositionWS()) {transformSourceType=TRANSFORMPOSSOURCE_World, transformType=TRANSFORMPOSSOURCE_View}
  n81 = Multiply(
   A=Divide(A=ObjectRadius(), B=ComponentMask(None=@n80) {b=true})
   B=ScalarParameter() {defaultValue=1, group=Constellation, parameterName=CS_UVRadialScaleZero}
  )
  Multiply(
   A=VertexInterpolator(VS=MaterialFunctionCall(UVs=@n79, Texture Scale=@n81) {materialFunction=&ScaleUVsByCenter}.UVs).PS
   B=ComponentMask(None=VectorParameter() {group=Constellation, parameterName=ScreenCoordScale, defaultValue=<rgba 10 10 0 0>}.RGB) {g=true, r=true}
  )
  TextureCoordinate()
 ]}
