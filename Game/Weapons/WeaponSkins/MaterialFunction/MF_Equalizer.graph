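# Weapon skin material function that layers an animated, audio-reactive LED-style equalizer over incoming material attributes, blended by an Alpha input.
# Mirrored, cell-quantized UVs, optionally remapped per vertex by a scale/offset texture, feed custom nodes drawing five parabolic bands.
# Band heights take the maximum of timed attack-decay pulses and event-audio collection values.
# Masks, noise, gradient and amplitude-driven colors shape the gridded emissive, with cubemap reflection in the base color background.
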MaterialFunction /Game/Weapons/WeaponSkins/MaterialFunction/MF_Equalizer
graph {nodes [
  n1 = GetMaterialAttributes(None=FunctionInput() {inputType=FunctionInput_MaterialAttributes}) {attributeGetTypes [#1, #2]}
  n2 = Multiply(A=@n1.Normal, B=Constant3Vector() {constant=<rgba 1 1 0.5 0>})
  n3 = ScalarParameter() {defaultValue=50, group=Equalizer, parameterName=CellsY}
  n4 = StaticSwitch(
   True=TextureCoordinate() {coordinateIndex=1}
   False=TextureCoordinate()
   Value=StaticBoolParameter() {group=Equalizer, parameterName=UseTexCoord1}
  )
  n5 = Frac(None=@n4)
  n6 = TextureSampleParameter2D(UVs=@n5) {
   desc="R = Scale V, G = Offset V, B = swap U&&V, A= Sinewave Amplitude"
   group=Equalizer
   parameterName=TexCoordScaleOffsetV
   samplerType=SAMPLERTYPE_LinearColor
   texture=/Game/Weapons/WeaponSkins/Textures/T_DefaultScaleOffset
  }
  n7 = Multiply(A=@n6.R) {constB=17}
  n8 = Add(A=Multiply(A=@n6.G) {constB=-7.5}, B=Multiply(A=@n7, B=ComponentMask(None=@n5) {g=true}))
  n9 = VertexInterpolator(VS=MaterialFunctionCall(X=ComponentMask(None=@n5) {r=true}, Y=@n8, Z=@n7) {materialFunction=&MakeFloat3}.Result)
  n10 = StaticSwitchParameter(True=ComponentMask(None=@n9.PS) {b=true}, False=Constant() {r=1}) {defaultValue=true, group=Equalizer, parameterName=UseUVOffsetTexture}
  n11 = StaticSwitchParameter(True=ComponentMask(None=@n9.PS) {g=true, r=true}, False=@n4) {defaultValue=true, group=Equalizer, parameterName=UseUVOffsetTexture}
  n12 = Multiply(
   A=@n11
   B=MaterialFunctionCall(X=ScalarParameter() {defaultValue=50, group=Equalizer, parameterName=CellsX}, Y=Divide(A=@n3, B=@n10)) {materialFunction=&MakeFloat2}.Result
  )
  n13 = TextureSample(UVs=@n12) {samplerType=SAMPLERTYPE_Normal, texture=/Game/Characters/Player/Male/Medium/Bodies/M_MED_DJ_Remix/Textures/TechArt_Textures/T_Gem_N}
  n14 = TextureSample(UVs=@n12) {samplerType=SAMPLERTYPE_Masks, texture=/Game/Characters/Player/Male/Medium/Bodies/M_MED_DJ_Remix/Textures/TechArt_Textures/T_GridCell}
  n15 = OneMinus(None=@n14.RGB)
  n16 = Multiply(A=@n15, B=ScalarParameter() {defaultValue=1, group=Equalizer, parameterName=LinesRoughness})
  n17 = ScalarParameter() {defaultValue=0.1, group=Equalizer, parameterName=Roughness}
  n18 = TextureSampleParameter2D(UVs=Multiply(A=@n4, B=ScalarParameter() {defaultValue=1, group=Equalizer, parameterName=GradientTextureUV_Mul})) {
   group=Equalizer
   parameterName=GradientTexture
   texture=/Game/Characters/Player/Male/Medium/Bodies/M_MED_DJ_Remix/Textures/TechArt_Textures/T_DJ_RemixEmissiveGradient
  }
  n19 = StaticSwitchParameter(True=@n8, False=ComponentMask(None=@n4) {g=true}) {defaultValue=true, group=Equalizer, parameterName=UseUVOffsetTexture}
  n20 = LinearInterpolate(
   A=VectorParameter() {group=Equalizer, parameterName=ColorHigh, defaultValue=<rgba 1 0 0 1>}.RGB
   B=VectorParameter() {group=Equalizer, parameterName=ColorLow, defaultValue=<rgba 0 1 0 1>}.RGB
   Alpha=MaterialFunctionCall(
    Alpha=Saturate(None=@n19)
    Min=ScalarParameter() {group=Equalizer, parameterName=GradientMin}
    Max=ScalarParameter() {defaultValue=1, group=Equalizer, parameterName=GradientMax}
   ) {materialFunction=&SmoothStep}.Result
  )
  n21 = VectorParameter() {group=Equalizer, parameterName=ColorEmoteHigh, defaultValue=<rgba 1 0 0 0>}
  n22 = VectorParameter() {group=Equalizer, parameterName=ColorEmoteLow, defaultValue=<rgba 0 0 1 0>}
  n23 = Power(Base=Saturate(None=@n19), Exp=ScalarParameter() {defaultValue=3, group=Equalizer, parameterName=EmoteColorGradientPow})
  n24 = ScalarParameter() {defaultValue=1, group=Equalizer, parameterName=EmoteThreshold}
  n25 = CollectionParameter() {collection=&FortniteMaterialParameters, parameterName=EventAudio_Amplitude}
  n26 = Multiply(A=@n25) {constB=0.5}
  n27 = Power(Base=@n26)
  n28 = VertexInterpolator(VS=LinearInterpolate(
    A=StaticSwitchParameter(True=@n18.RGB, False=@n20) {group=Equalizer, parameterName=UseGradientTexture}
    B=LinearInterpolate(A=@n21.RGB, B=@n22.RGB, Alpha=@n23)
    Alpha=Power(Base=@n27, Exp=@n24)
   ))
  n29 = MaterialFunctionCall(
   X=ScalarParameter() {defaultValue=50, group=Equalizer, parameterName=EQ_CellsX}
   Y=Divide(A=ScalarParameter() {defaultValue=50, group=Equalizer, parameterName=EQ_CellsY}, B=@n10)
  ) {materialFunction=&MakeFloat2}
  n30 = Multiply(A=@n11, B=@n29.Result) {constB=5}
  n31 = MaterialFunctionCall(Float2=Divide(A=Floor(None=@n30), B=@n29.Result) {constB=5}) {materialFunction=&BreakOutFloat2Components}
  n32 = AppendVector(A=Multiply(A=@n31.R, B=ScalarParameter() {defaultValue=4, group=Equalizer, parameterName=EQ_HorizTiling}) {constB=0.5}, B=@n31.G)
  MaterialExpressionMaterialFunctionCall_19 = MaterialFunctionCall(Float2=@n32) {materialFunction=&BreakOutFloat2Components}
  n33 = Custom!(x=@MaterialExpressionMaterialFunctionCall_19.R) {
   code="return abs(2.*frac(x)-1.);"
   description=Mirror
   outputType=CMOT_Float1
   inputs [{inputName=x, input {expression=&self:MaterialExpressionMaterialFunctionCall_19}}]
  }
  n34 = AppendVector(A=@n33, B=@MaterialExpressionMaterialFunctionCall_19.G)
  MaterialExpressionFrac_6 = Frac(None=@n34)
  n35 = ScalarParameter() {defaultValue=0.7, group=Equalizer, parameterName=EQ_MaxIntensityBass}
  n36 = ScalarParameter() {defaultValue=0.4, group=Equalizer, parameterName=EQ_MaxIntensityTreble}
  n37 = ScalarParameter() {defaultValue=2, group=Equalizer, parameterName=EQ_SpeedBass}
  n38 = ScalarParameter() {defaultValue=4, group=Equalizer, parameterName=EQ_SpeedTreble}
  n39 = Time()
  MaterialExpressionFrac_0 = Frac(None=Multiply(A=MaterialFunctionCall(X=@n37, Y=@n38) {materialFunction=&MakeFloat2}.Result, B=@n39))
  n40 = Custom!(InputFunction=@MaterialExpressionFrac_0) {
   code="float2 result = (2 * InputFunction) -2;\r\nresult  = result  * result  * result ;\r\nreturn -1.181 * InputFunction * result;"
   description=FastAttackSlowDecay
   outputType=CMOT_Float2
   inputs [{inputName=InputFunction, input {expression=&self:MaterialExpressionFrac_0}}]
  }
  n41 = Multiply(A=MaterialFunctionCall(X=@n35, Y=@n36) {materialFunction=&MakeFloat2}.Result, B=@n40)
  n42 = CollectionParameter() {collection=&FortniteMaterialParameters, parameterName=EventAudio_Spectrum}
  n43 = MaterialFunctionCall(Float4=@n42) {materialFunction=&BreakOutFloat4Components}
  n44 = Multiply(A=@n43.B) {constB=0.25}
  n45 = Power(Base=Add(A=Multiply(A=@n43.R) {constB=0.5}, B=@n44))
  n46 = MaterialFunctionCall(X=@n45, Y=@n27) {materialFunction=&MakeFloat2}
  n47 = MaterialFunctionCall(Float2=VertexInterpolator(VS=Max(A=@n41, B=@n46.Result)).PS) {materialFunction=&BreakOutFloat2Components}
  MaterialExpressionReroute_27 = Reroute(None=@n47.R)
  MaterialExpressionReroute_26 = Reroute(None=@n47.G)
  n48 = ComponentMask(None=@n34) {r=true}
  MaterialExpressionLinearInterpolate_0 = LinearInterpolate(A=@MaterialExpressionReroute_27, B=@MaterialExpressionReroute_26, Alpha=@n48)
  MaterialExpressionLinearInterpolate_7 = LinearInterpolate(A=@MaterialExpressionReroute_27, B=@MaterialExpressionReroute_26, Alpha=@n48)
  MaterialExpressionLinearInterpolate_8 = LinearInterpolate(A=@MaterialExpressionReroute_27, B=@MaterialExpressionReroute_26, Alpha=@n48)
  MaterialExpressionScalarParameter_16 = ScalarParameter() {defaultValue=3, group=Equalizer, parameterName=EQ_BandsWidth}
  n49 = Custom!(
   uv=@MaterialExpressionFrac_6
   waveIntensity0=@MaterialExpressionReroute_27
   waveIntensity1=@MaterialExpressionLinearInterpolate_0
   waveIntensity2=@MaterialExpressionLinearInterpolate_7
   waveIntensity3=@MaterialExpressionLinearInterpolate_8
   waveIntensity4=@MaterialExpressionReroute_26
   waveWidth=@MaterialExpressionScalarParameter_16
  ) {
   code="half waveA = waveWidth * uv.x;\r\nwaveA = -(waveA * waveA) + waveIntensity0;\r\nhalf waveB = waveWidth * uv.x - waveWidth * 0.25;\r\nwaveB = -(waveB * waveB) + waveIntensity1;\r\nhalf waveC = waveWidth * uv.x - waveWidth * 0.5;\r\nwaveC = -(waveC * waveC) + waveIntensity2;\r\nhalf waveD = waveWidth * uv.x - waveWidth * 0.75;\r\nwaveD = -(waveD * waveD) + waveIntensity3;\r\nhalf waveE = waveWidth * uv.x - waveWidth;\r\nwaveE = -(waveE * waveE) + waveIntensity4;\r\nreturn max(waveA,max(waveB,max(waveC,max(waveD,waveE))));"
   description=CreateWave
   outputType=CMOT_Float1
   inputs [
    {inputName=uv, input {expression=&self:MaterialExpressionFrac_6}}
    {inputName=waveIntensity0, input {expression=&self:MaterialExpressionReroute_27}}
    {inputName=waveIntensity1, input {expression=&self:MaterialExpressionLinearInterpolate_0}}
    {inputName=waveIntensity2, input {expression=&self:MaterialExpressionLinearInterpolate_7}}
    {inputName=waveIntensity3, input {expression=&self:MaterialExpressionLinearInterpolate_8}}
    {inputName=waveIntensity4, input {expression=&self:MaterialExpressionReroute_26}}
    {inputName=waveWidth, input {expression=&self:MaterialExpressionScalarParameter_16}}
   ]
  }
  MaterialExpressionSaturate_7 = Saturate(None=@n49)
  MaterialExpressionScalarParameter_34 = ScalarParameter() {defaultValue=0.5, group=Equalizer, parameterName=EQ_VerticalOffset}
  n50 = Custom!(uv=@MaterialExpressionFrac_6, waveFunction=@MaterialExpressionSaturate_7, verticalOffset=@MaterialExpressionScalarParameter_34) {
   code="half result = uv.y- verticalOffset + waveFunction;    \r\nreturn (step(0.1, result));"
   description=DrawWave
   outputType=CMOT_Float1
   inputs [
    {inputName=uv, input {expression=&self:MaterialExpressionFrac_6}}
    {inputName=waveFunction, input {expression=&self:MaterialExpressionSaturate_7}}
    {inputName=verticalOffset, input {expression=&self:MaterialExpressionScalarParameter_34}}
   ]
  }
  n51 = ScalarParameter() {defaultValue=4, group=Equalizer, parameterName=NoiseTiles}
  n52 = Time()
  n53 = ScalarParameter() {defaultValue=1, group=Equalizer, parameterName=NoiseSpeed}
  n54 = TextureSampleParameter2D(UVs=Add(A=Multiply(A=@n32, B=@n51), B=Multiply(A=@n52, B=@n53))) {group=Equalizer, parameterName=NoiseTexture, samplerType=SAMPLERTYPE_Masks, texture=/Game/Global/Textures/Default/Noise/T_CellularNoise}
  n55 = ComponentMask(None=@n54.RGB) {g=true, r=true}
  n56 = AppendVector(
   A=ScalarParameter() {defaultValue=1, group=Equalizer, parameterName=StaticPartsNoiseMin}
   B=ScalarParameter() {defaultValue=1, group=Equalizer, parameterName=EQ_NoiseMin}
  )
  n57 = ScalarParameter() {defaultValue=1, group=Equalizer, parameterName=StaticPartsNoisePow}
  n58 = ScalarParameter() {defaultValue=1, group=Equalizer, parameterName=EQ_NoisePow}
  n59 = MaterialFunctionCall(Float2=Power(Base=Max(A=@n55, B=@n56), Exp=AppendVector(A=@n57, B=@n58))) {materialFunction=&BreakOutFloat2Components}
  n60 = TextureSampleParameter2D(UVs=TextureCoordinate()) {group=Equalizer, parameterName=Masks, samplerType=SAMPLERTYPE_Masks, texture=/Game/Global/Textures/Default/Blanks/AllRed}
  n61 = Multiply(A=Multiply(A=@n50, B=@n59.G), B=@n60.R) {constA=1}
  n62 = Saturate(None=Add(A=Multiply(A=@n60.R, B=@n50), B=@n60.G))
  n63 = Multiply(A=@n27, B=@n62)
  n64 = Multiply(A=Multiply(A=@n60.G, B=@n59.R), B=Max(A=ScalarParameter() {defaultValue=1, group=Equalizer, parameterName=StaticEmissiveIntensity}, B=@n63))
  n65 = LinearInterpolate(
   A=VectorParameter() {group=Equalizer, parameterName=BGColor}.RGB
   B=Multiply(A=Multiply(A=@n14.RGB, B=@n28.PS), B=Max(A=@n61, B=@n64))
   Alpha=@n62
  )
  n66 = ScalarParameter() {defaultValue=0.1, group=Equalizer, parameterName=Emissive_Intensity}
  n67 = Multiply(A=@n15, B=VectorParameter() {group=Equalizer, parameterName=LinesColor}.RGB) {constB=0.5}
  n68 = Add(A=Multiply(A=ScalarParameter() {defaultValue=0.1, group=Equalizer, parameterName=Color_Intensity}, B=@n65) {constB=0.1}, B=@n67)
  n69 = TextureSampleParameterCube!(UVs=ReflectionVectorWS(), Apply View MipBias) {
   group=Equalizer
   parameterName=BackgroundCubemap
   samplerType=SAMPLERTYPE_LinearColor
   texture=/Game/Characters/Player/Male/Medium/Bodies/M_MED_DJ_Remix/Textures/TechArt_Textures/T_CoatCubeMap_Tex
  }
  n70 = Multiply(A=Multiply(A=@n69.RGB, B=OneMinus(None=@n62)), B=ScalarParameter() {defaultValue=0.05, group=Equalizer, parameterName=CubeMapIntensity}) {constB=0.05}
  n71 = SetMaterialAttributes!(
   MaterialAttributes=@n1.MaterialAttributes
   BaseColor=LinearInterpolate(
    A=Add(A=@n68, B=@n70) {constB=0}
    B=LinearInterpolate(A=@n67, B=@n1.BaseColor, Alpha=@n14.RGB)
    Alpha=ScalarParameter() {group=Equalizer, parameterName=BaseColorPassThrough}
   )
   EmissiveColor=Multiply(A=@n65, B=Max(A=@n66, B=@n63))
   Normal=Multiply(A=Add(A=@n2, B=@n13.RGB), B=Constant3Vector() {constant=<rgba 1 1 0.675 0>})
   Roughness=Saturate(None=Add(A=@n16, B=@n17))
  ) {attributeSetTypes [
    #2
    #3
    #1
    #4
   ]}
  FunctionOutput(None=BlendMaterialAttributes(A=@n1.MaterialAttributes, B=@n71, Alpha=FunctionInput() {bUsePreviewValueAsDefault=true, inputName=Alpha, inputType=FunctionInput_Scalar}))
  n72 = CollectionParameter() {collection=&FortniteMaterialParameters, parameterName=Emote_AudioLow}
  n73 = CollectionParameter() {collection=&FortniteMaterialParameters, parameterName=Emote_AudioHigh}
  n74 = Add(A=Add(A=Multiply(A=@n43.R) {constB=0.25}, B=Multiply(A=@n43.G) {constB=0.25}), B=Add(A=@n44, B=Multiply(A=@n43.A) {constB=0.25}))
  Max(A=Add(A=@n72, B=@n73), B=Power(Base=Add(A=@n74, B=@n26))) {desc=O3}
  Max(A=@n73, B=@n27) {desc=O2}
  Max(A=@n72, B=@n45) {desc=O1}
  MaterialExpressionFrac_4 = Frac(None=Multiply(A=@n39, B=@n38) {desc="frac(time * speed)"})
  n75 = Custom!(InputFunction=@MaterialExpressionFrac_4) {
   code="return -(1.181 * InputFunction * pow(((2 * InputFunction) -2), 3));"
   description=FastAttackSlowDecay
   outputType=CMOT_Float1
   inputs [{inputName=InputFunction, input {expression=&self:MaterialExpressionFrac_4}}]
  }
  Reroute(None=Max(A=Multiply(A=@n36, B=@n75), B=@n27))
  MaterialExpressionFrac_3 = Frac(None=Multiply(A=@n37, B=@n39))
  n76 = Custom(InputFunction=@MaterialExpressionFrac_3) {
   code="return -(1.181 * InputFunction * pow(((2 * InputFunction) -2), 3));"
   description=FastAttackSlowDecay
   outputType=CMOT_Float1
   inputs [{inputName=InputFunction, input {expression=&self:MaterialExpressionFrac_3}}]
  }
  Reroute(None=Max(A=Multiply(A=@n76, B=@n35), B=@n45))
 ]}
