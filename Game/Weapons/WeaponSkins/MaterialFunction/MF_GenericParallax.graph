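# Material function for weapon skins that adds a generic tiling parallax layer to the base color of an incoming material.
# A tiled diffuse texture, sampled through BumpOffset with a constant height parameter, gives the illusion of depth.
# A Fresnel term on the transformed material normal, scaled by a global opacity, blends it into BaseColor.
# An Alpha input then blends the modified attributes with the original material.
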MaterialFunction /Game/Weapons/WeaponSkins/MaterialFunction/MF_GenericParallax
graph {nodes [
  n1 = FunctionInput() {inputName=Material, inputType=FunctionInput_MaterialAttributes}
  n2 = GetMaterialAttributes(None=@n1) {attributeGetTypes [#1, #2]}
  n3 = Multiply(A=TextureCoordinate(), B=ScalarParameter() {defaultValue=1, group=Parallax, parameterName=Parallax Tiling})
  n4 = ScalarParameter() {defaultValue=-1, group=Parallax, parameterName=Parallax Height}
  n5 = Fresnel(ExponentIn=ScalarParameter() {parameterName=Parallax Opacity Fresnel}, Normal=Transform(None=@n2.Normal)) {baseReflectFraction=0}
  n6 = ScalarParameter() {group=Parallax, parameterName=Parallax Global Opacity}
  n7 = LinearInterpolate(
   A=@n2.BaseColor
   B=TextureSampleParameter2D(UVs=BumpOffset(Coordinate=@n3, Height=@n4)) {group=Parallax, parameterName=Tiling Parallax Diffuse, texture=&T_BlacksRGB}.RGB
   Alpha=Saturate(None=Multiply(A=@n5, B=@n6))
  )
  FunctionOutput(None=BlendMaterialAttributes(
    A=@n1
    B=SetMaterialAttributes!(MaterialAttributes=@n2.MaterialAttributes, BaseColor=@n7) {attributeSetTypes [#1]}
    Alpha=FunctionInput() {inputName=Alpha}
   ))
 ]}
