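# Material function for weapon skins that layers an animated growing-noise surface over incoming material attributes, with its own colors, metallic, specular, roughness and blended streak normals.
# A packed streak texture supplies distance field, random offset and growth gradient, with stepped sine time driving a sphere mask that grows the streaks.
# A crevice mask and panning noise shape and pulse the edges, optionally modulated by another mask channel.
# Only Epic quality adds the streak normals.
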MaterialFunction /Game/Weapons/WeaponSkins/MaterialFunction/MF_GrowingNoise
graph {nodes [
  n1 = GetMaterialAttributes(None=FunctionInput() {inputName=MA, inputType=FunctionInput_MaterialAttributes}) {attributeGetTypes [#1]}
  n2 = TextureCoordinate()
  n3 = Multiply(
   A=ScalarParameter() {defaultValue=0.005, group=GrowingNoise, parameterName=GrowingNoiseUVDistortionScale}
   B=Subtract(A=NamedRerouteUsage()) {constB=0.5}
  )
  n4 = Add(
   A=Multiply(A=Add(A=@n2, B=@n3), B=ScalarParameter() {defaultValue=1.9, group=GrowingNoise, parameterName=GrowingNoiseUVScale})
   B=AppendVector(
    A=ScalarParameter() {group=GrowingNoise, parameterName=GrowingNoiseBiasU}
    B=ScalarParameter() {group=GrowingNoise, parameterName=GrowingNoiseBiasV}
   )
  )
  n5 = TextureSampleParameter2D(UVs=@n4) {
   group=GrowingNoise
   parameterName=GrowingNoiseNormalA
   samplerType=SAMPLERTYPE_Normal
   texture=/Game/Characters/Player/Male/Medium/Bodies/M_MED_Drift_Horror/Textures/T_M_MED_Drift_Horror_StreakNormals
  }
  n6 = TextureSampleParameter2D() {
   group=GrowingNoise
   parameterName=GrowingNoiseNormalB
   samplerType=SAMPLERTYPE_Normal
   texture=/Game/Characters/Player/Male/Medium/Bodies/M_MED_Drift_Horror/Textures/T_M_MED_Drift_Horror_StreakNormalsPainted
  }
  n7 = Multiply(
   A=MaterialFunctionCall(BaseNormal=@n5.RGB, AdditionalNormal=@n6.RGB) {materialFunction=&BlendAngleCorrectedNormals}.Result
   B=VectorParameter() {group=GrowingNoise, parameterName=GrowingNoiseNormalStrength, defaultValue=<rgba 0.25 0.25 1 0>}.RGB
  )
  n8 = SetMaterialAttributes!(
   MaterialAttributes=@n1.MaterialAttributes
   BaseColor=LinearInterpolate(
    A=VectorParameter() {group=GrowingNoise, parameterName=GrowingNoiseColorA, defaultValue=<rgba 0.015 0.0122 0.0189 0>}.RGB
    B=VectorParameter() {group=GrowingNoise, parameterName=GrowingNoiseColorB, defaultValue=<rgba 0.0075 0.004 0.01 0>}.RGB
    Alpha=NamedRerouteUsage()
   )
   Metallic=ScalarParameter() {defaultValue=0.2, group=GrowingNoise, parameterName=GrowingNoiseMetallic}
   Specular=ScalarParameter() {defaultValue=0.05, group=GrowingNoise, parameterName=GrowingNoiseSpecular}
   Roughness=ScalarParameter() {defaultValue=0.45, group=GrowingNoise, parameterName=GrowingNoiseRoughness}
   Normal=QualitySwitch(
    Default=@n1.Normal
    Epic=MaterialFunctionCall(BaseNormal=@n1.Normal, AdditionalNormal=@n7) {materialFunction=&BlendAngleCorrectedNormals}.Result
   )
  ) {attributeSetTypes [
    #2
    #3
    #4
    #5
    #1
   ]}
  FunctionOutput(None=BlendMaterialAttributes(A=@n1.MaterialAttributes, B=@n8, Alpha=NamedRerouteUsage()))
  n9 = TextureCoordinate()
  n10 = ScalarParameter() {defaultValue=8, parameterName=GrowingNoiseDistortionUVScale}
  n11 = ScalarParameter() {defaultValue=0.1333, group=GrowingNoise, parameterName=GrowingNoiseDistortionPanU}
  n12 = ScalarParameter() {defaultValue=0.1667, group=GrowingNoise, parameterName=GrowingNoiseDistortionPanV}
  n13 = TextureSample(UVs=Panner(Coordinate=Multiply(A=@n9, B=@n10), Speed=AppendVector(A=@n11, B=@n12)) {bFractionalPart=true, speedY=1}) {samplerType=SAMPLERTYPE_LinearColor, texture=&T_NoisePack_LinearColor}
  NamedRerouteDeclaration(None=AppendVector(A=@n13.R, B=@n13.R)) {name=UVDistortion, nodeColor=<rgba 0 1 0.5781 1>}
  n14 = ScalarParameter() {defaultValue=10, group=GrowingNoise, parameterName=SkinFXMask_ModulateLERP_Scale}
  n15 = StaticComponentMaskParameter(None=NamedRerouteUsage()) {defaultB=true, group=GrowingNoise, parameterName=SkinFXMask_ModulateLERP_Channel}
  n16 = OneMinus(None=Saturate(None=Multiply(A=@n14, B=@n15)))
  n17 = TextureSampleParameter2D(UVs=@n4) {
   group=GrowingNoise
   parameterName=NoisePack
   samplerType=SAMPLERTYPE_LinearColor
   texture=/Game/Characters/Player/Male/Medium/Bodies/M_MED_Drift_Horror/Textures/T_M_MED_Drift_Horror_Streaks
  }
  n18 = StaticComponentMaskParameter(None=@n17.RGBA) {defaultG=true, group=GrowingNoise, parameterName=NoisePackDistanceFieldChannel}
  n19 = StaticComponentMaskParameter(None=@n17.RGBA) {defaultR=true, group=GrowingNoise, parameterName=NoisePackRandomValueChannel}
  n20 = ScalarParameter() {defaultValue=1.5, group=GrowingNoise, parameterName=GrowingNoiseRandomTimeOffsetScale}
  n21 = ScalarParameter() {defaultValue=8, group=GrowingNoise, parameterName=GrowingNoiseTimeDivisor}
  n22 = ScalarParameter() {defaultValue=90, group=GrowingNoise, parameterName=GrowingNoiseTimeSteps}
  n23 = Round(None=Multiply(
    A=MaterialFunctionCall!(Time, Divisor=@n21, Mod) {materialFunction=/Game/Effects/Fort_Effects/Materials/_Functions/MF_ModuloTime}.Unclamped
    B=@n22
   ))
  n24 = Sine(None=Add(A=Multiply(A=@n19, B=@n20), B=Divide(A=@n23, B=@n22)))
  n25 = Multiply(A=SphereMask(
    A=StaticComponentMaskParameter(None=@n17.RGBA) {defaultB=true, group=GrowingNoise, parameterName=NoisePackGrowthGradientChannel}
    B=Constant() {r=1}
    Radius=Multiply(A=@n24) {constB=1.15}
    Hardness=Constant()
   )) {constB=5}
  n26 = StaticSwitchParameter(True=Saturate(None=Subtract(A=@n18, B=@n25) {constB=0}), False=Constant()) {defaultValue=true, group=GrowingNoise, parameterName=UseDynamicNoise}
  n27 = TextureObject() {samplerType=SAMPLERTYPE_Masks, texture=/Game/Characters/Player/Male/Medium/Bodies/M_MED_Drift_Horror/Textures/T_M_MED_Drift_Horror_Body_FX}
  n28 = FunctionInput(Preview=@n27) {inputName=Mask, inputType=FunctionInput_Texture2D, sortPriority=2}
  n29 = StaticComponentMaskParameter(None=NamedRerouteDeclaration(None=TextureSample(Tex=@n28).RGBA) {name=SkinFXMask_RGBA, nodeColor=<rgba 0.0547 1 0 1>}) {defaultB=true, group=GrowingNoise, parameterName=SkinFXMask_Crevice}
  n30 = Multiply(A=Subtract(A=@n13.R) {constB=0.5}, B=ScalarParameter() {defaultValue=1.8, group=GrowingNoise, parameterName=GrowingNoiseEdgePulsingScale})
  n31 = Add(A=@n26, B=Multiply(A=Add(A=@n26, B=@n29), B=Add(A=@n29, B=@n30)))
  n32 = Multiply(A=@n31) {constB=10}
  n33 = ScalarParameter() {defaultValue=0.95, group=GrowingNoise, parameterName=GrowingNoiseInnerValue}
  n34 = ScalarParameter() {defaultValue=1, group=GrowingNoise, parameterName=GrowingNoiseOuterValue}
  n35 = SphereMask(A=@n31, B=Constant() {r=0.15}, Radius=Constant() {r=0.7}, Hardness=Constant() {r=0.1})
  n36 = Saturate(None=Multiply(A=Saturate(None=@n32), B=LinearInterpolate(A=@n33, B=@n34, Alpha=@n35)))
  NamedRerouteDeclaration(None=StaticSwitchParameter(True=Multiply(A=@n16, B=@n36), False=@n36) {group=GrowingNoise, parameterName=SkinFXMaskModulatesLERP}) {name=FinalLerp, nodeColor=<rgba 1 0.9844 0 1>}
  NamedRerouteDeclaration(None=@n13.G) {name=Noise, nodeColor=<rgba 0 0.0625 1 1>}
 ]}
