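# MF_Haunt is a weapon-skin material function that overrides input material attributes with an animated, glowing Haunt smoke-and-stars effect.
# A camera-facing virtual plane at the object pivot supplies swirled, panning UVs for smoke, twinkling stars and channel-offset nebula.
# These join mask-driven painted and eye glows in emissive, plus a distance-faded rim.
# Vertex-red-weighted noise drives vertex offset, while core-glow and lightning branches are left unconnected.
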MaterialFunction /Game/Weapons/WeaponSkins/MaterialFunction/MF_Haunt
graph {nodes [
  n1 = MaterialFunctionCall(Float3=VertexNormalWS()) {materialFunction=&BreakOutFloat3Components}
  Abs(None=@n1.B)
  n2 = FunctionInput(Preview=TextureSample() {texture=/Game/Characters/Player/Male/Medium/Bodies/M_MED_Haunt/Textures/T_Haunt_Body_FX}.RGB) {bUsePreviewValueAsDefault=true, inputName=Mask}
  n3 = GetMaterialAttributes(None=FunctionInput() {inputName=Material, inputType=FunctionInput_MaterialAttributes}) {attributeGetTypes [#1]}
  n4 = Transform(None=@n3.Normal)
  n5 = DotProduct(A=CameraVectorWS(), B=@n4)
  n6 = ScalarParameter() {parameterName=FresnelGlowFalloff}
  n7 = Multiply(
   A=Power(Base=ComponentMask(None=@n2) {r=true}, Exp=ScalarParameter() {defaultValue=1.15, group=ViperZero, parameterName=PaintedGlowPower}) {constExponent=1.25}
   B=Clamp(None=Subtract(A=@n5, B=@n6))
  )
  n8 = Clamp(None=@n7)
  n9 = VectorParameter() {group=ViperZero, parameterName=InnerGlowColor, defaultValue=<rgba 0 0.5703 1 1>}
  n10 = Multiply(A=Multiply(A=Power(Base=@n8), B=@n9.RGB), B=ScalarParameter() {defaultValue=1, group=ViperZero, parameterName=FaceGlowStrength})
  n11 = CameraVectorWS()
  n12 = CrossProduct(A=@n11, B=VectorParameter() {parameterName=Axis, defaultValue=<rgba 0 0 1 1>}.RGB)
  n13 = MaterialFunctionCall() {materialFunction=&ObjectPivotPoint}
  n14 = ComponentMask(None=@n13.Object Pivot Location) {g=true, r=true}
  n15 = ScalarParameter() {defaultValue=96, group=ViperZero, parameterName=PivotZAdjustment}
  n16 = Add(A=ComponentMask(None=@n13.Object Pivot Location) {b=true}, B=@n15) {constB=48}
  n17 = ScalarParameter() {defaultValue=100, group=ViperZero, parameterName=UVRadialScaleZero}
  n18 = Divide(A=ObjectRadius(), B=ScalarParameter() {defaultValue=97, group=Stars, parameterName=ActorRadius, sortPriority=0})
  n19 = Subtract(
   A=MaterialFunctionCall!(
    Plane Normal Axis=@n11
    Plane X Axis=Normalize(VectorInput=@n12)
    Plane Center=AppendVector(A=@n14, B=@n16)
    UV Scale=Multiply(A=@n17, B=@n18)
   ) {materialFunction=/Game/Athena/Environments/Nexus/Materials/VirtualPlaneCoordinatesAbs}.UVs
   B=ComponentMask(None=VectorParameter() {group=ViperZero, parameterName=CorePosition, defaultValue=<rgba 0.5 0.5 0 1>}.RGB) {g=true, r=true}
  )
  n20 = ComponentMask(None=Constant3Vector() {constant=<rgba 0.5 0.5 0 0>}) {g=true, r=true}
  n21 = Time()
  n22 = Add(A=@n21) {constB=0.5}
  n23 = MaterialFunctionCall(
   Float=@n21
   Vec2=Multiply(A=AppendVector(A=@n21, B=@n22), B=ScalarParameter() {defaultValue=0.02, group=ViperZero, parameterName=CoreWobbleSpeed})
  ) {materialFunction=&MF_Hash}
  n24 = ComponentMask(None=@n23.Vec2) {g=true, r=true}
  n25 = ScalarParameter() {defaultValue=0.02, group=ViperZero, parameterName=CoreMovement_Amt}
  n26 = ComponentMask(None=@n23.Float) {r=true}
  n27 = SphereMask(
   A=@n19
   B=Add(A=@n20, B=Multiply(A=@n24, B=@n25))
   Radius=Add(A=ScalarParameter() {defaultValue=0.5, group=ViperZero, parameterName=InnerGlowRadius}, B=Multiply(A=@n26) {constB=0.05})
   Hardness=ScalarParameter() {group=ViperZero, parameterName=InnerGlowHardness}
  )
  n28 = Power(Base=Multiply(A=@n27) {constB=0.5}) {constExponent=4}
  n29 = SphereMask(A=@n19, B=@n20, Radius=ScalarParameter() {defaultValue=0.75, group=ViperZero, parameterName=SwirlRadius}, Hardness=Constant())
  n30 = Rotator(Coordinate=@n19, Time=Multiply(A=@n29, B=ScalarParameter() {defaultValue=6, group=ViperZero, parameterName=CoreSwirlStrength}))
  n31 = ScalarParameter() {defaultValue=0.5, group=ViperZero, parameterName=SmokeTiling}
  n32 = VectorParameter() {group=ViperZero, parameterName=SmokePanSpeedandDirection, defaultValue=<rgba 0 -0.8 0 1>}
  n33 = Multiply(A=@n30, B=ScalarParameter() {defaultValue=0.5, group=ViperZero, parameterName=SmokeSwirlTiling})
  n34 = ComponentMask(None=VectorParameter() {group=ViperZero, parameterName=SwirlPanSpeedandDirection, defaultValue=<rgba 0 -0.5 0 1>}.RGB) {g=true, r=true}
  n35 = Multiply(A=TextureSample(UVs=Panner(Coordinate=@n33, Speed=@n34) {speedY=-0.5}) {samplerType=SAMPLERTYPE_LinearColor, texture=/Game/Global/Textures/Default/Flowmaps/T_SwirlFlow_02}.RGB) {constB=2}
  n36 = ComponentMask(None=Multiply(A=Subtract(A=@n35), B=ScalarParameter() {defaultValue=0.1, group=ViperZero, parameterName=DistortionStrength})) {g=true, r=true}
  n37 = Add(A=Panner(Coordinate=Multiply(A=@n30, B=@n31), Speed=ComponentMask(None=@n32.RGB) {g=true, r=true}) {speedY=-0.8}, B=@n36)
  n38 = ComponentMask(None=TextureSample(UVs=@n37) {texture=/Game/Global/Textures/Default/Tileable_Patterns/T_Thick_Smoke_05}.RGB) {r=true}
  n39 = Multiply(A=@n38, B=ScalarParameter() {defaultValue=0.04, group=ViperZero, parameterName=GlowSmokiness_NoiseStrength})
  n40 = Multiply(A=Clamp(None=Subtract(A=@n28, B=@n39)), B=@n9.RGB)
  n41 = ScalarParameter() {defaultValue=50, group=ViperZero, parameterName=InnerGlowStrength}
  n42 = VectorParameter() {group=ViperZero, parameterName=FalloffGlowColor, defaultValue=<rgba 0 0.1574 0.276 1>}
  n43 = Power(Base=SphereMask(
    A=@n19
    B=@n20
    Radius=ScalarParameter() {defaultValue=4, group=ViperZero, parameterName=GlowFalloffRadius}
    Hardness=ScalarParameter() {group=ViperZero, parameterName=GlowFalloffHardness}
   ))
  n44 = Multiply(A=@n38, B=ScalarParameter() {defaultValue=0.5, group=ViperZero, parameterName=GlowwFalloff_NoiseStrength})
  n45 = Multiply(A=Clamp(None=Subtract(A=@n43, B=@n44)), B=ScalarParameter() {defaultValue=1, group=ViperZero, parameterName=FalloffGlowStrength}) {constB=0.2}
  Add(
   A=Add(A=@n10, B=Multiply(A=@n40, B=@n41))
   B=LinearInterpolate(A=VectorParameter() {group=ViperZero, parameterName=GeneralGlowColor, defaultValue=<rgba 0 0.0356 0.0625 1>}.RGB, B=@n42.RGB, Alpha=@n45)
  )
  n46 = ComponentMask(None=VertexInterpolator(VS=PreSkinnedPosition()).PS) {b=true, r=true}
  n47 = ScalarParameter() {defaultValue=32, group=ViperZero, parameterName=SmokeNormalTiling}
  n48 = VectorParameter() {group=ViperZero, parameterName=NormalSmokePanSpeedandDirection, defaultValue=<rgba 0 -0.8 0 1>}
  n49 = Panner(
   Coordinate=Multiply(A=@n46, B=ScalarParameter() {defaultValue=0.5, group=ViperZero, parameterName=SmokeNormalSwirlTiling})
   Speed=ComponentMask(None=VectorParameter() {group=ViperZero, parameterName=SwirlPanSpeedandDirection, defaultValue=<rgba 0 -0.5 0 1>}.RGB) {g=true, r=true}
  ) {speedY=-0.5}
  n50 = Subtract(A=Multiply(A=TextureSample(UVs=@n49) {samplerType=SAMPLERTYPE_LinearColor, texture=&T_SwirlFlow_02}.RGB) {constB=2})
  n51 = ScalarParameter() {defaultValue=0.1, group=ViperZero, parameterName=DistortionStrength}
  Add(
   A=Panner(Coordinate=Multiply(A=@n46, B=@n47), Speed=ComponentMask(None=@n48.RGB) {g=true, r=true}) {speedY=-0.8}
   B=ComponentMask(None=Multiply(A=@n50, B=@n51)) {g=true, r=true}
  )
  n52 = Multiply(A=Subtract(A=@n20, B=@n19), B=ScalarParameter() {defaultValue=4, group=ViperZero, parameterName=Stars_Tiling})
  n53 = Multiply(A=@n29, B=ScalarParameter() {defaultValue=3, group=ViperZero, parameterName=StarSwirlStrength})
  n54 = VectorParameter() {group=ViperZero, parameterName=StarsPanSpeedandDirection, defaultValue=<rgba 0 0.5 0 1>}
  n55 = TextureSample(UVs=Panner(Coordinate=Rotator(Coordinate=@n52, Time=@n53), Speed=ComponentMask(None=@n54.RGB) {g=true, r=true}) {speedY=0.5}) {texture=/Game/Characters/Player/Male/Medium/Bodies/M_MED_Haunt/Textures/T_StarPattern}
  n56 = Time()
  n57 = Ceil(None=Multiply(A=@n52) {constB=10})
  n58 = ComponentMask(None=ConstantBiasScale(None=MaterialFunctionCall(Seed=@n56, Vec2=@n57) {materialFunction=&MF_Hash}.Vec2)) {r=true}
  n59 = Clamp(None=Multiply(A=@n55.R, B=@n58))
  n60 = VectorParameter() {group=ViperZero, parameterName=StarColor, defaultValue=<rgba 1 1 1 1>}
  n61 = Add(A=@n10, B=Multiply(A=Multiply(A=@n59, B=@n60.RGB), B=ScalarParameter() {defaultValue=10, group=ViperZero, parameterName=StarGlowStrength}))
  n62 = Clamp(None=@n38)
  n63 = ComponentMask(None=VectorParameter() {group=ViperZero, parameterName=LightningColor, defaultValue=<rgba 0 0.4603 1 1>}.RGB) {b=true, g=true, r=true}
  n64 = ScalarParameter() {defaultValue=1.5, group=ViperZero, parameterName=ThunderTiling}
  n65 = VectorParameter() {group=ViperZero, parameterName=ThunderPanSpeedandDirection, defaultValue=<rgba 0 -0.8 0 1>}
  n66 = Add(A=Panner(Coordinate=Multiply(A=@n30, B=@n64), Speed=ComponentMask(None=@n65.RGB) {g=true, r=true}) {speedY=-0.8}, B=@n36)
  n67 = Multiply(A=@n64, B=ScalarParameter() {defaultValue=4, group=ViperZero, parameterName=ThunderDistortNoise})
  n68 = VectorParameter() {group=ViperZero, parameterName=ThunderDistortPanSpeedandDirection, defaultValue=<rgba 0.5 4 0 1>}
  n69 = TextureSample(UVs=Panner(Coordinate=Multiply(A=@n66, B=@n67), Speed=ComponentMask(None=@n68.RGB) {g=true, r=true}) {speedX=0.5, speedY=4}) {texture=/Game/Global/Textures/Default/Noise/T_Noise_Cloud}
  n70 = Subtract(A=Multiply(A=ComponentMask(None=@n69.RGB) {g=true, r=true}) {constB=2})
  n71 = ScalarParameter() {defaultValue=0.02, group=ViperZero, parameterName=ThunderMAcroDistortion strength}
  n72 = TextureSample(UVs=Add(A=@n66, B=Multiply(A=@n70, B=@n71))) {samplerType=SAMPLERTYPE_Masks, texture=/Game/Weapons/WeaponSkins/Wraps/Textures/PlasmaCharge/PlasmaCharge_M}
  n73 = ScalarParameter() {defaultValue=2, group=ViperZero, parameterName=ThunderBrightness}
  n74 = ScalarParameter() {defaultValue=2, group=ViperZero, parameterName=CoreThunderBrightnes}
  n75 = Power(Base=Multiply(A=ComponentMask(None=@n72.RGB) {g=true}, B=LinearInterpolate(A=@n73, B=@n74, Alpha=@n27)) {constB=10})
  Add(A=@n61, B=Multiply(A=Add(A=@n27, B=@n62), B=Multiply(A=@n63, B=@n75)))
  n76 = PreSkinnedPosition()
  n77 = PreSkinnedLocalBounds()
  Clamp(None=Divide(A=ComponentMask(None=@n76) {b=true}, B=ComponentMask(None=@n77.Max) {b=true})) {minDefault=1}
  n78 = TextureSample(UVs=@n37) {samplerType=SAMPLERTYPE_Normal, texture=/Game/Characters/Player/Male/Medium/Bodies/M_MED_Haunt/Textures/T_Thick_Smoke_05_N}
  n79 = Transform(None=@n78.RGB) {transformSourceType=TRANSFORMSOURCE_Local, transformType=TRANSFORM_Tangent}
  ComponentMask(None=@n79) {r=true}
  n80 = ScalarParameter() {group=ViperZero, parameterName=Roughness}
  n81 = ScalarParameter() {defaultValue=0.1, group=ViperZero, parameterName=LowEndRoughness}
  n82 = Clamp(None=ComponentMask(None=@n78.RGB) {g=true})
  n83 = ComponentMask(None=@n2) {b=true}
  n84 = Panner(Coordinate=Multiply(A=@n19, B=ScalarParameter() {defaultValue=24, group=ViperZero, parameterName=EyeSmokeTiling})) {speedY=-0.25}
  n85 = Subtract(
   A=Power(Base=Multiply(A=@n83))
   B=Multiply(
    A=TextureSample(UVs=@n84) {texture=&T_NoisePack_01}.R
    B=ScalarParameter() {defaultValue=0.5, group=ViperZero, parameterName=EyeSmokeStrength}
   )
  )
  n86 = Multiply(
   A=Multiply(A=Clamp(None=@n85), B=VectorParameter() {group=ViperZero, parameterName=EyeGlowColor, defaultValue=<rgba 0 0.5687 1 1>}.RGB)
   B=ScalarParameter() {defaultValue=5, group=ViperZero, parameterName=EyeGlowStrength}
  )
  n87 = Add(A=Multiply(A=Power(Base=@n86), B=ScalarParameter() {group=ViperZero, parameterName=bIsHead}), B=@n61)
  n88 = ScalarParameter() {defaultValue=1, group=ViperZero, parameterName=NebulaTiling}
  n89 = VectorParameter() {group=ViperZero, parameterName=NebulaPanSpeed, defaultValue=<rgba 0 -0.2 0 1>}
  n90 = Add(A=Panner(Coordinate=Multiply(A=@n30, B=@n88), Speed=ComponentMask(None=@n89.RGB) {g=true, r=true}) {speedY=-0.8}, B=@n36)
  n91 = ComponentMask(None=@n90) {r=true}
  n92 = Add(A=@n91) {constB=0.025}
  n93 = ComponentMask(None=@n90) {g=true}
  n94 = Add(A=@n91) {constB=0.05}
  n95 = MaterialFunctionCall(
   X=TextureSample(UVs=@n90) {texture=/Game/Global/Textures/Default/Caustics/T_Turbulence_Seamless}.R
   Y=TextureSample(UVs=AppendVector(A=@n92, B=@n93)) {texture=&T_Turbulence_Seamless}.R
   Z=TextureSample(UVs=AppendVector(A=@n94, B=@n93)) {texture=&T_Turbulence_Seamless}.R
  ) {materialFunction=&MakeFloat3}
  n96 = Add(
   A=@n87
   B=Multiply(A=Multiply(A=@n38, B=@n95.Result), B=ScalarParameter() {defaultValue=1, group=ViperZero, parameterName=NebulaGlowStrength})
  )
  n97 = Multiply(A=Multiply(A=@n96, B=@n82), B=ScalarParameter() {defaultValue=1.1, group=ViperZero, parameterName=FakeSPecSterngth})
  n98 = SphereMask(A=@n19, B=@n20, Radius=ScalarParameter() {group=ViperZero, parameterName=MasterMaskRadius}, Hardness=Constant())
  n99 = Multiply(A=@n38, B=ScalarParameter() {defaultValue=1, group=ViperZero, parameterName=MasterMaskSwirlingCLoudStrength})
  n100 = Multiply(A=@n96, B=Clamp(None=Subtract(A=@n98, B=@n99)))
  n101 = Constant3Vector()
  n102 = DotProduct(A=CameraVectorWS(), B=@n4)
  n103 = Multiply(A=Subtract(A=OneMinus(None=@n102), B=ScalarParameter() {defaultValue=0.35, group=ViperZero, parameterName=RimSpread})) {constB=2}
  n104 = Divide(A=ScalarParameter() {defaultValue=97, group=Stars, parameterName=ActorRadius, sortPriority=0}, B=ObjectRadius())
  n105 = Multiply(
   A=Multiply(A=Clamp(None=@n103), B=VectorParameter() {group=ViperZero, parameterName=DistanceRimColor, defaultValue=<rgba 1 1 1 1>}.RGB)
   B=Multiply(A=ScalarParameter() {defaultValue=10, group=ViperZero, parameterName=DistanceRimStrength}, B=Clamp(None=@n104))
  )
  n106 = Distance(A=CameraPositionWS(), B=ObjectPositionWS())
  n107 = ScalarParameter() {defaultValue=1500, group=ViperZero, parameterName=DistanceRimOfset}
  n108 = Clamp(None=Divide(A=Subtract(A=@n106, B=@n107), B=ScalarParameter() {defaultValue=10000, group=ViperZero, parameterName=DistanceRimBias}))
  n109 = MaterialFunctionCall(X=@n1.R, Y=@n1.R, Z=@n1.R) {materialFunction=&MakeFloat3}
  n110 = VectorParameter() {group=ViperZero, parameterName=DirectionalStrength, defaultValue=<rgba 1 1 1 1>}
  n111 = Multiply(A=ComponentMask(None=@n76) {b=true, r=true}, B=ScalarParameter() {defaultValue=1, group=ViperZero, parameterName=Sports_Tiling})
  n112 = Clamp(None=TextureSample(UVs=Panner(Coordinate=@n111) {speedY=0.1}) {texture=/Game/Global/Textures/Default/Noise/3xPerlin_M}.R)
  n113 = Multiply(A=ComponentMask(None=VertexColor().R) {r=true}, B=ScalarParameter() {defaultValue=12, group=ViperZero, parameterName=Displacement_Amt})
  n114 = ScalarParameter() {defaultValue=1, group=ViperZero, parameterName=MaxSpecular}
  n115 = ScalarParameter() {defaultValue=0.5, parameterName=LowEndSpecular}
  n116 = SetMaterialAttributes!(
   MaterialAttributes=@n3.MaterialAttributes
   BaseColor=@n10
   Metallic=ScalarParameter() {group=ViperZero, parameterName=Metallness}
   Roughness=LinearInterpolate(A=Constant() {r=0.25}, B=QualitySwitch(Default=@n80, Low=@n81, High=@n80, Medium=@n81), Alpha=@n82)
   EmissiveColor=Add(A=Add(A=@n97, B=@n100), B=LinearInterpolate(A=@n101, B=@n105, Alpha=@n108))
   WorldPositionOffset=Multiply(A=Multiply(A=@n109.Result, B=@n110.RGB), B=Multiply(A=@n112, B=@n113))
   Specular=QualitySwitch(Default=@n114, Low=@n115, High=@n114, Medium=@n115)
   OpacityMask
  ) {attributeSetTypes [
    #2
    #3
    #4
    #5
    #6
    #7
    #8
   ]}
  FunctionOutput(None=@n116)
  LinearInterpolate()
  n117 = Fresnel() {exponent=2}
  LinearInterpolate(A=@n114, B=ScalarParameter() {group=ViperZero, parameterName=MinSpecular}, Alpha=Clamp(None=Power(Base=@n117)))
  MaterialFunctionCall(BaseNormal=@n3.Normal, AdditionalNormal=@n79) {materialFunction=&BlendAngleCorrectedNormals}
  MaterialFunctionCall() {materialFunction=&ObjectPivotPoint}
  n118 = Constant()
  n119 = ScalarParameter() {defaultValue=0.5, parameterName=LowEndNormalStrength}
  MaterialFunctionCall(Normal=@n79, Flatness=QualitySwitch(Default=@n118, Low=@n119, High=@n118, Medium=@n119)) {materialFunction=&FlattenNormal}
  n120 = OneMinus(None=@n27)
  n121 = ConstantBiasScale(None=@n5)
  n122 = Power(Base=Subtract(A=@n121, B=ScalarParameter() {defaultValue=0.5, group=ViperZero, parameterName=Falloff}))
  Multiply(A=Multiply(A=Clamp(None=@n120), B=@n55.R), B=Clamp(None=Multiply(A=@n122, B=@n38)))
  n123 = Subtract(A=CameraPositionWS(), B=ObjectPositionWS())
  n124 = ConstantBiasScale(None=DotProduct(A=Normalize(VectorInput=@n123), B=Transform(None=@n78.RGB)))
  n125 = ScalarParameter() {group=ViperZero, parameterName=VolumeShadowFalloff}
  Multiply(A=@n87, B=Clamp(None=Subtract(A=@n124, B=@n125) {constB=0.25}))
  Multiply(A=@n15, B=@n18)
  Multiply(A=@n45, B=@n42.RGB)
  Normalize(VectorInput=Subtract(
    A=MaterialFunctionCall() {materialFunction=&ObjectPivotPoint}.Object Pivot Location
    B=CameraPositionWS()
   ))
  Power(Base=Subtract(A=@n121, B=ScalarParameter() {group=ViperZero, parameterName=NormalFresnelFalloff}))
  Subtract(A=@n7, B=Multiply(A=@n38, B=ScalarParameter() {defaultValue=0.02, group=ViperZero, parameterName=PaintedGlowDistortionStrength}))
  Subtract(A=@n55.R, B=@n58)
  TextureCoordinate()
  TextureSample() {samplerType=SAMPLERTYPE_Normal, texture=/Game/Characters/Player/Male/Medium/Bodies/M_MED_Haunt/Textures/T_Haunt_Body_N}
  Transform(None=Constant3Vector() {constant=<rgba 0 1 0 0>}) {transformSourceType=TRANSFORMSOURCE_World, transformType=TRANSFORM_Local}
  VertexNormalWS()
  VertexNormalWS()
 ]}
