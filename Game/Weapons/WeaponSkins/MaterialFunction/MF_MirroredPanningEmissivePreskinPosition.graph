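# Weapon skin material function that adds a mirrored, panning emissive texture on top of the incoming material attributes' emissive.
# It lerps the pre-skinned local position toward its absolute value per axis for mirroring, then subtracts time times speed for panning.
# That position and the interpolated pre-skinned normal drive a WorldAlignedTexture projection.
# The result is tinted, raised to an exponent and masked by a selectable FXMask channel.
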MaterialFunction /Game/Weapons/WeaponSkins/MaterialFunction/MF_MirroredPanningEmissivePreskinPosition
graph {nodes [
  n1 = GetMaterialAttributes(None=FunctionInput() {inputName=MA, inputType=FunctionInput_MaterialAttributes}) {attributeGetTypes [#1]}
  n2 = TextureObjectParameter() {
   group=MirroredPanningEmissivePreskinPosition
   parameterName=MirroredPanningEmissiveTexture
   texture=&T_BlacksRGB
  }
  n3 = VectorParameter() {group=MirroredPanningEmissivePreskinPosition, parameterName=MirroredPanningEmissiveSize, defaultValue=<rgba 50 50 50 0>}
  n4 = FunctionInput() {inputName=PreskinLocalPosition, sortPriority=3}
  n5 = Abs(None=@n4)
  n6 = VectorParameter() {group=MirroredPanningEmissivePreskinPosition, parameterName=MirroredPanningEmissiveMirroredAxes}
  n7 = Time()
  n8 = VectorParameter() {group=MirroredPanningEmissivePreskinPosition, parameterName=MirroredPanningEmissiveSpeed, defaultValue=<rgba 0 0 10 0>}
  n9 = MaterialFunctionCall(
   TextureObject=@n2
   TextureSize=@n3.RGB
   WorldPosition=Subtract(A=LinearInterpolate(A=@n4, B=@n5, Alpha=@n6.RGB), B=Multiply(A=@n7, B=@n8.RGB))
   World Space Normal=VertexInterpolator(VS=PreSkinnedNormal()).PS
  ) {materialFunction=/Game/Weapons/WeaponSkins/Wraps/MaterialFunctions/WorldAlignedTexture}
  n10 = VectorParameter() {group=MirroredPanningEmissivePreskinPosition, parameterName=MirroredPanningEmissiveColor, defaultValue=<rgba 1 1 1 0>}
  n11 = FunctionInput(Preview=TextureObject()) {bUsePreviewValueAsDefault=true, inputName=FXMask, inputType=FunctionInput_Texture2D, sortPriority=1}
  n12 = Multiply(
   A=Power(
    Base=Multiply(A=@n9.XYZ Texture, B=@n10.RGB)
    Exp=ScalarParameter() {defaultValue=1, group=MirroredPanningEmissivePreskinPosition, parameterName=MirroredPanningEmissiveExponent}
   )
   B=ChannelMaskParameter(None=TextureSample(Tex=@n11) {samplerSource=SSM_Wrap_WorldGroupSettings}.RGBA) {group=MirroredPanningEmissivePreskinPosition, parameterName=FXMask_ChannelPicker}.R
  )
  FunctionOutput(None=SetMaterialAttributes!(MaterialAttributes=@n1.MaterialAttributes, EmissiveColor=Add(A=@n1.EmissiveColor, B=@n12)) {attributeSetTypes [#1]})
 ]}
