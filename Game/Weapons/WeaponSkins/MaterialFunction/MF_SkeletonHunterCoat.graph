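# Material function for the Skeleton Hunter coat weapon skin that adds animated displacement, folds and a patterned trim to incoming material attributes.
# World position offset combines panning Perlin noise with an object-relative direction, scaled by MaskInput R and G.
# Noise-distorted UVs build the trim mask, which flattens normals and blends curve-atlas color, specular and roughness.
# Panning fold normals use pre-skinned position, and UseWPOandNormalOnly restores the incoming color, specular and roughness.
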MaterialFunction /Game/Weapons/WeaponSkins/MaterialFunction/MF_SkeletonHunterCoat
graph {nodes [
  n1 = WorldPosition()
  n2 = ObjectPositionWS()
  n3 = Normalize(VectorInput=Transform(None=Subtract(A=@n1.XYZ, B=@n2)) {transformSourceType=TRANSFORMSOURCE_World, transformType=TRANSFORM_Local})
  Add(A=@n3) {constB=-1}
  n4 = VectorParameter() {group=SkeletonHunterCoat, parameterName=TrimPatternTiling, defaultValue=<rgba 128 128 128 1>}
  ComponentMask(None=@n4.RGB) {g=true, r=true}
  Constant3Vector()
  FunctionInput() {bUsePreviewValueAsDefault=true, inputName=GMap}
  n5 = GetMaterialAttributes(None=FunctionInput() {bUsePreviewValueAsDefault=true, inputName=Material, inputType=FunctionInput_MaterialAttributes}) {attributeGetTypes [
    #1
    #2
    #3
    #4
    #5
    #6
   ]}
  n6 = FunctionInput() {bUsePreviewValueAsDefault=true, inputName=MaskInput, inputType=FunctionInput_Vector4}
  n7 = FunctionInput(Preview=PreSkinnedPosition()) {bUsePreviewValueAsDefault=true, inputName=Preskinned-Position}
  n8 = Divide(
   A=ComponentMask(None=@n7) {b=true, r=true}
   B=ComponentMask(None=VectorParameter() {group=SkeletonHunterCoat, parameterName=NoiseTiling, defaultValue=<rgba 32 32 32 1>}.RGB) {g=true, r=true}
  )
  n9 = Multiply(
   A=Time()
   B=ComponentMask(None=VectorParameter() {group=SkeletonHunterCoat, parameterName=CoatNoisePanSpeedAndDirection, defaultValue=<rgba 0 1 0 1>}.RGB) {g=true, r=true}
  )
  n10 = Multiply(A=TextureSample(UVs=Add(A=@n8, B=@n9)) {texture=/Game/Global/Textures/Default/Noise/3xPerlin_M}.RGB) {constB=2}
  n11 = Add(
   A=@n3
   B=Multiply(A=Subtract(A=@n10), B=VectorParameter() {group=SkeletonHunterCoat, parameterName=Displacement_Strength, defaultValue=<rgba 2 2 2 1>}.RGB)
  )
  n12 = ComponentMask(None=@n6) {r=true}
  n13 = TextureCoordinate()
  n14 = VertexInterpolator(VS=@n7)
  n15 = VectorParameter() {group=SkeletonHunterCoat, parameterName=TrimPatternDistortionPanDirectionAndSpeed, defaultValue=<rgba 0 1 1 1>}
  n16 = Time()
  n17 = Add(
   A=Divide(A=@n14.PS, B=VectorParameter() {group=SkeletonHunterCoat, parameterName=TrimPatternDistortionTiling, defaultValue=<rgba 128 128 128 1>}.RGB)
   B=Multiply(A=@n15.RGB, B=@n16)
  )
  n18 = VertexNormalWS()
  n19 = LinearInterpolate(
   A=TextureSample(UVs=ComponentMask(None=@n17) {b=true, r=true}) {texture=&T_NoisePack_01}.RGB
   B=TextureSample(UVs=ComponentMask(None=@n17) {g=true, r=true}) {texture=&T_NoisePack_01}.RGB
   Alpha=Abs(None=ComponentMask(None=@n18) {b=true})
  )
  n20 = Multiply(
   A=Subtract(A=Multiply(A=@n19) {constB=2})
   B=VectorParameter() {group=SkeletonHunterCoat, parameterName=XYZ_DistortionStength, defaultValue=<rgba 0.02 0.02 0.02 1>}.RGB
  )
  n21 = Multiply(
   A=ComponentMask(None=@n20) {g=true, r=true}
   B=ScalarParameter() {defaultValue=0.02, group=SkeletonHunterCoat, parameterName=TrimMaskDistortionStrength}
  )
  n22 = Multiply(
   A=TextureSample(UVs=Add(A=@n13, B=@n21)) {texture=/Game/Characters/Player/Male/Medium/Bodies/M_MED_Skeleton_Hunter/Textures/T_GraphicNoise}.G
   B=@n12
  )
  n23 = MaterialFunctionCall(
   Normal=@n5.Normal
   Flatness=LinearInterpolate(A=Constant(), B=ScalarParameter() {defaultValue=1, group=SkeletonHunterCoat, parameterName=TrimNormalFlattenValue}, Alpha=@n22)
  ) {materialFunction=&FlattenNormal}
  n24 = Add(
   A=Divide(
    A=VertexInterpolator(VS=@n7).PS
    B=VectorParameter() {group=SkeletonHunterCoat, parameterName=FoldsTiling, defaultValue=<rgba 64 64 64 1>}.RGB
   )
   B=Multiply(A=Time(), B=VectorParameter() {group=SkeletonHunterCoat, parameterName=FoldsPanSpeedAndDirection', defaultValue=<rgba 0 0 2 1>}.RGB)
  )
  n25 = AppendVector(A=ComponentMask(None=@n24) {r=true}, B=OneMinus(None=ComponentMask(None=@n24) {b=true}))
  n26 = ScalarParameter() {defaultValue=0.05, group=SkeletonHunterCoat, parameterName=NormalStrength}
  n27 = MaterialFunctionCall(
   BaseNormal=@n23.Result
   AdditionalNormal=LinearInterpolate(
    A=@n23.Result
    B=TextureSample(UVs=@n25) {samplerType=SAMPLERTYPE_Normal, texture=/Game/Gadgets/Assets/VinderTech_GliderChute/Glider_BlackMondayCape/Textures/T_Folds_N}.RGB
    Alpha=@n26
   )
  ) {materialFunction=&BlendAngleCorrectedNormals}
  n28 = Clamp(None=Multiply(A=ComponentMask(None=@n6) {g=true}) {constB=5})
  n29 = Add(A=Divide(A=@n14.PS, B=@n4.RGB), B=@n20)
  n30 = VectorParameter() {group=SkeletonHunterCoat, parameterName=TrimPatternPanDirectionAndSpeed, defaultValue=<rgba 0 0.5 0.5 1>}
  n31 = Multiply(A=@n30.RGB, B=@n16)
  n32 = Add(A=@n29, B=@n31)
  n33 = CurveAtlasRowParameter(CurveTime=TextureSample(UVs=ComponentMask(None=@n32) {b=true, r=true}) {texture=&T_NoisePack_01}.R) {
   atlas=/Game/Characters/Shared/CurveAtlas/AthenaCharacterCurveAtlas
   curve=/Game/Characters/Player/Male/Medium/Bodies/M_MED_Skeleton_Hunter/Textures/skeletonhuntercolor
   defaultValue=5
   group=SkeletonHunterCoat
   parameterName=CurveAtlas
  }
  n34 = Multiply(
   A=@n33
   B=LinearInterpolate(
    A=ScalarParameter() {defaultValue=0.5, group=SkeletonHunterCoat, parameterName=TrimMinGlowStrength}
    B=ScalarParameter() {defaultValue=1, group=SkeletonHunterCoat, parameterName=TrimMaxGlowStrength}
    Alpha=ComponentMask(None=@n6) {g=true}
   )
  )
  n35 = ScalarParameter() {group=SkeletonHunterCoat, parameterName=UseWPOandNormalOnly}
  n36 = ScalarParameter() {defaultValue=1, group=SkeletonHunterCoat, parameterName=TrimSpecValue}
  n37 = ScalarParameter() {group=SkeletonHunterCoat, parameterName=TrimRoughnessValue}
  n38 = SetMaterialAttributes!(
   MaterialAttributes=@n5.MaterialAttributes
   WorldPositionOffset=Multiply(A=ComponentMask(None=@n6) {g=true}, B=Multiply(A=@n11, B=@n12))
   Normal=LinearInterpolate(A=@n23.Result, B=LinearInterpolate(A=@n23.Result, B=@n27.Result, Alpha=@n28), Alpha=@n12)
   BaseColor=LinearInterpolate(A=LinearInterpolate(A=@n5.BaseColor, B=@n34, Alpha=@n22), B=@n5.BaseColor, Alpha=@n35)
   Specular=LinearInterpolate(A=LinearInterpolate(A=@n5.Specular, B=@n36, Alpha=@n22), B=@n5.Specular, Alpha=@n35)
   Roughness=LinearInterpolate(A=LinearInterpolate(A=@n5.Roughness, B=@n37, Alpha=@n22), B=@n5.Roughness, Alpha=@n35)
  ) {attributeSetTypes [
    #2
    #1
    #4
    #5
    #6
   ]}
  FunctionOutput(None=@n38)
  MaterialFunctionCall() {materialFunction=&ObjectPivotPoint}
  Multiply(A=@n3)
  n39 = FunctionInput(Preview=PreSkinnedLocalBounds().Max) {bUsePreviewValueAsDefault=true, inputName=Bounds}
  n40 = OneMinus(None=ComponentMask(None=Divide(A=@n7, B=@n39)) {b=true})
  Multiply(A=ComponentMask(None=VertexInterpolator(VS=@n40).PS) {r=true}, B=@n26) {constB=0.25, desc=1}
  Panner(Coordinate=@n29, Speed=ComponentMask(None=@n31) {g=true, r=true})
  VertexInterpolator(VS=@n7)
 ]}
