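# Material function for skins that overlays animated falling snow and a gradient background on masked areas of input material attributes.
# Three panning, bump-offset snowflake layers use screen-space UVs anchored to the actor and scaled by radius over view depth.
# Their maximum fades by facing angle and optional close-camera (ADS) distance.
# Pre-skinned position gradients drive curve-atlas color and flake strength, feeding base color and fresnel-edged emissive, alongside metallic, specular and roughness overrides.
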MaterialFunction /Game/Weapons/WeaponSkins/MaterialFunction/MF_SnowFall
graph {nodes [
  n1 = MaterialFunctionCall(Float2=ViewSize()) {materialFunction=&BreakOutFloat2Components}
  n2 = AppendVector(A=Constant() {r=1}, B=Divide(A=@n1.G, B=@n1.G))
  n3 = Multiply(A=ConstantBiasScale(None=ScreenPosition().ViewportUV) {bias=-0.5, scale=2}, B=@n2)
  n4 = Multiply(
   A=Constant2Vector() {g=-0.5, r=0.5}
   B=MaterialFunctionCall(World Position=ActorPositionWS()) {materialFunction=&TransformToClipSpace}.Clip Space XY
  )
  n5 = ObjectRadius()
  n6 = ComponentMask(None=TransformPosition(None=ActorPositionWS()) {transformSourceType=TRANSFORMPOSSOURCE_World, transformType=TRANSFORMPOSSOURCE_View}) {b=true}
  n7 = MaterialFunctionCall(
   UVs=Subtract(A=ConstantBiasScale(None=@n3), B=Multiply(A=@n2, B=@n4))
   Texture Scale=Multiply(A=Divide(A=@n5, B=@n6), B=ScalarParameter() {defaultValue=1, group=SnowFall, parameterName=SF_UVRadialScaleZero})
  ) {materialFunction=&ScaleUVsByCenter}
  n8 = BumpOffset(
   Coordinate=Multiply(A=@n7.UVs, B=ScalarParameter() {defaultValue=0.4, group=SnowFall, parameterName=SF_Layer_1_Tiling})
   Height=ScalarParameter() {group=SnowFall, parameterName=SF_Layer_01_Height}
  )
  n9 = ScalarParameter() {defaultValue=1.5, group=SnowFall, parameterName=SF_GlobalSpeed}
  n10 = Multiply(
   A=ComponentMask(None=VectorParameter() {group=SnowFall, parameterName=SF_Layer_01_PanSpeedAndDirection, defaultValue=<rgba 0.035 -0.25 0 1>}.RGB) {g=true, r=true}
   B=@n9
  )
  n11 = TextureObjectParameter() {
   group=SnowFall
   parameterName=SF_SnowflakeTexture
   texture=/Game/Characters/Player/Female/Medium/Bodies/F_MED_Gift_Wrap/Textures/T_Snowflake_Pattern
  }
  n12 = ComponentMask(None=TextureSample(UVs=Panner(Coordinate=@n8, Speed=@n10), Tex=@n11) {texture=&T_Snowflake_Pattern}.RGB) {r=true}
  n13 = ScalarParameter() {defaultValue=0.75, group=SnowFall, parameterName=SF_Layer_2_Tiling}
  n14 = VectorParameter() {group=SnowFall, parameterName=SF_Layer_02_PanSpeedAndDirection, defaultValue=<rgba -0.1 -0.32 0 1>}
  n15 = Panner(
   Coordinate=BumpOffset(Coordinate=Multiply(A=@n7.UVs, B=@n13), Height=ScalarParameter() {defaultValue=-0.5, group=SnowFall, parameterName=SF_Layer_02_Height})
   Speed=Multiply(A=ComponentMask(None=@n14.RGB) {g=true, r=true}, B=@n9)
  )
  n16 = Multiply(
   A=ComponentMask(None=TextureSample(UVs=@n15, Tex=@n11) {texture=&T_Snowflake_Pattern}.RGB) {r=true}
   B=ScalarParameter() {defaultValue=0.5, group=SnowFall, parameterName=SF_Layuer_02_Brightness}
  ) {constB=0.5}
  n17 = ScalarParameter() {defaultValue=1.25, group=SnowFall, parameterName=SF_Layer_3_Tiling}
  n18 = VectorParameter() {group=SnowFall, parameterName=SF_Layer_03_PanSpeedAndDirection, defaultValue=<rgba 0.02 -0.5 0 1>}
  n19 = Panner(
   Coordinate=BumpOffset(Coordinate=Multiply(A=@n7.UVs, B=@n17), Height=ScalarParameter() {defaultValue=-1, group=SnowFall, parameterName=SF_Layer_03_Height})
   Speed=Multiply(A=ComponentMask(None=@n18.RGB) {g=true, r=true}, B=@n9)
  )
  n20 = Multiply(
   A=ComponentMask(None=TextureSample(UVs=@n19, Tex=@n11) {texture=&T_Snowflake_Pattern}.RGB) {r=true}
   B=ScalarParameter() {defaultValue=0.25, group=SnowFall, parameterName=SF_Layuer_03_Brightness}
  ) {constB=0.25}
  Add(A=Add(A=@n12, B=@n16), B=@n20)
  n21 = GetMaterialAttributes(None=FunctionInput() {inputName=Material, inputType=FunctionInput_MaterialAttributes}) {attributeGetTypes [
    #1
    #2
    #3
    #4
    #5
   ]}
  n22 = Max(A=Max(A=@n12, B=@n16), B=@n20)
  n23 = Distance(A=CameraPositionWS(), B=WorldPosition().XYZ)
  n24 = ScalarParameter() {defaultValue=32, group=SnowFall, parameterName=SF_ADS_BiasDistance}
  n25 = Clamp(None=Divide(A=Subtract(A=@n23, B=@n24), B=ScalarParameter() {defaultValue=128, group=SnowFall, parameterName=SF_ADSBlendDistance}))
  n26 = DotProduct(A=CameraVectorWS(), B=VertexNormalWS())
  n27 = Power(Base=Subtract(A=Clamp(None=@n26)) {constB=0.1})
  n28 = Multiply(
   A=LinearInterpolate(A=@n22, B=Multiply(A=@n25, B=@n22), Alpha=ScalarParameter() {defaultValue=1, group=SnowFall, parameterName=SF_UseADSBlending})
   B=Clamp(None=Multiply(A=@n27) {constB=2})
  )
  n29 = TextureSample() {samplerType=SAMPLERTYPE_Masks, texture=/Game/Characters/Player/Female/Medium/Bodies/F_MED_Gift_Wrap/Textures/T_F_MED_Gift_Wrap_Body_FX}
  n30 = ChannelMaskParameter(None=@n29.G) {group=SnowFall, maskChannel=Green, parameterName=SF_FresnelMaskChanel, defaultValue=<rgba 0 1 0 0>}
  n31 = ChannelMaskParameter(None=VertexInterpolator(VS=PreSkinnedPosition()).PS) {group=SnowFall, maskChannel=Blue, parameterName=GradientChannel, defaultValue=<rgba 0 0 1 0>}
  n32 = LinearInterpolate(
   A=ScalarParameter() {defaultValue=100, group=SnowFall, parameterName=SF_Gradient_01_Scale}
   B=ChannelMaskParameter(None=MaterialFunctionCall() {materialFunction=&ObjectLocalBounds}.Local Bounds Size) {group=SnowFall, maskChannel=Blue, parameterName=GradientChannel, defaultValue=<rgba 0 0 1 0>}.R
   Alpha=ScalarParameter() {group=SnowFall, parameterName=SF_UseObjectScaleForGradient}
  )
  n33 = Add(A=@n31.R, B=ScalarParameter() {defaultValue=-32, group=SnowFall, parameterName=Gradient_02_StartPosition})
  n34 = ScalarParameter() {defaultValue=32, group=SnowFall, parameterName=SF_Gradient_02_Scale}
  n35 = FunctionInput(Preview=@n29.RGBA) {bUsePreviewValueAsDefault=true, inputName=Mask, inputType=FunctionInput_Vector4}
  n36 = ChannelMaskParameter(None=@n35) {group=SnowFall, maskChannel=Alpha, parameterName=SF_Gradient_02MaskChannel, defaultValue=<rgba 0 0 0 1>}
  n37 = LinearInterpolate(A=Clamp(None=Divide(A=@n31.R, B=@n32)), B=Clamp(None=Divide(A=@n33, B=@n34)), Alpha=@n36.R)
  n38 = CurveAtlasRowParameter(CurveTime=LinearInterpolate(A=@n37, B=Add(A=@n30.R, B=@n37), Alpha=ScalarParameter() {group=SnowFall, parameterName=SF_UseFresnelEdgeMsk})) {
   atlas=&AthenaCurveAtlas01
   curve=/Game/Characters/Player/Female/Medium/Bodies/F_MED_Gift_Wrap/Textures/SnowfallBG_Color
   defaultValue=172
   group=SnowFall
   parameterName=SF_BG_ColorCurve
  }
  n39 = CurveAtlasRowParameter(CurveTime=@n37) {
   atlas=&AthenaCurveAtlas01
   curve=/Game/Characters/Player/Female/Medium/Bodies/F_MED_Gift_Wrap/Textures/GiftwrapPantsGradient
   defaultValue=179
   group=SnowFall
   parameterName=SnowflakeOpacityCurve
  }
  n40 = Multiply(
   A=Desaturation(None=Multiply(A=@n28, B=@n38), Fraction=ScalarParameter() {defaultValue=0.75, group=SnowFall, parameterName=SF_SnowflakeDesaturation})
   B=LinearInterpolate(
    A=ScalarParameter() {defaultValue=20, group=SnowFall, parameterName=SF_SnowFlakeMinStrength}
    B=ScalarParameter() {defaultValue=100, group=SnowFall, parameterName=SF_SnowFlakeStrength}
    Alpha=ComponentMask(None=@n39) {r=true}
   )
  )
  n41 = ChannelMaskParameter(None=@n35) {group=SnowFall, parameterName=SF_FX_MaskChannel}
  n42 = OneMinus(None=@n26)
  n43 = ScalarParameter() {group=SnowFall, parameterName=SF_EmissiveFalloff}
  n44 = Clamp(None=Power(Base=Subtract(A=@n42, B=@n43) {constB=0.35}, Exp=ScalarParameter() {defaultValue=2, group=SnowFall, parameterName=SF_EmissvePower}))
  n45 = LinearInterpolate(
   A=Multiply(A=Multiply(A=@n38, B=@n44), B=ScalarParameter() {defaultValue=5, group=SnowFall, parameterName=SF_Emissive_Strength})
   B=@n40
   Alpha=@n28
  )
  n46 = SetMaterialAttributes!(
   MaterialAttributes=@n21.MaterialAttributes
   BaseColor=LinearInterpolate(A=@n21.BaseColor, B=Add(A=@n40, B=@n38), Alpha=@n41.R)
   Metallic=LinearInterpolate(A=@n21.Metallic, B=ScalarParameter() {group=SnowFall, parameterName=SF_Metalness}, Alpha=@n41.R)
   Specular=LinearInterpolate(A=@n21.Specular, B=ScalarParameter() {defaultValue=1, group=SnowFall, parameterName=SF_SpecValue}, Alpha=@n41.R)
   Roughness=LinearInterpolate(A=@n21.Roughness, B=ScalarParameter() {defaultValue=0.1, group=SnowFall, parameterName=SF_Roughness}, Alpha=@n41.R)
   EmissiveColor=LinearInterpolate(A=@n21.EmissiveColor, B=Add(A=@n21.EmissiveColor, B=@n45), Alpha=@n41.R)
  ) {attributeSetTypes [
    #1
    #2
    #3
    #4
    #5
   ]}
  FunctionOutput(None=@n46)
 ]}
