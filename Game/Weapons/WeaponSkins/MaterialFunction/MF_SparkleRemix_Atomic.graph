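# Weapon skin material function for Sparkle Remix "Atomic" that replaces emissive on input Material Attributes, blended by Alpha.
# Packed Atomic_M channels drive a background, optionally tinted by a gradient panning in pre-skinned space, and animated SphereMask accents.
# Emote_Rand hue-shifts both layers half a cycle apart and toggles between static gradient-and-pulse and panning flat-and-sweep variants.
# Emote_AudioLow from FortniteMaterialParameters raises overall brightness and widens the pulse accent radius.
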MaterialFunction /Game/Weapons/WeaponSkins/MaterialFunction/MF_SparkleRemix_Atomic
graph {nodes [
  n1 = FunctionInput() {inputName=Material, inputType=FunctionInput_MaterialAttributes}
  n2 = Divide(A=ScalarParameter() {group=Marshemello, parameterName=Emote_Rand}) {constB=100}
  n3 = Constant3Vector() {constant=<rgba 0 1 0.6445 0>}
  n4 = Multiply(
   A=MaterialFunctionCall(Hue Shift Percentage=@n2, Texture=@n3) {materialFunction=&HueShift}.Result
   B=ScalarParameter() {defaultValue=2, group=Atomic, parameterName="Background  Emissive Brightness"}
  )
  n5 = Multiply(A=TextureCoordinate(), B=ScalarParameter() {defaultValue=4, group=Atomic, parameterName=Light Threshold Texture Tiling})
  n6 = ScalarParameter() {defaultValue=1, group=Atomic, parameterName=Global Tempo}
  n7 = Multiply(A=Time(), B=@n6)
  n8 = Multiply(A=@n7) {constB=0.25}
  n9 = Round(None=Frac(None=@n2))
  n10 = TextureSample(
   UVs=LinearInterpolate(A=@n5, B=Add(A=@n5, B=@n8), Alpha=@n9)
   Tex=TextureObject() {samplerType=SAMPLERTYPE_Masks, texture=/Game/Characters/Player/Female/Medium/Bodies/F_MED_SparkleRemix/Textures/Atomic/Atomic_M}
  ) {samplerType=SAMPLERTYPE_Masks, texture=&Atomic_M}
  n11 = ScalarParameter() {defaultValue=0.25, group=Atomic, parameterName=Gradient Tiling}
  n12 = Divide(A=VertexInterpolator(VS=PreSkinnedPosition()).PS, B=Constant() {r=40})
  n13 = Multiply(A=@n6, B=Time())
  n14 = VectorParameter() {group=Sparkle Remix, parameterName=Background Gradient Panning Speed, defaultValue=<rgba 0 0 0.5 1>}
  n15 = ComponentMask(None=Add(A=Multiply(A=@n11, B=@n12), B=Multiply(A=@n13, B=@n14.RGB))) {b=true, r=true}
  n16 = LinearInterpolate(
   A=Multiply(
    A=@n10.R
    B=TextureSample(UVs=@n15) {samplerType=SAMPLERTYPE_Masks, texture=/Game/Characters/Player/Female/Medium/Bodies/F_MED_SparkleRemix/Textures/Atomic/Atomic_Grad}.RGB
   )
   B=@n10.R
   Alpha=@n9
  )
  n17 = ScalarParameter() {defaultValue=3, parameterName=ForeGround Element Brightness}
  n18 = CollectionParameter() {collection=&FortniteMaterialParameters, parameterName=Emote_AudioLow}
  n19 = If(A=@n18, B=Constant() {r=0.001}, "A > B"=Constant() {r=1}, "A < B"=Constant())
  n20 = LinearInterpolate(
   A=ScalarParameter() {defaultValue=0.25, parameterName=Accent Radius}
   B=ScalarParameter() {defaultValue=0.66, parameterName=Accent Radius Emote}
   Alpha=@n19
  )
  n21 = Frac(None=Multiply(A=@n7) {constB=0.25})
  n22 = LinearInterpolate(Alpha=@n21) {constB=2}
  n23 = StaticBool() {value=true}
  n24 = SphereMask(
   A=@n10.G
   B=LinearInterpolate(
    A=@n20
    B=Add(B=@n20) {constA=1}
    Alpha=MaterialFunctionCall(Value=@n22, -1 to 1=@n23) {materialFunction=&LinearSine}.Linear Sine
   )
   Radius=@n20
   Hardness=Constant() {r=1}
  )
  n25 = SphereMask(A=Add(A=Subtract(A=@n10.B), B=Multiply(A=@n21) {constB=2}), B=Constant() {r=0.5}, Radius=Constant() {r=0.1}, Hardness=Constant() {r=1})
  n26 = Multiply(
   A=MaterialFunctionCall(Hue Shift Percentage=Add(A=@n2) {constB=0.5}, Texture=@n3) {materialFunction=&HueShift}.Result
   B=Multiply(A=@n10.A, B=LinearInterpolate(A=@n24, B=@n25, Alpha=@n9))
  )
  n27 = ScalarParameter() {defaultValue=3.5, group=Atomic, parameterName=Global Atomic Brightness}
  n28 = Multiply(
   A=Add(A=Multiply(A=@n4, B=@n16), B=Multiply(A=@n17, B=@n26))
   B=LinearInterpolate(A=@n27, B=Multiply(A=@n27) {constB=4}, Alpha=@n19) {constB=2.5}
  )
  FunctionOutput(None=BlendMaterialAttributes(
    A=@n1
    B=SetMaterialAttributes!(MaterialAttributes=@n1, EmissiveColor=@n28) {attributeSetTypes [#1]}
    Alpha=FunctionInput() {inputName=Alpha, inputType=FunctionInput_Scalar}
   ))
 ]}
