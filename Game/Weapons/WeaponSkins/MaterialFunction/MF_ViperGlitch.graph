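# Material function for the Viper glitch skin effect, layering an animated, color-curved glitch pattern onto incoming material attributes through the Mask input's blue channel.
# AO-varied panning noise distorts the pattern, projected on pre-skinned XZ position, which drives base color, emissive and blended normals, plus a mask-texture edge glow.
# A head scalar blends in normal distortion, fresnel face glow and a matte finish.
# The Perlin noise branch and GMapInput are unconnected to the output.
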MaterialFunction /Game/Weapons/WeaponSkins/MaterialFunction/MF_ViperGlitch
graph {nodes [
  n1 = Constant() {r=1}
  Clamp(None=Multiply(A=FunctionInput(Preview=@n1) {bUsePreviewValueAsDefault=true, inputName=GMapInput, inputType=FunctionInput_Scalar}) {constB=2})
  n2 = GetMaterialAttributes(None=FunctionInput() {inputName=Material, inputType=FunctionInput_MaterialAttributes}) {attributeGetTypes [
    #1
    #2
    #3
    #4
    #5
   ]}
  n3 = Multiply(
   A=ComponentMask(None=Transform(None=@n2.Normal)) {g=true, r=true}
   B=ScalarParameter() {defaultValue=0.5, group=ViperGlitch, parameterName=NormalDistortionStrength}
  )
  n4 = ScalarParameter() {group=ViperGlitch, parameterName=bIsheadViperHEad}
  n5 = ComponentMask(None=VertexInterpolator(VS=PreSkinnedPosition()).PS) {b=true, r=true}
  n6 = Multiply(
   A=@n5
   B=AppendVector(
    A=ScalarParameter() {defaultValue=2, group=ViperGlitch, parameterName=Pattern_U_Tiling}
    B=ScalarParameter() {defaultValue=2, group=ViperGlitch, parameterName=Pattern_V_Tiling}
   )
  )
  n7 = Multiply(
   A=@n5
   B=AppendVector(
    A=ScalarParameter() {defaultValue=1, group=ViperGlitch, parameterName=Noise_U_Tiling}
    B=ScalarParameter() {defaultValue=1, group=ViperGlitch, parameterName=Noise_V_Tiling}
   )
  )
  n8 = LinearInterpolate(
   A=ScalarParameter() {defaultValue=0.5, group=ViperGlitch, parameterName=MinSpeedVariance}
   B=ScalarParameter() {defaultValue=1, group=ViperGlitch, parameterName=MaxSpeedVariance}
   Alpha=FunctionInput(Preview=@n1) {bUsePreviewValueAsDefault=true, inputName=AO_Input, inputType=FunctionInput_Scalar}
  )
  n9 = ComponentMask(None=Multiply(A=@n8, B=VectorParameter() {group=ViperGlitch, parameterName=NoisePanSpeedandDirection, defaultValue=<rgba 0.1 -0.1 0 1>}.RGB)) {g=true, r=true}
  n10 = ComponentMask(None=TextureSample(UVs=Panner(Coordinate=@n7, Speed=@n9) {speedX=0.1, speedY=-0.1}) {texture=/Game/Characters/Player/Male/Medium/Bodies/M_MED_Viper/Textures/Glitch/T_SoftNoise}.RGB) {g=true, r=true}
  n11 = Multiply(A=Subtract(A=Multiply(A=@n10) {constB=2}), B=ScalarParameter() {defaultValue=0.4, group=ViperGlitch, parameterName=Distortion_Strength})
  n12 = VectorParameter() {group=ViperGlitch, parameterName=TexturePanSpeedandDirection, defaultValue=<rgba -0.1 0.1 0 1>}
  n13 = Panner(
   Coordinate=Add(A=LinearInterpolate(B=@n3, Alpha=@n4), B=Add(A=@n6, B=@n11))
   Speed=ComponentMask(None=Multiply(A=@n8, B=@n12.RGB)) {g=true, r=true}
  ) {speedX=-0.1, speedY=0.1}
  n14 = TextureSample(UVs=@n13) {texture=/Game/Characters/Player/Male/Medium/Bodies/M_MED_Viper/Textures/Glitch/T_ViperGlitchNoise}
  n15 = Multiply(A=@n14.R) {constB=2}
  n16 = CurveAtlasRowParameter(CurveTime=@n15) {
   atlas=&AthenaCurveAtlas01
   curve=/Game/Characters/Player/Male/Medium/Bodies/M_MED_Viper/Textures/Glitch/viperGllitchColor
   defaultValue=136
   parameterName=ColorCurve
  }
  n17 = ScalarParameter() {defaultValue=0.05, group=ViperGlitch, parameterName=BG_Strength}
  n18 = Multiply(A=@n16, B=Multiply(A=Clamp(None=@n15, Min=@n17), B=ScalarParameter() {defaultValue=1, group=ViperGlitch, parameterName=EmissiveStrtength}))
  n19 = VectorParameter() {group=ViperGlitch, parameterName=EdgeColor, defaultValue=<rgba 0 0.627 1 1>}
  n20 = TextureSampleParameter2D(UVs=TextureCoordinate()) {group=ViperGlitch, parameterName=GlitchGLowMask, texture=&T_BlacksRGB}
  n21 = LinearInterpolate(A=@n20.R, B=@n20.R, Alpha=@n4)
  n22 = ScalarParameter() {defaultValue=0.1, group=ViperGlitch, parameterName=FresnelDistortStrength}
  n23 = Clamp(
   None=Subtract(A=@n21, B=Multiply(A=@n14.R, B=@n22))
   Min=ScalarParameter() {defaultValue=0.5, group=ViperGlitch, parameterName=FresnelMinStrength}
  )
  n24 = Power(Base=Multiply(A=Multiply(A=@n19.RGB, B=@n23), B=ScalarParameter() {defaultValue=5, group=ViperGlitch, parameterName=EdgeGlowStrength}))
  n25 = ComponentMask(None=FunctionInput() {inputName=Mask}) {b=true}
  n26 = Multiply(A=@n25, B=Add(A=Power(Base=@n18), B=@n24))
  n27 = Add(A=@n2.EmissiveColor, B=@n26)
  n28 = Add(A=LinearInterpolate(A=@n27, B=Multiply(A=@n27, B=@n20.R), Alpha=@n4), B=@n24)
  n29 = MaterialFunctionCall(
   BaseNormal=@n2.Normal
   AdditionalNormal=TextureSample(UVs=@n13) {samplerType=SAMPLERTYPE_Normal, texture=/Game/Global/Textures/Default/Noise/T_VIperGlitchNoise_N}.RGB
  ) {materialFunction=&BlendAngleCorrectedNormals}
  n30 = Fresnel(ExponentIn=ScalarParameter() {defaultValue=2, group=ViperGlitch, parameterName=FaceFresnelExponent}, Normal=Transform(None=@n29.Result)) {exponent=10}
  n31 = ScalarParameter() {group=ViperGlitch, parameterName=FaceHighlightingFalloff}
  n32 = Multiply(A=@n20.R, B=Clamp(None=Subtract(A=@n30, B=@n31)))
  n33 = Multiply(A=@n21, B=ScalarParameter() {defaultValue=0.2, group=ViperGlitch, parameterName=FaceCenterGlowStrength}) {constB=2}
  n34 = Multiply(
   A=Multiply(A=VectorParameter() {group=ViperGlitch, parameterName=FaceGlowCOlor, defaultValue=<rgba 0 0.6549 1 1>}.RGB, B=Add(A=@n32, B=@n33))
   B=ScalarParameter() {defaultValue=1, group=ViperGlitch, parameterName=FaceHighlightStrength}
  ) {constB=5}
  n35 = LinearInterpolate(A=@n2.BaseColor, B=@n26, Alpha=@n25)
  n36 = LinearInterpolate(A=@n35, B=Multiply(A=@n35, B=@n20.R), Alpha=@n4) {constB=0}
  n37 = DotProduct(
   A=MaterialFunctionCall() {materialFunction=&CameraDirectionVector}.Result
   B=Transform(None=@n2.Normal)
  )
  n38 = SetMaterialAttributes!(
   MaterialAttributes=@n2.MaterialAttributes
   EmissiveColor=LinearInterpolate(A=@n28, B=Add(A=@n28, B=@n34), Alpha=@n4)
   Roughness=LinearInterpolate(A=@n2.Roughness, B=Constant() {r=1}, Alpha=@n4)
   Specular=LinearInterpolate(A=@n2.Specular, B=Constant(), Alpha=@n4)
   BaseColor=Multiply(A=Add(A=@n36, B=@n24), B=Clamp(None=@n37) {minDefault=0.5})
   Normal=LinearInterpolate(A=@n2.Normal, B=@n29.Result, Alpha=@n25)
  ) {attributeSetTypes [
    #1
    #4
    #5
    #3
    #2
   ]}
  FunctionOutput(None=@n38)
  TextureCoordinate()
  TextureCoordinate()
  n39 = Multiply(
   A=@n5
   B=AppendVector(
    A=ScalarParameter() {defaultValue=1, group=ViperGlitch, parameterName=PerlinNoise_U_Tiling}
    B=ScalarParameter() {defaultValue=1, group=ViperGlitch, parameterName=PerlinNoise_V_Tiling}
   )
  )
  n40 = VectorParameter() {group=ViperGlitch, parameterName=PerlinNoisePanDirectionANdSpeed, defaultValue=<rgba 0 -0.1 0 1>}
  TextureSample(UVs=Panner(Coordinate=Add(A=@n11, B=@n39), Speed=ComponentMask(None=@n40.RGB) {g=true, r=true}) {speedX=0.1, speedY=-0.1}) {texture=/Game/Global/Textures/Default/Noise/3xPerlin_M}
  VertexNormalWS()
 ]}
