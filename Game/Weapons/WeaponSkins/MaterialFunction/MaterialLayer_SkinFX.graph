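# Material layer function for weapon skins that adds a tinted, noise-driven surface effect over incoming material attributes through an Alpha Mask input.
# A tiled noise texture mixes two colours, applied to base colour by a Fresnel on the layer normal, and varies roughness between min and max.
# The layer sets metallic and specular and blends its intensity-scaled normal into the base normal.
# Emissive passes through and also feeds Opacity.
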MaterialFunction /Game/Weapons/WeaponSkins/MaterialFunction/MaterialLayer_SkinFX
graph {nodes [
  n1 = FunctionInput() {inputName=Material, inputType=FunctionInput_MaterialAttributes, sortPriority=-200}
  n2 = GetMaterialAttributes(None=@n1) {attributeGetTypes [
    #1
    #2
    #3
    #4
    #5
    #6
    #7
   ]}
  n3 = VectorParameter() {group=Material Layer SkinFX, parameterName=MaterialLayer_Color, sortPriority=1, defaultValue=<rgba 0.0312 0.0134 0.0094 1>}
  n4 = VectorParameter() {group=Material Layer SkinFX, parameterName=MaterialLayer_Color 2, sortPriority=1, defaultValue=<rgba 0.1719 0.0554 0.0605 1>}
  n5 = Multiply(A=TextureCoordinate(), B=ScalarParameter() {defaultValue=5, group=Material Layer SkinFX, parameterName=MaterialLayer_Texture Tiling})
  n6 = TextureSampleParameter2D!(
   UVs=@n5
   Bias=ScalarParameter() {defaultValue=1, group=Material Layer SkinFX, parameterName=MaterialLayer_Noise Mip Bias}
   Apply View MipBias
  ) {
   group=Material Layer SkinFX
   mipValueMode=TMVM_MipBias
   parameterName=MaterialLayer_Noise
   texture=&T_WhitesRGB
  }
  n7 = ScalarParameter() {group=Material Layer SkinFX, parameterName=MaterialLayer_Blending}
  n8 = TextureSampleParameter2D(UVs=@n5, Bias=ScalarParameter() {defaultValue=1, group=Material Layer SkinFX, parameterName=MaterialLayer_Normal Mip Bias}) {
   group=Material Layer SkinFX
   mipValueMode=TMVM_MipBias
   parameterName=MaterialLayer_Normal
   samplerType=SAMPLERTYPE_Normal
   texture=&FlatNormalBlue
  }
  n9 = Transform(None=@n8.RGB)
  n10 = Saturate(None=OneMinus(None=ScalarParameter() {group=Material Layer SkinFX, parameterName=MaterialLayer_Normal Intensity}))
  n11 = SetMaterialAttributes!(
   MaterialAttributes=@n1
   BaseColor=LinearInterpolate(A=@n2.BaseColor, B=LinearInterpolate(A=@n3.RGB, B=@n4.RGB, Alpha=@n6.R), Alpha=Fresnel(ExponentIn=@n7, Normal=@n9)) {constB=0}
   Metallic=ScalarParameter() {group=Material Layer SkinFX, parameterName=MaterialLayer_Metalness, sliderMax=1, sortPriority=4}
   Specular=ScalarParameter() {defaultValue=0.5, group=Material Layer SkinFX, parameterName=MaterialLayer_Specular, sliderMax=1, sortPriority=3}
   Roughness=LinearInterpolate(
    A=ScalarParameter() {group=Material Layer SkinFX, parameterName=MaterialLayer_Roughness Min, sortPriority=5}
    B=ScalarParameter() {defaultValue=1, group=Material Layer SkinFX, parameterName=MaterialLayer_Roughness Max, sortPriority=5}
    Alpha=@n6.R
   )
   EmissiveColor=@n2.EmissiveColor
   Opacity=@n2.EmissiveColor
   Normal=MaterialFunctionCall(
    BaseNormal=@n2.Normal
    AdditionalNormal=MaterialFunctionCall(Normal=@n8.RGB, Flatness=@n10) {materialFunction=&FlattenNormal}.Result
   ) {materialFunction=&BlendAngleCorrectedNormals}.Result
  ) {attributeSetTypes [
    #1
    #2
    #3
    #4
    #5
    #6
    #7
   ]}
  FunctionOutput(None=BlendMaterialAttributes(A=@n1, B=@n11, Alpha=FunctionInput() {bUsePreviewValueAsDefault=true, inputName=Alpha Mask, inputType=FunctionInput_Scalar, sortPriority=24}))
 ]}
