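# Material function for weapon skins that multiplies the emissive color of incoming material attributes by a scanline pattern.
# The ScanLines function supplies the pattern, with a scalar parameter blending its UVs between the first and second sets.
# A channel-selectable FX mask can confine the effect, with Mask Scanlines blending from unmasked to masked.
# Only emissive is overridden on output; a trailing mask multiply is unconnected.
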MaterialFunction /Game/Weapons/WeaponSkins/MaterialFunction/MultiplyScanlineEmissive
graph {nodes [
  n1 = GetMaterialAttributes(None=FunctionInput() {inputType=FunctionInput_MaterialAttributes}) {attributeGetTypes [#1]}
  n2 = TextureCoordinate()
  n3 = TextureCoordinate() {coordinateIndex=1}
  n4 = ScalarParameter() {group=Scanline, parameterName=Scanline_useSecondaryUVs}
  n5 = Multiply(
   A=@n1.EmissiveColor
   B=MaterialFunctionCall!(ScanLine_Tiling_1, Scanline_PannerSpeed_1, ScanLine_Tiling, Scanline_PannerSpeed, UVs=LinearInterpolate(A=@n2, B=@n3, Alpha=@n4)) {materialFunction=&/ScanLines}.Result
  )
  n6 = ChannelMaskParameter(None=FunctionInput() {bUsePreviewValueAsDefault=true, inputName=FX mask}) {group=Scanline, parameterName=Mask Channel}
  n7 = LinearInterpolate(A=@n1.EmissiveColor, B=@n5, Alpha=@n6.R)
  n8 = ScalarParameter() {group=Scanline, parameterName=Mask Scanlines}
  FunctionOutput(None=SetMaterialAttributes!(MaterialAttributes=@n1.MaterialAttributes, EmissiveColor=LinearInterpolate(A=@n5, B=@n7, Alpha=@n8)) {attributeSetTypes [#1]})
  Multiply(A=@n6.R, B=@n8)
 ]}
