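# Weapon skin material function that adds an animated radiant plasma gradient to incoming base color and sets emissive, blended by Alpha.
# Two panning samples of one plasma noise texture form the pattern, one warping the other's UVs alongside an inverted-AO offset.
# Fresnel times noise indexes a tinted fuchsia gradient, masked by AO, fresnel and a sine-cosine pulse.
# Quality switches and vertex interpolators are present but unconnected to the output.
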MaterialFunction /Game/Weapons/WeaponSkins/MaterialFunction/RadiantPlasma_Gradient
graph {nodes [
  n1 = GetMaterialAttributes(None=FunctionInput() {bUsePreviewValueAsDefault=true, inputName=Material, inputType=FunctionInput_MaterialAttributes}) {attributeGetTypes [#1]}
  n2 = FunctionInput(Preview=Constant() {r=1}) {bUsePreviewValueAsDefault=true, inputName=AO, inputType=FunctionInput_Scalar}
  n3 = OneMinus(None=@n2)
  n4 = ScalarParameter() {defaultValue=0.1, parameterName=AO_OFFset}
  n5 = TextureCoordinate()
  n6 = Panner(
   Coordinate=Multiply(A=@n5, B=ScalarParameter() {defaultValue=0.7, parameterName=GradientWarp_Tiling})
   Speed=AppendVector(
    A=ScalarParameter() {defaultValue=-0.03, parameterName=Sequen_PannerX}
    B=ScalarParameter() {defaultValue=0.04, parameterName=Sequen_PannerY}
   )
  ) {speedY=-0.0125}
  n7 = Panner(
   Coordinate=Multiply(A=@n5, B=ScalarParameter() {defaultValue=0.05, parameterName=GradientWarp_Tiling_2})
   Speed=AppendVector(
    A=ScalarParameter() {defaultValue=0.005, parameterName=Sequen_PannerX_1}
    B=ScalarParameter() {defaultValue=-0.008, parameterName=Sequen_PannerY_1}
   )
  ) {speedY=-0.0125}
  n8 = TextureObject() {texture=/Game/Effects/Fort_Effects/Textures/Plasma/T_PlasmaNoise_A_01}
  n9 = TextureSample(UVs=@n7, Tex=@n8) {texture=&T_PlasmaNoise_A_01}
  n10 = Multiply(A=@n9.G, B=ScalarParameter() {defaultValue=0.5, parameterName=UVWobble2}) {constB=0.05}
  n11 = TextureSample(UVs=Add(A=Multiply(A=@n3, B=@n4), B=Add(A=@n6, B=@n10)), Tex=@n8) {texture=&T_PlasmaNoise_A_01}
  n12 = Multiply(A=Time(), B=ScalarParameter() {defaultValue=0.2, parameterName=SequenFlux})
  n13 = Multiply(A=@n12) {constB=0.7}
  n14 = Multiply(A=Multiply(A=Cosine(None=@n13)) {constB=0.5}, B=Sine(None=@n12))
  n15 = Multiply(A=@n9.B, B=Multiply(A=ScalarParameter() {defaultValue=3, parameterName=SequenFluxINT}, B=Abs(None=@n14)))
  n16 = LinearInterpolate(A=ScalarParameter() {defaultValue=0.25, parameterName=GradMin}, Alpha=Add(A=@n11.R, B=Saturate(None=@n15)))
  n17 = ScalarParameter() {defaultValue=2, parameterName=FresnelMax}
  n18 = Fresnel(ExponentIn=ScalarParameter() {defaultValue=0.6596, group=Sequen, parameterName=SequenFres_EX}, Normal=PixelNormalWS()) {baseReflectFraction=0}
  n19 = Clamp(None=@n18)
  n20 = Multiply(A=Multiply(A=@n2, B=Saturate(None=@n16)), B=Clamp(None=LinearInterpolate(B=@n17, Alpha=@n19)))
  n21 = Multiply(A=@n18, B=ScalarParameter() {defaultValue=0.15, parameterName=GradScale})
  n22 = ComponentMask(None=@n11.RGB) {g=true, r=true}
  n23 = Saturate(None=Multiply(A=Multiply(A=@n21, B=@n22)) {constB=7})
  n24 = TextureObjectParameter() {parameterName=FuschiaGradient, texture=/Game/Characters/Player/Male/Medium/Bodies/M_Med_Soldier_04/Skins/BR_30/Textures/T_FuschiaGrad}
  n25 = Multiply(
   A=@n20
   B=Multiply(
    A=TextureSample(UVs=@n23, Tex=@n24) {texture=&T_FuschiaGrad}.RGB
    B=VectorParameter() {group=Sequen, parameterName=FresnelTint, defaultValue=<rgba 1 1 1 1>}.RGB
   )
  )
  n26 = Saturate(None=Add(A=@n1.BaseColor, B=@n25))
  n27 = ScalarParameter() {defaultValue=1, parameterName=Sequen_EmissiveINT}
  n28 = Saturate(None=Multiply(A=@n20, B=Multiply(A=@n25, B=@n27)))
  FunctionOutput(None=BlendMaterialAttributes(
    A=@n1.MaterialAttributes
    B=SetMaterialAttributes!(MaterialAttributes=@n1.MaterialAttributes, BaseColor=@n26, EmissiveColor=@n28) {attributeSetTypes [#1, #2]}
    Alpha=FunctionInput() {bUsePreviewValueAsDefault=true, inputName=Alpha, inputType=FunctionInput_Scalar}
   ))
  n29 = Multiply(
   A=@n2
   B=TextureSample(Tex=@n24) {texture=&T_FuschiaGrad}.RGB
  )
  QualitySwitch(Default=@n26, Low=@n29, High=@n26, Medium=@n26)
  n30 = Multiply(A=@n27, B=@n29)
  QualitySwitch(Default=@n28, Low=@n30, High=@n28, Medium=@n30)
  VertexInterpolator(VS=@n7)
  VertexInterpolator(VS=@n6)
 ]}
