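# Material function for weapon skins that blends an animated rainbow hue shift into the input material's base color, weighted by Alpha.
# HueShift recolors either the diffuse or a Hue parameter, driven by time plus a pre-skinned, bounds-scaled axis gradient (vertex-interpolated) or tiled UVs.
# An optional flow map offsets the gradient.
# Low quality and mobile shading use a uniform time-only shift, skipping gradient and flow map.
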MaterialFunction /Game/Weapons/WeaponSkins/MaterialFunction/RainbowHue
graph {nodes [
  n1 = ObjectBounds()
  n2 = ScalarParameter() {defaultValue=3, group=RainbowHue, parameterName=World Aligned Gradient Softness}
  n3 = StaticComponentMaskParameter(None=Divide(A=PreSkinnedPosition(), B=Multiply(A=@n1, B=@n2))) {defaultR=true, group=RainbowHue, parameterName=World Aligned Gradient Axis}
  n4 = TextureCoordinate()
  n5 = ScalarParameter() {group=RainbowHue, parameterName=HueEffect Tiling}
  n6 = Frac(None=Multiply(A=Time(), B=ScalarParameter() {defaultValue=0.1, group=RainbowHue, parameterName=HueEffect Panning Speed}))
  n7 = Add(
   A=StaticSwitchParameter(True=VertexInterpolator(VS=@n3).PS, False=Multiply(A=@n4, B=@n5)) {defaultValue=true, group=RainbowHue, parameterName=Use World Aligned Gradient for HueEffect}
   B=ComponentMask(None=OneMinus(None=@n6)) {r=true}
  )
  n8 = TextureSampleParameter2D(UVs=Multiply(A=TextureCoordinate(), B=ScalarParameter() {group=RainbowHue, parameterName=HueEffect Flowmap Tiling})) {
   group=RainbowHue
   parameterName=RainbowHue Flow Map
   samplerSource=SSM_Wrap_WorldGroupSettings
   samplerType=SAMPLERTYPE_LinearColor
   texture=/Game/VisualThreatManager/StormVisuals/Test/SpawnParticles/Streamers/NewFolder/SampleFlow
  }
  n9 = AppendVector(A=@n8.R, B=@n8.R)
  n10 = ScalarParameter() {defaultValue=5, group=RainbowHue, parameterName=HueEffect Flowmap Intensity}
  n11 = StaticSwitchParameter(True=Add(A=ComponentMask(None=@n7) {r=true}, B=Multiply(A=@n9, B=@n10)), False=@n7) {group=RainbowHue, parameterName="UseFlowMap for HueEffect "}
  n12 = FunctionInput() {inputName=Material, inputType=FunctionInput_MaterialAttributes}
  n13 = GetMaterialAttributes(None=@n12) {attributeGetTypes [#1]}
  n14 = VectorParameter() {group=RainbowHue, parameterName=Hue, defaultValue=<rgba 1 0.2514 0.516 1>}
  n15 = Multiply(A=Time(), B=ScalarParameter() {defaultValue=0.1, group=RainbowHue, parameterName=HueEffect Panning Speed_2})
  n16 = MaterialFunctionCall(
   Hue Shift Percentage=OneMinus(None=Frac(None=@n15))
   Texture=StaticSwitchParameter(True=@n13.BaseColor, False=@n14.RGB) {group=RainbowHue, parameterName=Use Diffuse for Color}
  ) {materialFunction=&HueShift}
  n17 = ShadingPathSwitch(
   Default=StaticSwitchParameter(
    True=MaterialFunctionCall(Hue Shift Percentage=@n11, Texture=@n13.BaseColor) {materialFunction=&HueShift}.Result
    False=MaterialFunctionCall(Hue Shift Percentage=@n11, Texture=@n14.RGB) {materialFunction=&HueShift}.Result
   ) {group=RainbowHue, parameterName=Use Diffuse for Color}
   Mobile=@n16.Result
  ) {desc=Migrated from FeatureLevelSwitch}
  n18 = QualitySwitch(Default=@n17, Low=@n16.Result, High=@n17, Medium=@n17)
  FunctionOutput(None=BlendMaterialAttributes(
    A=@n12
    B=SetMaterialAttributes!(MaterialAttributes=@n12, BaseColor=@n18) {attributeSetTypes [#1]}
    Alpha=FunctionInput() {inputName=Alpha, inputType=FunctionInput_Scalar}
   ))
 ]}
