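# Material function for weapon skins that blends a black, glowing gradient-fill effect into incoming material attributes through an Alpha input.
# A gradient input, defaulting to a texture mask, lights up below a threshold that is either a parameter or normalized object world height.
# The emissive tint mixes in animated scanlines and fades by Fresnel.
# Alpha also scales the glow, while base color and roughness are overridden across the whole blended area.
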MaterialFunction /Game/Weapons/WeaponSkins/MaterialFunction/ScalarDrivenGradient
graph {nodes [
  Add()
  n1 = FunctionInput() {inputType=FunctionInput_MaterialAttributes}
  n2 = Constant()
  n3 = VectorParameter() {group=ScalarDriven, parameterName=Scalar Emissive Color, defaultValue=<rgba 1 0 0.5342 1>}
  n4 = LinearInterpolate(
   A=@n3.RGB
   B=Multiply(A=@n3.RGB) {constB=0.2}
   Alpha=MaterialFunctionCall!(ScanLine_Tiling_1, Scanline_PannerSpeed_1, ScanLine_Tiling, Scanline_PannerSpeed, UVs) {materialFunction=&/ScanLines}.Result
  )
  n5 = ScalarParameter() {defaultValue=0.5, group=ScalarDriven, parameterName=Scanline Intensity}
  n6 = Constant() {r=0.05}
  n7 = FunctionInput(Preview=TextureSampleParameter2D() {group=ScalarDriven, parameterName=ScalarGradient, texture=/Game/Global/Textures/Default/Masks_and_Gradients/CylGradient}.R) {bUsePreviewValueAsDefault=true, inputName=Gradient, inputType=FunctionInput_Scalar}
  n8 = Add(A=ComponentMask(None=ObjectPositionWS()) {b=true}, B=ScalarParameter() {defaultValue=50, group=ScalarDriven, parameterName=add})
  n9 = ScalarParameter() {defaultValue=6500, group=ScalarDriven, parameterName=Div}
  n10 = LinearInterpolate(
   A=ScalarParameter() {group=ScalarDriven, parameterName=Gradient Threshold}
   B=Clamp(None=Divide(A=@n8, B=@n9), Max=Constant() {r=1})
   Alpha=ScalarParameter() {group=ScalarDriven, parameterName=UseAltitude}
  )
  n11 = Constant()
  n12 = Constant() {r=1}
  n13 = FunctionInput() {bUsePreviewValueAsDefault=true, inputName=Alpha, inputType=FunctionInput_Scalar}
  n14 = LinearInterpolate(
   B=LinearInterpolate(A=LinearInterpolate(A=@n3.RGB, B=@n4, Alpha=@n5), Alpha=Fresnel(ExponentIn=@n6) {baseReflectFraction=0}) {constB=0}
   Alpha=Multiply(A=If(A=@n7, B=@n10, "A > B"=@n11, "A < B"=@n12), B=@n13)
  )
  n15 = Constant() {r=0.5}
  FunctionOutput(None=BlendMaterialAttributes(
    A=@n1
    B=SetMaterialAttributes!(MaterialAttributes=@n1, BaseColor=@n2, EmissiveColor=@n14, Roughness=@n15) {attributeSetTypes [#1, #2, #3]}
    Alpha=@n13
   ))
 ]}
