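# Material function that layers cosmetic skin effects, from gradients and galaxy to slurp and glitter, onto character Material Attributes.
# One SkinFX_Mask texture, sampled through a glitch subfunction, feeds the effects, mostly via selectable channel parameters, and is also output.
# Effects chain in series, most behind static switches, with some always called, ending in cloth-sim WPO.
# Custom Numbers and Rotator branches do not reach the output.
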MaterialFunction /Game/Packages/Fortress_SharedMaterials/Base_Material_Functions/BaseCharacterShaderFunctions/CharacterSkinFX
graph {nodes [
  n1 = MaterialFunctionCall!(In=NamedRerouteDeclaration(None=TextureObjectParameter() {group=Skin FX, parameterName=SkinFX_Mask}) {name=SkinFX_Mask, nodeColor=<rgba 0.9842 0 1 1>}) {materialFunction=/Game/Weapons/WeaponSkins/MaterialFunction/MF_GlitchSubfunction}
  n2 = StaticSwitchParameter(True=@n1.Glitched, False=@n1.Glitched) {group=Skin FX, parameterName=UseGlitch}
  n3 = ScalarParameter() {
   desc=Set to 1 if the Red Channel in the FX Mask should be counted for Hologram masking in the FE
   group=HologramMasking
   parameterName=UseSkinFXMaskForOpacity
  }
  n4 = NamedRerouteDeclaration(None=LinearInterpolate(A=@n2, B=@n2, Alpha=@n3)) {name=SkinFX_Mask_Sampled, nodeColor=<rgba 0.7475 0 1 1>}
  n5 = MaterialFunctionCall(Float4=StaticComponentMaskParameter(None=@n4) {defaultG=true, group=Skin FX, parameterName=GmapSkinCustomization_Channel}) {materialFunction=&BreakOutFloat4Components}
  n6 = Max(A=Max(A=@n5.R, B=@n5.R), B=@n5.B)
  n7 = NamedRerouteDeclaration(None=FunctionInput() {bUsePreviewValueAsDefault=true, inputName=Gmap, inputType=FunctionInput_Scalar}) {name=Gmap, nodeColor=<rgba 0.7578 1 0 1>}
  n8 = MaterialFunctionCall!(
   FurOpacityMaskTexture=NamedRerouteUsage()
   In=FunctionInput() {bUsePreviewValueAsDefault=true, inputName=MaterialAttributes, inputType=FunctionInput_MaterialAttributes}
  ) {materialFunction=/Game/Weapons/WeaponSkins/MaterialFunction/MF_FurNormalAndUVBlend}
  n9 = StaticSwitchParameter(
   True=MaterialFunctionCall!(Material=@n8.Result, Alpha=Max(A=@n6, B=@n5.A), Gmap=@n7) {materialFunction=/Game/Weapons/WeaponSkins/MaterialFunction/GmapGradientLayers}.Result
   False=@n8.Result
  ) {group=Skin FX, parameterName=useGmapGradientLayers}
  n10 = MaterialFunctionCall!(Local Normal, Local Position, Material=@n9) {materialFunction=/Game/Packages/Fortress_SharedMaterials/Base_Material_Functions/MF_SuperLevellingMask}
  n11 = StaticSwitchParameter(
   True=MaterialFunctionCall!(Material=@n10.Result, Alpha=StaticComponentMaskParameter(None=@n4) {defaultG=true, group=Skin FX, parameterName=DiscoLights_Channel}) {materialFunction=/Game/Weapons/WeaponSkins/MaterialFunction/DiscoFloorFunction}.Result
   False=@n10.Result
  ) {group=Skin FX, parameterName=Use Disco Lights}
  n12 = StaticSwitchParameter(
   True=MaterialFunctionCall!(
    Alpha=StaticComponentMaskParameter(None=@n4) {defaultR=true, group=Skin FX, parameterName=Sequen_Channel}
    AO=StaticComponentMaskParameter(None=@n4) {defaultG=true, group=Skin FX, parameterName=SequenAO_Channel}
    Material=@n11
   ) {materialFunction=/Game/Weapons/WeaponSkins/MaterialFunction/RadiantPlasma_Gradient}.Result
   False=@n11
  ) {group=Skin FX, parameterName=Use Sequen Fresnel}
  n13 = StaticSwitchParameter(
   True=MaterialFunctionCall!(
    Alpha=StaticComponentMaskParameter(None=@n4) {defaultR=true, group=Skin FX, parameterName=PanningEmissive_Channel}
    isVehicle
    Gmap=@n7
    UVOffset=StaticComponentMaskParameter(None=@n4) {defaultB=true, group=Skin FX, parameterName=PanningEmissive_UVOffset_Channel}
    Local Normal
    Local Position
    EmissiveColorOverride
    Material=@n12
   ) {materialFunction=/Game/Weapons/WeaponSkins/MaterialFunction/PanningEmissive}.Result
   False=@n12
  ) {group=Skin FX, parameterName=Use Panning Emissive}
  n14 = StaticSwitchParameter(
   True=MaterialFunctionCall!(GMap=@n7, Material=@n13) {materialFunction=/Game/Weapons/WeaponSkins/MaterialFunction/CustomizeableColors}.Result
   False=@n13
  ) {group=Skin FX, parameterName=UseCustomColors}
  n15 = StaticSwitchParameter(
   True=MaterialFunctionCall!(Alpha=StaticComponentMaskParameter(None=@n4) {defaultR=true, group=Skin FX, parameterName=TinyComputer_Channel}, Material=@n14) {materialFunction=/Game/Characters/Player/Male/Medium/Bodies/M_MED_Astronaut_01/Materials/BackpackScreen}.Result
   False=@n14
  ) {group=Skin FX, parameterName=Small Computer}
  n16 = StaticSwitchParameter(
   True=MaterialFunctionCall!(
    Alpha=StaticComponentMaskParameter(None=@n4) {defaultR=true, group=Skin FX, parameterName=PatternEmissive_Channel}
    PanningEmissiveBrightnessAdjustment
    Material=@n15
   ) {materialFunction=/Game/Weapons/WeaponSkins/MaterialFunction/PatternEmissiveFresnelSequence}.Result
   False=@n15
  ) {group=Skin FX, parameterName=UsePatternEmissive}
  n17 = StaticSwitchParameter(
   True=MaterialFunctionCall!(Alpha=StaticComponentMaskParameter(None=@n4) {defaultR=true, group=Skin FX, parameterName=Cloth Channel}, Gmap=@n7, Material=@n16) {materialFunction=/Game/Weapons/WeaponSkins/MaterialFunction/ClothFuzz}.Result
   False=@n16
  ) {group=Skin FX, parameterName=Use Cloth fuzz}
  n18 = StaticSwitchParameter(
   True=MaterialFunctionCall!(Alpha=ChannelMaskParameter(None=@n4) {group=Skin FX, parameterName=Silk Channel}.R, Material=@n17) {materialFunction=/Game/Weapons/WeaponSkins/MaterialFunction/MF_Silk}.Result
   False=@n17
  ) {group=Skin FX, parameterName=Use Silk}
  n19 = StaticSwitchParameter(
   True=MaterialFunctionCall!(Alpha=StaticComponentMaskParameter(None=@n4) {defaultR=true, group=Skin FX, parameterName=Rainbow Hue Channel}, Material=@n18) {materialFunction=/Game/Weapons/WeaponSkins/MaterialFunction/RainbowHue}.Result
   False=@n18
  ) {group=Skin FX, parameterName=Use Rainbow Hue}
  n20 = StaticSwitchParameter(True=MaterialFunctionCall!(Gmap=@n7, Material=@n19) {materialFunction=/Game/Weapons/WeaponSkins/MaterialFunction/Banana}.Result, False=@n19) {group=Skin FX, parameterName=Use Banana}
  n21 = StaticSwitchParameter(
   True=MaterialFunctionCall!(Alpha=StaticComponentMaskParameter(None=@n4) {defaultR=true, group=Skin FX, parameterName=Clear Coat Channel}, In=@n20) {materialFunction=/Game/Weapons/WeaponSkins/MaterialFunction/Clearcoat}.Result
   False=@n20
  ) {group=Skin FX, parameterName=Use Clear Coat}
  n22 = StaticSwitchParameter(
   True=MaterialFunctionCall!(Alpha=StaticComponentMaskParameter(None=@n4) {defaultB=true, group=Skin FX, parameterName=Equalizer Channel}, In=@n21) {materialFunction=/Game/Weapons/WeaponSkins/MaterialFunction/MF_Equalizer}.Result
   False=@n21
  ) {group=Skin FX, parameterName=Use Equalizer}
  n23 = StaticSwitchParameter(
   True=MaterialFunctionCall!(
    In=@n22
    Warp Mask=StaticComponentMaskParameter(None=@n4) {defaultR=true, group=Skin FX, parameterName=Galaxy Warp Mask Channel}
    GalaxyAlternateFresnel
    Alpha=StaticComponentMaskParameter(None=@n4) {defaultR=true, group=Skin FX, parameterName=Galaxy Channel}
   ) {materialFunction=/Game/Weapons/WeaponSkins/MaterialFunction/Galaxy}.Result
   False=@n22
  ) {group=Skin FX, parameterName=Use Galaxy}
  n24 = StaticSwitchParameter(
   True=MaterialFunctionCall!(In=@n23) {materialFunction=&/AsymetricalUVOffset}.Result
   False=@n23
  ) {group=Skin FX, parameterName=Use asymetrical UV offset}
  n25 = StaticSwitchParameter(
   True=MaterialFunctionCall!(Alpha=ChannelMaskParameter(None=@n4) {group=Skin FX, parameterName=Sparkle Remix Atomic}.R, Material=@n24) {materialFunction=/Game/Weapons/WeaponSkins/MaterialFunction/MF_SparkleRemix_Atomic}.Result
   False=@n24
  ) {group=Skin FX, parameterName=Use Sparkle Remix Atomic}
  n26 = StaticSwitchParameter(
   True=MaterialFunctionCall!(Alpha=ChannelMaskParameter(None=@n4) {group=Skin FX, parameterName=Sparkle Remix Burst}.R, Material=@n25) {materialFunction=/Game/Weapons/WeaponSkins/MaterialFunction/MF_SparkleRemix_Burst}.Result
   False=@n25
  ) {group=Skin FX, parameterName=Use Sparkle Remix Burst}
  n27 = StaticSwitchParameter(
   True=MaterialFunctionCall!(Alpha=StaticComponentMaskParameter(None=@n4) {defaultR=true, group=Skin FX, parameterName=Lightning Channel}, Material=@n26) {materialFunction=/Game/Weapons/WeaponSkins/MaterialFunction/Lightning}.Result
   False=@n26
  ) {group=Skin FX, parameterName=Use Lightning}
  n28 = StaticSwitchParameter(
   True=MaterialFunctionCall!(
    Alpha=ChannelMaskParameter(None=@n4) {group=Skin FX, parameterName=Simple Threshold Channel}.R
    Sequence=ChannelMaskParameter(None=@n4) {group=Skin FX, parameterName=Simple Threshold Sequence Channel}.R
    Material=@n27
   ) {materialFunction=/Game/Weapons/WeaponSkins/MaterialFunction/SimpleThresholdEmissive}.Result
   False=@n27
  ) {group=Skin FX, parameterName=Use Simple Threshold}
  n29 = StaticSwitchParameter(
   True=MaterialFunctionCall!(
    Edge Map=StaticComponentMaskParameter(None=@n4) {defaultR=true, group=Skin FX, parameterName=Crystal Edge Map Channel}
    Material=@n28
    Alpha=StaticComponentMaskParameter(None=@n4) {defaultR=true, group=Skin FX, parameterName=Crystal Channel}
   ) {materialFunction=/Game/Weapons/WeaponSkins/MaterialFunction/Crystal}.Result
   False=@n28
  ) {group=Skin FX, parameterName=Use Crystal}
  n30 = StaticSwitchParameter(
   True=MaterialFunctionCall!(
    MASK=StaticComponentMaskParameter(None=@n4) {defaultR=true, group=Skin FX, parameterName=Generic Gem Effect Channel}
    Color Change Sequence=StaticComponentMaskParameter(None=@n4) {defaultB=true, group=Skin FX, parameterName=Generic Gem Color Change Channel}
    Material=@n29
   ) {materialFunction=/Game/Weapons/WeaponSkins/MaterialFunction/GenericGem}.Result
   False=@n29
  ) {group=Skin FX, parameterName=Use Generic Gem}
  n31 = StaticComponentMaskParameter(None=@n4) {defaultR=true, group=Skin FX, parameterName=Ice Channel}
  n32 = StaticComponentMaskParameter(None=@n4) {defaultR=true, group=Skin FX, parameterName=Ice Edge Channel}
  n33 = MaterialFunctionCall!(
   Alpha=ChannelMaskParameter(None=@n4) {group=Skin FX, parameterName=BlackIceMaskChannel}.R
   In=StaticSwitchParameter(
    True=MaterialFunctionCall!(Material=@n30, EdgeMap=@n32, Alpha=@n31) {materialFunction=/Game/Weapons/WeaponSkins/MaterialFunction/Ice}.Result
    False=@n30
   ) {group=Skin FX, parameterName=Use Ice}
  ) {materialFunction=/Game/Weapons/WeaponSkins/MaterialFunction/Ice_V2}
  n34 = StaticSwitchParameter(
   True=MaterialFunctionCall!(Materials=@n33.Result) {materialFunction=/Game/Weapons/WeaponSkins/MaterialFunction/CustomCrests}.Result
   False=@n33.Result
  ) {group=Skin FX, parameterName=Use Custom Crest}
  n35 = StaticSwitchParameter(
   True=MaterialFunctionCall!(Alpha=StaticComponentMaskParameter(None=@n4) {defaultR=true, group=Skin FX, parameterName=ThinFilm_Channel}, Material Attributes=@n34) {materialFunction=/Game/Weapons/WeaponSkins/MaterialFunction/ThinFilmFunction}.Result
   False=@n34
  ) {group=Skin FX, parameterName=Use Thin Film}
  n36 = StaticComponentMaskParameter(None=@n4) {defaultR=true, group=Skin FX, parameterName=Flipbook Texture Channel}
  n37 = StaticSwitchParameter(
   True=MaterialFunctionCall!(Alpha=StaticComponentMaskParameter(None=@n4) {defaultG=true, group=Skin FX, parameterName=ScalarDrivenGradient Channel}, Gradient, In=@n35) {materialFunction=/Game/Weapons/WeaponSkins/MaterialFunction/ScalarDrivenGradient}.Result
   False=StaticSwitchParameter(
    True=MaterialFunctionCall!(Alpha=@n36, Material=@n35) {materialFunction=/Game/Weapons/WeaponSkins/MaterialFunction/SubUVTextures}.Result
    False=@n35
   ) {group=Skin FX, parameterName=Use Sub UV texture}
  ) {group=Skin FX, parameterName=UseScalarDrivenGradient}
  n38 = StaticSwitchParameter(
   True=MaterialFunctionCall!(Alpha=ChannelMaskParameter(None=@n4) {group=Skin FX, parameterName=Simple Parallax Channel}.R, Material=@n37) {materialFunction=/Game/Weapons/WeaponSkins/MaterialFunction/MF_GenericParallax}.Result
   False=@n37
  ) {group=Skin FX, parameterName=Use Generic Parallax}
  n39 = StaticSwitchParameter(
   True=MaterialFunctionCall!(
    In=@n38
    Dual SS Secondary Alpha=StaticComponentMaskParameter(None=@n4) {defaultG=true, group=Skin FX, parameterName=SS Masks: Dual SS Secondary Mask}
    Alpha=StaticComponentMaskParameter(None=@n4) {defaultR=true, group=Skin FX, parameterName=SS Masks}
   ) {materialFunction=/Game/Weapons/WeaponSkins/MaterialFunction/SS_Shader_Model}.Result
   False=@n38
  ) {group=Skin FX, parameterName="Uses Subsurface Shading Model (NOT Subsurface Profile)"}
  n40 = FunctionInput() {bUsePreviewValueAsDefault=true, inputName=AO, inputType=FunctionInput_Scalar, sortPriority=1, previewValue=<1 1 1 1>}
  n41 = NamedRerouteDeclaration(None=@n40) {name=AO, nodeColor=<rgba 1 0 0.3984 1>}
  n42 = StaticSwitchParameter(
   True=MaterialFunctionCall!(
    In=@n39
    Alpha=StaticComponentMaskParameter(None=@n4) {defaultR=true, group=Skin FX, parameterName="Channel Embracing T H E C U B E "}
    AO=@n41
   ) {materialFunction=/Game/Weapons/WeaponSkins/MaterialFunction/TheCUBE}.Result
   False=@n39
  ) {group=Skin FX, parameterName=Use Cube Effect}
  n43 = StaticSwitchParameter(
   True=MaterialFunctionCall!(Material=@n42, Alpha=StaticComponentMaskParameter(None=@n4) {defaultR=true, group=Skin FX, parameterName=ChameleonColorShiftMask}) {materialFunction=/Game/Weapons/WeaponSkins/MaterialFunction/MF_ChameleonColorShift}.Result
   False=@n42
  ) {group=Skin FX, parameterName=ChameleonColorShift}
  n44 = StaticSwitchParameter(
   True=MaterialFunctionCall!(Mask=@n4, Material=@n43) {materialFunction=/Game/Weapons/WeaponSkins/MaterialFunction/MF_SnowFall}.Result
   False=@n43
  ) {group=Skin FX, parameterName=UseSnowFall}
  n45 = StaticSwitchParameter(
   True=MaterialFunctionCall!(Mask=@n4, Material=@n44) {materialFunction=/Game/Weapons/WeaponSkins/MaterialFunction/MF_Constellation}.Result
   False=@n44
  ) {group=Skin FX, parameterName=UseConstellation}
  n46 = StaticSwitchParameter(
   True=MaterialFunctionCall!(
    FX_Mask_Vectoir=@n4
    MA_Input=@n45
    FX_MaskInput=StaticComponentMaskParameter(None=@n4) {defaultG=true, group=Skin FX, parameterName=BlinkLight_Channel}
   ) {materialFunction=/Game/Weapons/WeaponSkins/MaterialFunction/MF_BlinkLight}.MA_Output
   False=@n45
  ) {group=Skin FX, parameterName=Use Blink Light}
  n47 = StaticSwitchParameter(
   True=MaterialFunctionCall!(FX mask=@n4, In=@n46) {materialFunction=/Game/Weapons/WeaponSkins/MaterialFunction/MultiplyScanlineEmissive}.Result
   False=@n46
  ) {group=Skin FX, parameterName=Use Scan Line Emissive}
  n48 = StaticSwitchParameter(
   True=MaterialFunctionCall!(Material=@n47, Alpha Mask=StaticComponentMaskParameter(None=@n4) {defaultR=true, group=Skin FX, parameterName=Material Layer Channel}) {materialFunction=/Game/Weapons/WeaponSkins/MaterialFunction/MaterialLayer_SkinFX}.Result
   False=@n47
  ) {
   desc="Use this for a simple material to layer on top of the primary material. \r\n\r\nUseful for Dirt, Grime, etc."
   group=Skin FX
   parameterName=Use Material Layer
  }
  n49 = StaticSwitchParameter(
   True=MaterialFunctionCall!(
    Grime=StaticComponentMaskParameter(None=@n4) {defaultR=true, group=Skin FX, parameterName=Grime_Channel}
    Scratch=StaticComponentMaskParameter(None=@n4) {defaultR=true, group=Skin FX, parameterName=Scratch_Channel}
    Gmap=@n7
    Material=@n48
   ) {materialFunction=/Game/Weapons/WeaponSkins/MaterialFunction/GrimeScratch}.Result
   False=@n48
  ) {
   desc="Use this for a simple material to layer on top of the primary material. \r\n\r\nUseful for Dirt, Grime, etc."
   group=Skin FX
   parameterName=Use Grime And Scratch
  }
  n50 = StaticSwitchParameter(
   True=MaterialFunctionCall!(Alpha=StaticComponentMaskParameter(None=@n4) {defaultR=true, group=Skin FX, parameterName=HeightDisplacement_Channel}, Materials=@n49) {materialFunction=/Game/Weapons/WeaponSkins/MaterialFunction/DisplacementHeight}.Result
   False=@n49
  ) {group=Skin FX, parameterName=useHeightDisplacement}
  n51 = StaticSwitchParameter(
   True=MaterialFunctionCall!(Material=@n50, In=StaticComponentMaskParameter(None=@n4) {defaultR=true, group=Skin FX, parameterName=OpacityMask_Channel}) {materialFunction=/Game/Weapons/WeaponSkins/MaterialFunction/AlphaMask}.Result
   False=@n50
  ) {group=Skin FX, parameterName=useOpacityMask}
  n52 = StaticSwitchParameter(
   True=MaterialFunctionCall!(
    Material=@n51
    MASK=StaticComponentMaskParameter(None=@n4) {defaultR=true, group=Skin FX, parameterName=Holographic Neon Lines Effect Channel}
   ) {materialFunction=/Game/Weapons/WeaponSkins/MaterialFunction/Holographic_Neon_Lines}.Result
   False=@n51
  ) {group=Skin FX, parameterName=Use Holographic Neon Lines}
  n53 = StaticSwitchParameter(
   True=MaterialFunctionCall!(MaskInput=@n4, Preskinned-Position, Bounds, GMap, Material=@n52) {materialFunction=/Game/Weapons/WeaponSkins/MaterialFunction/MF_SkeletonHunterCoat}.Result
   False=@n52
  ) {group=Skin FX, parameterName=UseSkeletonHunterCoatFX}
  n54 = StaticSwitchParameter(
   True=MaterialFunctionCall!(Gmap=@n7, Materials=@n53) {materialFunction=/Game/Weapons/WeaponSkins/MaterialFunction/PlayerBanner}.Result
   False=@n53
  ) {group=Skin FX, parameterName=Use PlayerBanner}
  n55 = StaticSwitchParameter(
   True=MaterialFunctionCall!(Mask=@n4, Material=@n54) {materialFunction=/Game/Weapons/WeaponSkins/MaterialFunction/MF_SleepyTime}.Result
   False=@n54
  ) {group=Skin FX, parameterName=UseSleepyTime}
  n56 = FunctionInput() {inputName=PreskinnedPosition}
  n57 = StaticSwitchParameter(
   True=MaterialFunctionCall!(Preskinned_Input=NamedRerouteDeclaration(None=@n56) {name=PreskinnedPosition, nodeColor=<rgba 0 0 1 1>}, Material=@n55) {materialFunction=/Game/Characters/Player/Male/Medium/Bodies/M_MED_Plum_Retro/Skin/Plum_Retro_Reactive/Materials/MF_CapeWind}.Result
   False=@n55
  ) {group=Skin FX, parameterName=UseCapeWind}
  n58 = StaticSwitchParameter(
   True=MaterialFunctionCall!(Mask=@n4, Material=@n57) {materialFunction=/Game/Weapons/WeaponSkins/MaterialFunction/MF_Haunt}.Result
   False=@n57
  ) {group=Skin FX, parameterName=UseViperZero}
  n59 = StaticSwitchParameter(
   True=MaterialFunctionCall!(Material=@n58, GMapInput=@n7, AO_Input=@n41, Mask=@n4) {materialFunction=/Game/Weapons/WeaponSkins/MaterialFunction/MF_ViperGlitch}.Result
   False=@n58
  ) {group=Skin FX, parameterName=UseViperGlitch}
  n60 = StaticComponentMaskParameter(None=@n4) {defaultR=true, group=Skin FX, parameterName=Slurp Channel}
  n61 = StaticComponentMaskParameter(None=@n4) {defaultG=true, group=Skin FX, parameterName=Slurp Bubbles Channel}
  n62 = StaticComponentMaskParameter(None=@n4) {defaultB=true, group=Skin FX, parameterName=Slurp Normal Distortion Channel}
  n63 = MaterialFunctionCall!(
   In=StaticSwitchParameter(
    True=MaterialFunctionCall!(Alpha=@n60, AO=@n41, BubblesMask=@n61, SlurpNormalDistortionMask=@n62, Pre Skinned Position, Material=@n59) {materialFunction=/Game/Weapons/WeaponSkins/MaterialFunction/MF_Slurp}.Result
    False=@n59
   ) {group=Skin FX, parameterName=UseSlurp}
   Mask=StaticComponentMaskParameter(None=@n4) {defaultB=true, group=Skin FX, parameterName=TwistedGoo}
   isVehicle
   isMech
  ) {materialFunction=/Game/Characters/Player/Male/Medium/Bodies/M_MED_Mastermind/Materials/MF_TwistedGoo}
  n64 = StaticSwitchParameter(
   True=MaterialFunctionCall!(Mask=StaticComponentMaskParameter(None=@n4) {defaultG=true, group=Skin FX, parameterName=PanAndDistort Channel}, Material=@n63.Result) {materialFunction=/Game/Weapons/WeaponSkins/MaterialFunction/MF_PanAndDistort}.Result
   False=@n63.Result
  ) {group=Skin FX, parameterName=Use PanAndDistort}
  n65 = StaticSwitchParameter(
   True=MaterialFunctionCall!(Mask=@n4, Material=@n64) {materialFunction=/Game/Weapons/WeaponSkins/MaterialFunction/MF_DistortionFresnel}.Result
   False=@n64
  ) {group=Skin FX, parameterName=UseDistortionFresnel}
  n66 = StaticSwitchParameter(
   True=MaterialFunctionCall!(SuperLevellingMask=@n10.SuperLevelling_Mask, SkinFXMask=@n4, Material=@n65) {materialFunction=/Game/Packages/Fortress_SharedMaterials/Base_Material_Functions/MF_SimpleMaterial}.Result
   False=@n65
  ) {group=Skin FX, parameterName=UseSimpleMaterial}
  n67 = StaticSwitchParameter(
   True=MaterialFunctionCall(Mask=@n4, Materials=@n66) {materialFunction=&MF_MetalsLUT}.Result
   False=@n66
  ) {group=Skin FX, parameterName=UseMetalLUT}
  n68 = StaticSwitchParameter(
   True=MaterialFunctionCall!(FXMask=NamedRerouteUsage(), MaterialAttributes=@n67) {materialFunction=/Game/Weapons/WeaponSkins/MaterialFunction/MF_Tattoo}.Result
   False=@n67
  ) {group=Skin FX, parameterName=UseTattoo}
  n69 = MaterialFunctionCall!(Mask=NamedRerouteUsage(), TechArtMask, inMat=@n68) {materialFunction=/Game/Weapons/WeaponSkins/MaterialFunction/MF_Sequins}
  n70 = MaterialFunctionCall!(
   In=MaterialFunctionCall!(In=StaticSwitchParameter(True=@n69.Result, False=@n68) {group=Skin FX, parameterName=UseSequins}) {materialFunction=/Game/Weapons/WeaponSkins/MaterialFunction/MF_EmissiveFresnel}.Result
   Mask=StaticComponentMaskParameter(None=@n4) {defaultR=true, group=Skin FX, parameterName=InteriorGlow Mask}
  ) {materialFunction=/Game/Weapons/WeaponSkins/MaterialFunction/MF_InteriorGlow}
  n71 = StaticSwitchParameter(
   True=MaterialFunctionCall!(In=@n70.Result) {materialFunction=/Game/Weapons/WeaponSkins/MaterialFunction/HueShiftColorByAngle}.Result
   False=@n70.Result
  ) {group=Skin FX, parameterName=UseColorShiftByAngle}
  n72 = StaticSwitchParameter(
   True=MaterialFunctionCall!(MA=@n71, FXMask=NamedRerouteUsage(), PreskinLocalPosition=NamedRerouteUsage()) {materialFunction=/Game/Weapons/WeaponSkins/MaterialFunction/RemapCurveByFresnel}.Result
   False=@n71
  ) {group=Skin FX, parameterName=UseRemapCurveByFresnel}
  n73 = StaticSwitchParameter(
   True=MaterialFunctionCall!(MA=@n72, FXMask=NamedRerouteUsage(), PreskinLocalPosition=NamedRerouteUsage()) {materialFunction=/Game/Weapons/WeaponSkins/MaterialFunction/MF_MirroredPanningEmissivePreskinPosition}.Result
   False=@n72
  ) {group=Skin FX, parameterName=UseMirroredPanningEmissivePreskinPosition}
  n74 = StaticSwitchParameter(
   True=MaterialFunctionCall!(MA=@n73, PreskinnedLocalPoistion=NamedRerouteUsage()) {materialFunction=&/MF_Fringe_Effects}.Result
   False=@n73
  ) {group=Skin FX, parameterName=UseFringe}
  n75 = StaticSwitchParameter(
   True=MaterialFunctionCall!(MA=@n74, Mask=NamedRerouteUsage()) {materialFunction=/Game/Weapons/WeaponSkins/MaterialFunction/MF_GrowingNoise}.Result
   False=@n74
  ) {group=Skin FX, parameterName=UseGrowingNoise}
  n76 = MaterialFunctionCall!(
   In=MaterialFunctionCall!(
    In=MaterialFunctionCall!(In=@n75) {materialFunction=/Game/Weapons/WeaponSkins/MaterialFunction/MF_AnisotropicShading}.Result
    Alpha=NamedRerouteUsage()
   ) {materialFunction=/Game/Weapons/WeaponSkins/MaterialFunction/MF_BananaWinterMouth}.Result
   SkinFX_Mask=NamedRerouteUsage()
  ) {materialFunction=/Game/Weapons/WeaponSkins/MaterialFunction/MF_GlitterFX}
  n77 = MaterialFunctionCall!(In=MaterialFunctionCall!(In=MaterialFunctionCall!(MA=@n76.Result) {materialFunction=/Game/Packages/Fortress_SharedMaterials/Base_Material_Functions/MFC_EyeShader}.Result) {materialFunction=/Game/Animation/Props/Gadgets_High_Tech_Backpack/Materials/MF_Gadget_High_Tech_Backpack}.Result) {materialFunction=/Game/Packages/Fortress_SharedMaterials/Base_Material_Functions/MFC_ReactiveEmissive}
  n78 = NamedRerouteUsage()
  n79 = MaterialFunctionCall!(
   MA=MaterialFunctionCall!(
    In=MaterialFunctionCall!(In=@n77.Result, Skin FX Mask=@n78) {materialFunction=/Game/Packages/Fortress_SharedMaterials/Base_Material_Functions/MFC_ColorOverride}.Result
    Skin FX Mask=@n78
   ) {materialFunction=/Game/Packages/Fortress_SharedMaterials/Base_Material_Functions/MFC_DefectMetal}.Result
   SkinFX=@n78
  ) {materialFunction=/Game/Packages/Fortress_SharedMaterials/Base_Material_Functions/MFC_GalaxyScreenSpace}
  FunctionOutput(None=MaterialFunctionCall!(
    Material=MaterialFunctionCall!(MA=@n79.Result, SkinFX=@n78) {materialFunction=/Game/Packages/Fortress_SharedMaterials/Base_Material_Functions/MFC_FlatShading}.Result
    SkinFX_Mask=@n78
   ) {materialFunction=/Game/Global/MaterialFunctions/MF_ClothSimWPO_v2}.Result)
  FunctionOutput(None=@n4) {outputName=SkinFX_Mask, sortPriority=10}
  NamedRerouteUsage()
  Reroute(None=@n4)
  StaticComponentMaskParameter(None=@n4) {defaultR=true, group=Skin FX, parameterName=Custom Color GMap Channel}
  StaticComponentMaskParameter(None=@n4) {defaultR=true, group=Skin FX, parameterName=Custom Color Dirt Channel}
  StaticComponentMaskParameter(None=@n4) {defaultR=true, group=Skin FX, parameterName=Marshmello Channel}
  StaticComponentMaskParameter(None=@n4) {defaultR=true, group=Skin FX, parameterName=Marshmello Low Channel}
  StaticComponentMaskParameter(None=@n4) {defaultG=true, group=Skin FX, parameterName=Marshmello Offset Channel}
  StaticComponentMaskParameter(None=@n4) {defaultG=true, group=Skin FX, parameterName=Marshmello High Channel}
  StaticComponentMaskParameter(None=@n4) {defaultG=true, group=Skin FX, parameterName=Numbers Optimized Channel}
  StaticComponentMaskParameter(None=@n4) {defaultR=true, group=Skin FX, parameterName=MetallicLiquid_Channel}
  StaticComponentMaskParameter(None=@n4) {defaultR=true, group=Skin FX, parameterName=MetallicLiquidIntensityMask_Channel}
  n80 = StaticComponentMaskParameter(None=@n4) {defaultR=true, group=Skin FX, parameterName=Custom Number Channel}
  n81 = StaticSwitchParameter(
   True=MaterialFunctionCall!(Alpha, Material) {materialFunction=/Game/Weapons/WeaponSkins/MaterialFunction/CustomNumbers_Optimized}.Result
   False=StaticSwitchParameter(
    True=MaterialFunctionCall!(Alpha=@n80, Material=@n33.Result) {materialFunction=/Game/Weapons/WeaponSkins/MaterialFunction/CustomNumbers_DoNotUse}.Result
    False=@n33.Result
   ) {group=Skin FX, parameterName="Use OLD Custom Numbers (Do Not Use)"}
  ) {group=Skin FX, parameterName=Use Custom Numbers Optimized}
  StaticSwitchParameter(True=MaterialFunctionCall!(Alpha, Material=@n81) {materialFunction=/Game/Weapons/WeaponSkins/MaterialFunction/UV_Rotator}.Result, False=@n81) {group=Skin FX, parameterName=Use Rotator}
  StaticSwitchParameter(True=@n10.Result, False=@n9) {group=Skin FX, parameterName=UseSuperLeveling}
 ]}
