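# Material function in the shared character shader set that reduces camera clipping by nudging nearby vertices away from the camera.
# A zero-hardness SphereMask on camera distance scales a push along the normalized camera-to-position vector, strongest closest to the camera and fading out over a short radius.
# The push is added to the incoming World Position Offset of the Material Attributes.
# A static switch, Enable AntiCameraClipping, otherwise passes the attributes through unchanged.
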MaterialFunction /Game/Packages/Fortress_SharedMaterials/Base_Material_Functions/BaseCharacterShaderFunctions/MF_AntiCameraClipping
graph {nodes [
  n1 = FunctionInput() {inputType=FunctionInput_MaterialAttributes}
  n2 = GetMaterialAttributes(None=@n1) {attributeGetTypes [#1]}
  n3 = WorldPosition()
  n4 = CameraPositionWS()
  n5 = Subtract(A=@n3.XYZ, B=@n4)
  n6 = SphereMask(A=Distance(A=@n3.XYZ, B=@n4), B=Constant()) {attenuationRadius=20, hardnessPercent=0}
  n7 = Add(A=@n2.WorldPositionOffset, B=Multiply(A=Normalize(VectorInput=@n5), B=LinearInterpolate(Alpha=@n6) {constB=11}))
  FunctionOutput(None=StaticSwitchParameter(
    True=SetMaterialAttributes!(MaterialAttributes=@n2.MaterialAttributes, WorldPositionOffset=@n7) {attributeSetTypes [#1]}
    False=@n1
   ) {group=Anti Camera Clipping, parameterName=Enable AntiCameraClipping})
 ]}
