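# Material function for first-person meshes that squashes vertices toward the camera through World Position Offset, to prevent them penetrating scene geometry.
# One branch scales depth past the near plane by DepthScalar (default 0.1).
# A custom HLSL branch applies FOV correction and first-person scale, with static switches for separate FOV and uniform scaling.
# EnableFPVertexScaling gates the override, and ShadowReplace keeps the original offset for shadows.
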MaterialFunction /Game/Packages/Fortress_SharedMaterials/Base_Material_Functions/BaseCharacterShaderFunctions/MF_FPMesh_PreventGeometryPenetration
graph {nodes [
  n1 = FunctionInput() {inputType=FunctionInput_MaterialAttributes}
  n2 = GetMaterialAttributes(None=@n1) {attributeGetTypes [#1]}
  n3 = ShadowReplace(Default=NamedRerouteUsage(), Shadow=@n2.WorldPositionOffset)
  FunctionOutput(None=StaticSwitchParameter(
    True=SetMaterialAttributes!(MaterialAttributes=@n2.MaterialAttributes, WorldPositionOffset=@n3) {attributeSetTypes [#1]}
    False=@n1
   ) {group=FP Mode, parameterName=EnableFPVertexScaling})
  n4 = WorldPosition() {worldPositionShaderOffset=WPT_CameraRelative}
  n5 = Normalize(VectorInput=@n4.XYZ)
  n6 = MaterialFunctionCall() {materialFunction=&CameraDirectionVector}
  n7 = Multiply(
   A=Divide(A=@n5, B=DotProduct(A=@n5, B=@n6.Result))
   B=Custom() {code=View.NearPlane, description=NearClipPlane, outputType=CMOT_Float1, showCode=true, inputs []}
  )
  n8 = Subtract(A=@n4.XYZ, B=@n7)
  n9 = DotProduct(A=@n8, B=MaterialFunctionCall() {materialFunction=&CameraDirectionVector}.Result)
  n10 = Multiply(A=@n8, B=Max(A=Step(Y=@n9) {constX=0}, B=NamedRerouteUsage()))
  NamedRerouteDeclaration(None=Subtract(A=Add(A=@n10, B=@n7), B=@n4.XYZ)) {name=UniformScaleWPO, nodeColor=<rgba 0.0547 0 1 1>}
  MaterialExpressionWorldPosition_2 = WorldPosition() {worldPositionShaderOffset=WPT_CameraRelative}
  MaterialExpressionMaterialFunctionCall_3 = MaterialFunctionCall() {materialFunction=&CameraDirectionVector}
  MaterialExpressionMaterialFunctionCall_4 = MaterialFunctionCall() {materialFunction=/Engine/Functions/Engine_MaterialFunctions02/Utility/FOV}
  MaterialExpressionStaticSwitchParameter_2 = StaticSwitchParameter(
   True=MaterialFunctionCall!(Degrees=Constant() {r=90}) {materialFunction=/Engine/Functions/Engine_MaterialFunctions01/Math/DegreesToRadians}.Radians
   False=@MaterialExpressionMaterialFunctionCall_4.FOV in radians
  ) {group=FP Mode, parameterName=SeparateFOV}
  MaterialExpressionStaticSwitchParameter_1 = StaticSwitchParameter(True=NamedRerouteUsage(), False=Constant() {r=1}) {defaultValue=true, group=FP Mode, parameterName=UniformVertexScaling}
  n11 = Custom!(
   Position=@MaterialExpressionWorldPosition_2.XYZ
   CameraDirection=@MaterialExpressionMaterialFunctionCall_3.Result
   SceneFOV=@MaterialExpressionMaterialFunctionCall_4.FOV in radians
   FirstPersonFOV=@MaterialExpressionStaticSwitchParameter_2
   FirstPersonScale=@MaterialExpressionStaticSwitchParameter_1
  ) {
   code="float3 Projection = dot(Position, CameraDirection) * CameraDirection;\r\nfloat3 Rejection = Position - Projection;\r\nfloat FOVCorrection = tan(SceneFOV * 0.5f) / tan(FirstPersonFOV * 0.5f);\r\nfloat3 FOVCorrectedPosition = (FOVCorrection - 1.0f) * Rejection + Position;\r\nreturn FOVCorrectedPosition * FirstPersonScale - Position;"
   showCode=true
   inputs [
    {
     inputName=Position
     input {
      expression=&self:MaterialExpressionWorldPosition_2
      mask=1
      maskB=1
      maskG=1
      maskR=1
     }
    }
    {
     inputName=CameraDirection
     input {expression=&self:MaterialExpressionMaterialFunctionCall_3}
    }
    {
     inputName=SceneFOV
     input {
      expression=&self:MaterialExpressionMaterialFunctionCall_4
      outputIndex=1
     }
    }
    {
     inputName=FirstPersonFOV
     input {expression=&self:MaterialExpressionStaticSwitchParameter_2}
    }
    {
     inputName=FirstPersonScale
     input {expression=&self:MaterialExpressionStaticSwitchParameter_1}
    }
   ]
  }
  NamedRerouteDeclaration(None=@n11) {name=CalculateWPO, nodeColor=<rgba 1 0.4922 0 1>}
  n12 = FunctionInput() {
   bUsePreviewValueAsDefault=true
   desc=Amount to scale filtered verts by
   inputName=DepthScalar
   inputType=FunctionInput_Scalar
   sortPriority=1
   previewValue=<0.1 0 0 1>
  }
  NamedRerouteDeclaration(None=@n12) {name=ParamInputDepthScalar, nodeColor=<rgba 0.2656 0 1 1>}
  NamedRerouteUsage()
  TruncateLWC!(None=Subtract(A=WorldPosition().XYZ, B=CameraPositionWS()))
 ]}
