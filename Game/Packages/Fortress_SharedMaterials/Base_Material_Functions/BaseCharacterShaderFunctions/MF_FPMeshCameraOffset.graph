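# Material function for character first-person mode that pushes masked parts of the mesh along the view axis through World Position Offset.
# A square or sphere mask built from pre-skinned position weights the offset, with tuning from FortniteMaterialParameters or local parameters.
# A Custom node uses FP_WPOStrength to switch between normal and FP offsets.
# Static switches add geometry-penetration prevention and debug color; the ShadowReplace branch is deliberately disconnected because FP meshes cast no shadows.
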MaterialFunction /Game/Packages/Fortress_SharedMaterials/Base_Material_Functions/BaseCharacterShaderFunctions/MF_FPMeshCameraOffset
graph {
 properties {libraryCategoriesText [Misc, ""]}
 nodes [
  CollectionParameter() {collection=&FortniteMaterialParameters, parameterName=FP_Active}
  Constant() {r=1}
  Divide(B=PreSkinnedLocalBounds().Max)
  MaterialExpressionScalarParameter_0 = ScalarParameter() {group=FP Mode, parameterName=FP_WPOStrength, sortPriority=50}
  MaterialExpressionNamedRerouteUsage_33 = NamedRerouteUsage()
  MaterialExpressionNamedRerouteUsage_34 = NamedRerouteUsage()
  n1 = Custom!(FP=@MaterialExpressionScalarParameter_0, WPO=@MaterialExpressionNamedRerouteUsage_33, WPO_FP=@MaterialExpressionNamedRerouteUsage_34) {
   code="if(FP)\r\n{\r\n return WPO_FP;\r\n}\r\nreturn WPO;"
   inputs [
    {
     inputName=FP
     input {expression=&self:MaterialExpressionScalarParameter_0}
    }
    {
     inputName=WPO
     input {expression=&self:MaterialExpressionNamedRerouteUsage_33}
    }
    {
     inputName=WPO_FP
     input {expression=&self:MaterialExpressionNamedRerouteUsage_34}
    }
   ]
  }
  n2 = SetMaterialAttributes!(MaterialAttributes=NamedRerouteUsage(), WorldPositionOffset=@n1) {attributeSetTypes [#1]}
  n3 = NamedRerouteUsage()
  FunctionOutput(None=StaticSwitchParameter(True=SetMaterialAttributes!(MaterialAttributes=@n2, BaseColor=@n3) {attributeSetTypes [#2]}, False=@n2) {group=FP Mode, parameterName=FP_DebugColor, sortPriority=20})
  n4 = NamedRerouteUsage()
  LinearInterpolate(A=Constant3Vector() {constant=<rgba 0 1 0 0>}, B=Constant3Vector() {constant=<rgba 1 0 0 0>}, Alpha=@n4)
  n5 = NamedRerouteUsage()
  Multiply(A=MaterialFunctionCall() {materialFunction=&CameraDirectionVector}.Result, B=@n5)
  n6 = NamedRerouteUsage()
  n7 = NamedRerouteUsage()
  n8 = StaticBoolParameter() {defaultValue=true, group=FP Mode, parameterName=FP_SquareMask, sortPriority=40}
  NamedRerouteDeclaration(None=StaticSwitch(True=VertexInterpolator(VS=@n6).PS, False=VertexInterpolator(VS=@n7).PS, Value=@n8)) {name=MaskDebug, nodeColor=<rgba 0 0.8594 1 1>}
  n9 = Constant()
  n10 = Transform(None=MaterialFunctionCall(X=@n9, Y=@n9, Z=@n5) {materialFunction=&MakeFloat3}.Result) {transformSourceType=TRANSFORMSOURCE_View}
  n11 = StaticSwitch(True=NamedRerouteUsage(), False=NamedRerouteUsage(), Value=@n8)
  NamedRerouteDeclaration(None=Add(A=NamedRerouteUsage(), B=LinearInterpolate(B=@n10, Alpha=@n11))) {name=WPO_FP, nodeColor=<rgba 0 0 0 1>}
  NamedRerouteDeclaration(None=StaticSwitch(
    True=CollectionParameter() {collection=&FortniteMaterialParameters, parameterName=FP_MaskRadius}
    False=ScalarParameter() {defaultValue=40, group=FP Mode, parameterName=FP_MaskRadius, sortPriority=50}
    Value=NamedRerouteUsage()
   )) {name=ParamMaskRadius, nodeColor=<rgba 0.3359 1 0 1>}
  n12 = CollectionParameter() {collection=&FortniteMaterialParameters, parameterName=FP_MaskHardness}
  NamedRerouteDeclaration(None=StaticSwitch(
    True=@n12
    False=ScalarParameter() {defaultValue=0.2, group=FP Mode, parameterName=FP_MaskHardness, sortPriority=50}
    Value=NamedRerouteUsage()
   )) {name=ParamMaskHardness, nodeColor=<rgba 1 0 0.2344 1>}
  NamedRerouteDeclaration(None=StaticSwitch(
    True=CollectionParameter() {collection=&FortniteMaterialParameters, parameterName=FP_MaskPosZ}
    False=ScalarParameter() {defaultValue=150, group=FP Mode, parameterName=FP_MaskPosZ, sortPriority=50}
    Value=NamedRerouteUsage()
   )) {name=ParamMaskPosZ, nodeColor=<rgba 0.3359 1 0 1>}
  n13 = PreSkinnedPosition()
  n14 = MaterialFunctionCall(Float3=@n13) {materialFunction=&BreakOutFloat3Components}
  n15 = NamedRerouteUsage()
  n16 = SphereMask(A=@n14.R, B=Constant(), Radius=NamedRerouteUsage(), Hardness=@n15)
  n17 = NamedRerouteUsage()
  n18 = NamedRerouteUsage()
  n19 = Divide(A=@n14.B, B=@n17)
  n20 = Saturate(None=Add(A=SphereMask(A=@n14.B, B=@n17, Radius=@n18, Hardness=@n15) {attenuationRadius=30, hardnessPercent=50}, B=Floor(None=@n19)))
  NamedRerouteDeclaration(None=Saturate(None=Multiply(A=@n16, B=@n20))) {name=SquareMask, nodeColor=<rgba 1 0.1641 0 1>}
  n21 = Floor(None=Divide(
    A=MaterialFunctionCall(Float3=@n13) {materialFunction=&BreakOutFloat3Components}.B
    B=NamedRerouteUsage()
   ))
  n22 = SphereMask(
   A=@n13
   B=MaterialFunctionCall(Z=NamedRerouteUsage()) {materialFunction=&MakeFloat3}.Result
   Radius=NamedRerouteUsage()
   Hardness=NamedRerouteUsage()
  )
  NamedRerouteDeclaration(None=Saturate(None=Add(A=@n21, B=@n22))) {name=SphereMask, nodeColor=<rgba 0.7812 1 0 1>}
  n23 = FunctionInput() {inputType=FunctionInput_MaterialAttributes}
  NamedRerouteDeclaration(None=@n23) {name=MA, nodeColor=<rgba 0.0547 1 0 1>}
  NamedRerouteDeclaration(None=StaticBoolParameter() {group=FP Mode, parameterName=PreventGeoPenetration}) {name=ParamPreventGeoPenetration, nodeColor=<rgba 0 0.1328 1 1>}
  NamedRerouteDeclaration(None=Constant() {r=0.1}) {name=ParamPreventGeoPenetrationDepthScalar, nodeColor=<rgba 0.3594 1 0 1>}
  NamedRerouteDeclaration(None=StaticSwitch(
    True=CollectionParameter() {collection=&FortniteMaterialParameters, parameterName=FP_MaskWidth}
    False=ScalarParameter() {defaultValue=120, group=FP Mode, parameterName=FP_MaskWidth, sortPriority=50}
    Value=NamedRerouteUsage()
   )) {name=ParamWidth, nodeColor=<rgba 1 0 0.0703 1>}
  NamedRerouteDeclaration(None=StaticSwitch(
    True=CollectionParameter() {collection=&FortniteMaterialParameters, parameterName=FP_MaskHeight}
    False=ScalarParameter() {defaultValue=25, group=FP Mode, parameterName=FP_MaskHeight, sortPriority=50}
    Value=NamedRerouteUsage()
   )) {name=ParamHeight, nodeColor=<rgba 0 1 0.3438 1>}
  NamedRerouteDeclaration(None=@n4) {name=DebugOutput, nodeColor=<rgba 1 0 0.3516 1>}
  n24 = MaterialFunctionCall!(In=@n23, DepthScalar=NamedRerouteUsage()) {materialFunction=&/MF_FPMesh_PreventGeometryPenetration}
  n25 = NamedRerouteUsage()
  NamedRerouteDeclaration(None=GetMaterialAttributes(None=StaticSwitch(True=@n24.Result, False=@n23, Value=@n25)) {attributeGetTypes [#1]}.WorldPositionOffset) {name=WPO, nodeColor=<rgba 0.9688 1 0 1>}
  NamedRerouteDeclaration(None=StaticBoolParameter() {group=FP Mode, parameterName=FP_UseMPC, sortPriority=40}) {name=UseMPC, nodeColor=<rgba 0 1 0.5547 1>}
  NamedRerouteDeclaration(None=StaticSwitch(
    True=CollectionParameter() {collection=&FortniteMaterialParameters, parameterName=FP_MaskPower}
    False=ScalarParameter() {defaultValue=4, group=FP Mode, parameterName=FP_Power, sortPriority=50}
    Value=NamedRerouteUsage()
   )) {name=ParamPower, nodeColor=<rgba 1 0.1875 0 1>}
  NamedRerouteDeclaration(None=StaticSwitch(
    True=CollectionParameter() {collection=&FortniteMaterialParameters, parameterName=FP_MeshOffset}
    False=ScalarParameter() {defaultValue=-25, group=FP Mode, parameterName=FP_MeshOffset, sortPriority=50}
    Value=NamedRerouteUsage()
   )) {name=ParamMeshOffset, nodeColor=<rgba 0.875 1 0 1>}
  n26 = NamedRerouteUsage()
  n27 = MaterialFunctionCall() {materialFunction=&BreakOutFloat3Components}
  n28 = ComponentMask(None=Add(A=@n27.R, B=@n26)) {r=true}
  n29 = NamedRerouteUsage()
  n30 = OneMinus(None=@n27.R)
  n31 = ComponentMask(None=Add(A=Add(A=@n30) {constB=-1}, B=@n26)) {r=true}
  n32 = Multiply(A=Saturate(None=Multiply(A=@n28, B=@n29) {constB=100}), B=Saturate(None=Multiply(A=@n31, B=@n29) {constB=100}))
  n33 = NamedRerouteUsage()
  n34 = Multiply(A=ComponentMask(None=Add(A=@n27.B, B=@n33)) {b=true}, B=@n29) {constB=100}
  Power(Base=Saturate(None=Multiply(A=@n32, B=@n34)), Exp=NamedRerouteUsage())
  ShadowReplace(Default=@n1, Shadow=NamedRerouteUsage()) {desc=Leaving disconnected because the FP mesh doesn't cast shadows}
  Transform(None=MaterialFunctionCall(X=@n9, Y=@n5, Z=@n9) {materialFunction=&MakeFloat3}.Result) {transformSourceType=TRANSFORMSOURCE_Local}
 ]
}
