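# Material function for character shaders that sets First Person Interpolation Alpha per position to blend the first-person FOV effect.
# Pre-skinned instance-local height becomes a 0-1 alpha, applied inside a box whose width and depth lerp with height; outside, alpha is 1.
# Custom HLSL applies this only when FP_FOVStrength is non-zero, otherwise keeping the incoming alpha.
# A debug static switch tints areas with reduced alpha red.
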MaterialFunction /Game/Packages/Fortress_SharedMaterials/Base_Material_Functions/BaseCharacterShaderFunctions/MF_InterpFP_FOV
graph {nodes [
  n1 = FunctionInput() {inputType=FunctionInput_MaterialAttributes}
  MaterialExpressionScalarParameter_0 = ScalarParameter() {group=FP Mode, parameterName=FP_FOVStrength, sortPriority=50}
  MaterialExpressionGetMaterialAttributes_0 = GetMaterialAttributes(None=@n1) {attributeGetTypes [#1]}
  n2 = LocalPosition() {localOrigin=InstancePreSkinning}
  MaterialExpressionMaterialFunctionCall_1 = MaterialFunctionCall(Float2=@n2.XY) {materialFunction=&BreakOutFloat2Components}
  MaterialExpressionSaturate_6 = Saturate(None=InverseLinearInterpolate(B=ScalarParameter() {defaultValue=120, group=FP Mode, parameterName=FP_FOVInterpolation_BlendHeight}, Value=@n2.Z) {constB=120})
  MaterialExpressionScalarParameter_3 = ScalarParameter() {defaultValue=35, group=FP Mode, parameterName=FP_LerpFOV_BottomWidth}
  MaterialExpressionScalarParameter_5 = ScalarParameter() {defaultValue=20, group=FP Mode, parameterName=FP_LerpFOV_TopWidth}
  MaterialExpressionScalarParameter_7 = ScalarParameter() {defaultValue=90, group=FP Mode, parameterName=FP_LerpFOV_BottomDepth}
  MaterialExpressionScalarParameter_6 = ScalarParameter() {defaultValue=90, group=FP Mode, parameterName=FP_LerpFOV_TopDepth}
  MaterialExpressionCustom_2 = Custom!(
   XPos=@MaterialExpressionMaterialFunctionCall_1.R
   YPos=@MaterialExpressionMaterialFunctionCall_1.R
   AdjustedAlpha=@MaterialExpressionSaturate_6
   BottomWidth=@MaterialExpressionScalarParameter_3
   TopWidth=@MaterialExpressionScalarParameter_5
   BottomDepth=@MaterialExpressionScalarParameter_7
   TopDepth=@MaterialExpressionScalarParameter_6
  ) {
   code="float Width = lerp(BottomWidth, TopWidth, AdjustedAlpha);\r\nfloat Depth = lerp(BottomDepth, TopDepth, AdjustedAlpha);\r\n\r\nif (abs(XPos) < Width &&&& abs(YPos) < Depth)\r\n{\r\n return AdjustedAlpha;\r\n}\r\nreturn 1.0f;"
   outputType=CMOT_Float1
   showCode=true
   inputs [
    {
     inputName=XPos
     input {expression=&self:MaterialExpressionMaterialFunctionCall_1}
    }
    {
     inputName=YPos
     input {
      expression=&self:MaterialExpressionMaterialFunctionCall_1
      outputIndex=1
     }
    }
    {
     inputName=AdjustedAlpha
     input {expression=&self:MaterialExpressionSaturate_6}
    }
    {
     inputName=BottomWidth
     input {expression=&self:MaterialExpressionScalarParameter_3}
    }
    {
     inputName=TopWidth
     input {expression=&self:MaterialExpressionScalarParameter_5}
    }
    {
     inputName=BottomDepth
     input {expression=&self:MaterialExpressionScalarParameter_7}
    }
    {
     inputName=TopDepth
     input {expression=&self:MaterialExpressionScalarParameter_6}
    }
   ]
  }
  n3 = Custom!(
   FP=@MaterialExpressionScalarParameter_0
   StartingAlpha=@MaterialExpressionGetMaterialAttributes_0.FirstPersonInterpolationAlpha
   AdjustedAlpha=@MaterialExpressionCustom_2
  ) {
   code="if(FP)\r\n{\r\n return AdjustedAlpha;\r\n}\r\nreturn StartingAlpha;"
   outputType=CMOT_Float1
   inputs [
    {
     inputName=FP
     input {expression=&self:MaterialExpressionScalarParameter_0}
    }
    {
     inputName=StartingAlpha
     input {
      expression=&self:MaterialExpressionGetMaterialAttributes_0
      outputIndex=1
     }
    }
    {
     inputName=AdjustedAlpha
     input {expression=&self:MaterialExpressionCustom_2}
    }
   ]
  }
  n4 = SetMaterialAttributes!(MaterialAttributes=@n1, FirstPersonInterpolationAlpha=@n3) {attributeSetTypes [#1]}
  n5 = GetMaterialAttributes(None=@n4) {attributeGetTypes [#2]}
  n6 = LinearInterpolate(
   A=Constant3Vector() {constant=<rgba 1 0 0 0>}
   B=@n5.BaseColor
   Alpha=Step(X=VertexInterpolator(VS=@MaterialExpressionCustom_2).PS) {constY=0.999}
  )
  FunctionOutput(None=StaticSwitchParameter(
    True=SetMaterialAttributes!(MaterialAttributes=@n5.MaterialAttributes, BaseColor=@n6) {attributeSetTypes [#2]}
    False=@n4
   ) {group=FP Mode, parameterName=FP_DebugInterpFOV})
 ]}
