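# Material function for character shaders that scales an opacity value, presumably an outline's, down as the Season 22 chrome character effect strengthens.
# It inverts the S22ChromeCharacters_strength scalar parameter and multiplies it with the Opacity input, so full strength yields zero.
# The inverted value passes through a ShadingPathSwitch with only its default input connected, noted as migrated from a FeatureLevelSwitch.
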MaterialFunction /Game/Packages/Fortress_SharedMaterials/Base_Material_Functions/BaseCharacterShaderFunctions/MF_TechArt_ChromeOutlineOverride
graph {nodes [
  n1 = OneMinus(None=ScalarParameter() {group=S22ChromeCharacters, parameterName=S22ChromeCharacters_strength, sliderMax=1})
  FunctionOutput(None=Multiply(
    A=ShadingPathSwitch(Default=@n1) {desc=Migrated from FeatureLevelSwitch}
    B=FunctionInput() {inputName=Opacity, inputType=FunctionInput_Scalar}
   ))
 ]}
